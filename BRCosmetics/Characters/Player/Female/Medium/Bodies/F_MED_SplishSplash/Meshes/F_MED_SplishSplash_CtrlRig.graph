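# Control Rig for the F_MED_SplishSplash player body that drives corrective motion on skirt, slipper, holster and forearm twist bones.
# Parent constraints attach skirt and accessory bones, while thigh, calf and hand rotations are extracted through swing-twist decomposition.
# These angles, pose readers, pelvis height and curves weight layered additive bone offsets.
# Interpolated gameplay flags blend special skirt shapes for crouching, strafing, sliding, rail grinding and DBNO.
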ControlRigBlueprint /BRCosmetics/Characters/Player/Female/Medium/Bodies/F_MED_SplishSplash/Meshes/F_MED_SplishSplash_CtrlRig
graphs [
 CollapseNode_ContainedGraph {nodes [
   n1 = RigVMFunctionEntryNode()
   &RigUnit_ParentConstraint(
    AdvancedSettings=(InterpolationType=Average,RotationOrderForFilter=XZY)
    Child=(Type=Bone,Name="base_L_skirt_a_1")
    Parents=((Item=(Type=Bone,Name="thigh_twist_01_l"),Weight=1))
    Weight=1
    ExecutePin=@n1.ExecutePin
   )
   > &RigUnit_ParentConstraint(
    AdvancedSettings=(InterpolationType=Average,RotationOrderForFilter=XZY)
    Child=(Type=Bone,Name="base_L_skirt_b_1")
    Parents=((Item=(Type=Bone,Name="thigh_twist_01_l"),Weight=1))
    Weight=1
   )
   > &RigUnit_ParentConstraint(
    AdvancedSettings=(InterpolationType=Average,RotationOrderForFilter=XZY)
    Child=(Type=Bone,Name="base_L_skirt_c_1")
    Parents=((Item=(Type=Bone,Name="thigh_twist_01_l"),Weight=1))
    Weight=1
   )
   > &RigUnit_ParentConstraint(
    AdvancedSettings=(InterpolationType=Average,RotationOrderForFilter=XZY)
    Child=(Type=Bone,Name="base_R_skirt_a_1")
    Parents=((Item=(Type=Bone,Name="thigh_twist_01_r"),Weight=1))
    Weight=1
   )
   > &RigUnit_ParentConstraint(
    AdvancedSettings=(InterpolationType=Average,RotationOrderForFilter=XZY)
    Child=(Type=Bone,Name="base_R_skirt_b_1")
    Parents=((Item=(Type=Bone,Name="thigh_twist_01_r"),Weight=1))
    Weight=1
   )
   > &RigUnit_ParentConstraint(
    AdvancedSettings=(InterpolationType=Average,RotationOrderForFilter=XZY)
    Child=(Type=Bone,Name="base_R_skirt_c_1")
    Parents=((Item=(Type=Bone,Name="thigh_twist_01_r"),Weight=1))
    Weight=1
   )
   > &RigUnit_ParentConstraint(
    AdvancedSettings=(InterpolationType=Average,RotationOrderForFilter=XZY)
    Child=(Type=Bone,Name="base_C_skirt_a_1")
    Parents=((Item=(Type=Bone,Name="thigh_twist_01_l"),Weight=0.5),(Item=(Type=Bone,Name="thigh_twist_01_r"),Weight=0.5))
    Weight=1
   )
   > n2 = &RigUnit_ParentConstraint(
    AdvancedSettings=(InterpolationType=Average,RotationOrderForFilter=XZY)
    Child=(Type=Bone,Name="base_C_skirt_b_1")
    Parents=((Item=(Type=Bone,Name="thigh_twist_01_l"),Weight=0.5),(Item=(Type=Bone,Name="thigh_twist_01_r"),Weight=0.5))
    Weight=1
   )
   RigVMFunctionReturnNode(ExecutePin=@n2.ExecutePin)
  ]}
 Multiply_12_ContainedGraph {nodes [
   n1 = RigVMFunctionEntryNode()
   RigVMFunctionReturnNode(Result=&RigVMFunction_MathDoubleMul(B=@n1.C, A=&RigVMFunction_MathDoubleMul(A=@n1.A, B=@n1.B).Result).Result)
  ]}
 Multiply_2_2_1_ContainedGraph {nodes [
   n1 = RigVMFunctionEntryNode()
   RigVMFunctionReturnNode(Result=&RigVMFunction_MathDoubleMul(B=@n1.C, A=&RigVMFunction_MathDoubleMul(A=@n1.A, B=@n1.B).Result).Result)
  ]}
 Multiply_2_2_2_2_ContainedGraph {nodes [
   n1 = RigVMFunctionEntryNode()
   RigVMFunctionReturnNode(Result=&RigVMFunction_MathDoubleMul(B=@n1.C, A=&RigVMFunction_MathDoubleMul(A=@n1.A, B=@n1.B).Result).Result)
  ]}
 Multiply_2_3_ContainedGraph {nodes [
   n1 = RigVMFunctionEntryNode()
   RigVMFunctionReturnNode(Result=&RigVMFunction_MathDoubleMul(B=@n1.C, A=&RigVMFunction_MathDoubleMul(A=@n1.A, B=@n1.B).Result).Result)
  ]}
 Multiply_3_1_ContainedGraph {nodes [
   n1 = RigVMFunctionEntryNode()
   RigVMFunctionReturnNode(Result=&RigVMFunction_MathDoubleMul(B=@n1.C, A=&RigVMFunction_MathDoubleMul(A=@n1.A, B=@n1.B).Result).Result)
  ]}
 New Function_ContainedGraph {nodes [
   n1 = RigVMFunctionEntryNode()
   RigVMFunction_MathBoolToInteger = &RigVMFunction_MathBoolToInteger(Value=@n1.onlyTwist)
   RigVMFunction_MathQuaternionSwingTwist = &RigVMFunction_MathQuaternionSwingTwist(
    TwistAxis=@n1.twistAxis
    Input=&RigUnit_GetRelativeTransformForItem(bChildInitial=False, bParentInitial=False, Child=@n1.child, Parent=@n1.parent).RelativeTransform.Rotation
   )
   n2 = RigVMDispatchNode(
    Values=((),())
    Index=@RigVMFunction_MathBoolToInteger.Result
    Values.0=@RigVMFunction_MathQuaternionSwingTwist.Swing
    Values.1=@RigVMFunction_MathQuaternionSwingTwist.Twist
   ) {template="DISPATCH_RigVMDispatch_SelectInt32(in Index,in Values,out Result)"}
   RigVMFunction_MathQuaternionToEuler = &RigVMFunction_MathQuaternionToEuler(Value=@n2.Result, RotationOrder=@n1.Rotation Order)
   > DISPATCH_RigVMDispatch_Print = RigVMDispatchNode(ScreenColor=(R=1,G=1,B=1,A=1), ScreenDuration=1.5, Enabled=@n1.debugRotation, Value=@RigVMFunction_MathQuaternionToEuler.Result) {template="DISPATCH_RigVMDispatch_Print(in Prefix,in Value,in Enabled,in ScreenDuration,in ScreenColor)"}
   Cast = RigVMDispatchNode(Value=@n1.triggerAxis) {template="DISPATCH_RigVMDispatch_CastEnumToInt(in Value,out Result)"}
   DISPATCH_RigVMDispatch_SelectInt32 = RigVMDispatchNode(
    Values=(0,0,0)
    Index=@Cast.Result
    Values.0=@RigVMFunction_MathQuaternionToEuler.Result.X
    Values.1=@RigVMFunction_MathQuaternionToEuler.Result.Y
    Values.2=@RigVMFunction_MathQuaternionToEuler.Result.Z
   ) {template="DISPATCH_RigVMDispatch_SelectInt32(in Index,in Values,out Result)"}
   Remap = &RigVMFunction_MathDoubleRemap(
    bClamp=False
    Value=@DISPATCH_RigVMDispatch_SelectInt32.Result
    TargetMinimum=@n1.outputMinMax.X
    TargetMaximum=@n1.outputMinMax.Y
    SourceMinimum=@n1.rotationMinMax.X
    SourceMaximum=@n1.rotationMinMax.Y
   )
   Absolute = &RigVMFunction_MathDoubleAbs(Value=@Remap.Result)
   Multiply = &RigVMFunction_MathDoubleMul(
    A=&RigVMFunction_MathDoubleClamp(Value=RigVMDispatchNode(True=@Absolute.Result, Condition=@n1.absoluteValue, False=@Remap.Result) {template="DISPATCH_RigVMDispatch_If(in Condition,in True,in False,out Result)"}.Result).Result
    B=@n1.premult
   )
   > RigVMFunctionReturnNode(outputQuat=@n2.Result, outputWeight=@Multiply.Result)
  ]}
 New Function_ContainedGraph {nodes [
   RigVMFunctionEntryNode()
   > DeltaRotation v4 = RigVMFunctionReferenceNode(
    Rotation Order=ZYX
    child=(Type=Bone,Name="hand_l")
    onlyTwist=True
    outputMinMax=(X=0,Y=1)
    parent=(Type=Null,Name="hand_l_null")
    premult=1
    rotationMinMax=(X=70,Y=100)
    twistAxis=(X=1,Y=0,Z=0)
   )
   > &RigUnit_ModifyTransforms(
    ItemToModify="((Item=(Type=Bone,Name=\"lowerarm_twist_01_l\"),Transform=(Rotation=(X=0,Y=-0.0262,Z=0,W=0.9997),Translation=(X=-0.7,Y=-0.17,Z=1),Scale3D=(X=1,Y=1,Z=1))),(Item=(Type=Bone,Name=\"lowerarm_twist_02_l\"),Transform=(Rotation=(X=0.342,Y=0,Z=0,W=0.9397),Translation= ...(truncated)"
    WeightMinimum=0
    Weight=@"DeltaRotation v4".outputWeight
   )
   > RigVMFunctionReturnNode()
  ]}
 New Function_ContainedGraph {nodes [
   n1 = RigVMFunctionEntryNode()
   RigVMFunction_Sequence = &RigVMFunction_Sequence(ExecuteContext=@n1.ExecuteContext)
   constraints = RigVMCollapseNode(ExecutePin=@RigVMFunction_Sequence.A)
   n2 = &RigVMFunction_AlphaInterp(
    InterpSpeedDecreasing=15
    InterpSpeedIncreasing=15
    Scale=1
    bClampResult=False
    Value=&RigVMFunction_MathBoolToFloat(Value=RigVMVariableNode() {variable=pub isGrindRail}.Value).Result
   )
   RigVMVariableNode(Value=@n2.Result, ExecuteContext=@constraints.ExecutePin) {variable=isGrindRail Lerp}
   n3 = &RigVMFunction_AlphaInterp(
    InterpSpeedDecreasing=10
    InterpSpeedIncreasing=10
    Scale=1
    bClampResult=False
    Value=&RigVMFunction_MathBoolToFloat(Value=RigVMVariableNode() {variable=pub isBaseballSliding}.Value).Result
   )
   > RigVMVariableNode(Value=@n3.Result) {variable=isBaseballSliding Lerp}
   n4 = &RigVMFunction_AlphaInterp(
    InterpSpeedDecreasing=5
    InterpSpeedIncreasing=15
    Scale=1
    bClampResult=False
    Value=&RigVMFunction_MathBoolToFloat(Value=RigVMVariableNode() {variable=pub isKneeSliding}.Value).Result
   )
   > RigVMVariableNode(Value=@n4.Result) {variable=isKneeSliding Lerp}
   n5 = &RigVMFunction_AlphaInterp(
    InterpSpeedDecreasing=15
    InterpSpeedIncreasing=15
    Scale=1
    bClampResult=False
    Value=&RigVMFunction_MathBoolToFloat(Value=RigVMVariableNode() {variable=pub isCrouching}.Value).Result
   )
   > RigVMVariableNode(Value=@n5.Result) {variable=isCrouching Lerp}
   n6 = &RigUnit_GetRelativeTransformForItem(Child=(Type=Bone,Name="calf_l"), Parent=(Type=Null,Name="calf_l_null"), bChildInitial=False, bParentInitial=False)
   > RigVMVariableNode(Value=&RigVMFunction_MathQuaternionToEuler(
     RotationOrder=ZYX
     Value=&RigVMFunction_MathQuaternionSwingTwist(TwistAxis=(X=1,Y=0,Z=0), Input=@n6.RelativeTransform.Rotation).Swing
    ).Result.Z) {variable=l calfLocal rz float}
   n7 = &RigUnit_GetRelativeTransformForItem(Child=(Type=Bone,Name="calf_r"), Parent=(Type=Null,Name="calf_r_null"), bChildInitial=False, bParentInitial=False)
   > RigVMVariableNode(Value=&RigVMFunction_MathQuaternionToEuler(
     RotationOrder=ZYX
     Value=&RigVMFunction_MathQuaternionSwingTwist(TwistAxis=(X=1,Y=0,Z=0), Input=@n7.RelativeTransform.Rotation).Swing
    ).Result.Z) {variable=r calfLocal rz float}
   RigVMFunction_MathQuaternionSwingTwist = &RigVMFunction_MathQuaternionSwingTwist(
    TwistAxis=(X=1,Y=0,Z=0)
    Input=&RigUnit_GetRelativeTransformForItem(Child=(Type=Bone,Name="thigh_l"), Parent=(Type=Null,Name="thigh_l_null"), bChildInitial=False, bParentInitial=False).RelativeTransform.Rotation
   )
   RigVMFunction_MathQuaternionToEuler = &RigVMFunction_MathQuaternionToEuler(RotationOrder=ZYX, Value=@RigVMFunction_MathQuaternionSwingTwist.Swing)
   > RigVMVariableNode(Value=@RigVMFunction_MathQuaternionToEuler.Result.Y) {variable=l thighLocal ry float}
   > RigVMVariableNode(Value=@RigVMFunction_MathQuaternionToEuler.Result.Z) {variable=l thighLocal rz float}
   n8 = &RigVMFunction_MathQuaternionToEuler(
    RotationOrder=ZYX
    Value=&RigVMFunction_MathQuaternionSwingTwist(
     TwistAxis=(X=1,Y=0,Z=0)
     Input=&RigUnit_GetRelativeTransformForItem(Child=(Type=Bone,Name="thigh_r"), Parent=(Type=Null,Name="thigh_r_null"), bChildInitial=False, bParentInitial=False).RelativeTransform.Rotation
    ).Swing
   )
   > RigVMVariableNode(Value=@n8.Result.Y) {variable=r thighLocal ry float}
   > RigVMVariableNode(Value=@n8.Result.Z) {variable=r thighLocal rz float}
   > RigVMFunction_Sequence_1 = RigVMAggregateNode()
   n9 = &RigVMFunction_ControlFlowBranch(ExecuteContext=@RigVMFunction_Sequence_1.B, Condition=RigVMVariableNode() {variable=pub isDBNO}.Value)
   RigVMFunction_AnimRichCurve = /Script/RigVM.RigVMFunction_AnimRichCurve(Curve="(EditorCurveData=(Keys=((InterpMode=RCIM_Cubic,TangentMode=RCTM_User,ArriveTangent=1.5462,LeaveTangent=0.3774),(InterpMode=RCIM_Cubic,Time=0.5,Value=0.379,ArriveTangent=1,LeaveTangent=1),(InterpMode=RCIM_Cubic,TangentMode=RCTM_User,Time=1,Value=1,ArriveTan ...(truncated)") {is=struct RigVMUnitNode}
   n10 = &RigVMFunction_AnimEvalRichCurve(
    SourceMinimum=0
    TargetMinimum=0
    Value=&RigVMFunction_MathDoubleRemap(SourceMaximum=120, SourceMinimum=70, TargetMaximum=1, TargetMinimum=0, Value=RigVMVariableNode() {variable=l calfLocal rz float}.Value).Result
    Curve=@RigVMFunction_AnimRichCurve.Curve
   )
   &RigUnit_ModifyTransforms(
    ItemToModify="((Item=(Type=Bone,Name=\"base_L_skirt_c_2\"),Transform=(Rotation=(X=0,Y=-0.4226,Z=0,W=0.9063),Translation=(X=-2,Y=0,Z=0),Scale3D=(X=0.5,Y=1,Z=1))),(Item=(Type=Bone,Name=\"base_L_skirt_b_2\"),Transform=(Rotation=(X=0,Y=-0.1736,Z=0,W=0.9848),Translation=(X=0,Y=0 ...(truncated)"
    WeightMinimum=0
    Weight=@n10.Result
    ExecutePin=@n9.False
   )
   n11 = &RigVMFunction_AnimEvalRichCurve(
    SourceMinimum=0
    TargetMinimum=0
    Value=&RigVMFunction_MathDoubleRemap(SourceMaximum=120, SourceMinimum=70, TargetMaximum=1, TargetMinimum=0, Value=RigVMVariableNode() {variable=r calfLocal rz float}.Value).Result
    Curve=@RigVMFunction_AnimRichCurve.Curve
   )
   > &RigUnit_ModifyTransforms(
    ItemToModify="((Item=(Type=Bone,Name=\"base_R_skirt_c_2\"),Transform=(Rotation=(X=0,Y=-0.4226,Z=0,W=0.9063),Translation=(X=-2,Y=0,Z=0),Scale3D=(X=0.5,Y=1,Z=1))),(Item=(Type=Bone,Name=\"base_R_skirt_b_2\"),Transform=(Rotation=(X=0,Y=-0.1736,Z=0,W=0.9848),Translation=(X=0,Y=0 ...(truncated)"
    WeightMinimum=0
    Weight=@n11.Result
   )
   > &RigUnit_ModifyTransforms(
    ItemToModify="((Item=(Type=Bone,Name=\"base_C_skirt_b_2\"),Transform=(Rotation=(X=0,Y=0.342,Z=0,W=0.9397),Translation=(X=0,Y=0,Z=1),Scale3D=(X=0.5,Y=1,Z=1))),(Item=(Type=Bone,Name=\"base_L_skirt_c_2\"),Transform=(Rotation=(X=0,Y=-0.0872,Z=0,W=0.9962),Translation=(X=0,Y=0,Z= ...(truncated)"
    WeightMinimum=0
    Weight=&RigVMFunction_MathDoubleMul(A=@n11.Result, B=@n10.Result).Result
   )
   n12 = &RigVMFunction_MathDoubleRemap(SourceMaximum=140, SourceMinimum=120, TargetMaximum=1, TargetMinimum=0, Value=RigVMVariableNode() {variable=l calfLocal rz float}.Value)
   > &RigUnit_ModifyTransforms(
    ItemToModify=((Item=(Type=Bone,Name="base_L_skirt_c_2"),Transform=(Rotation=(X=-0.1149,Y=0.5568,Z=0.1506,W=-0.8088),Translation=(X=-3,Y=0,Z=0),Scale3D=(X=1,Y=1,Z=1))))
    WeightMinimum=0
    Weight=@n12.Result
   )
   n13 = &RigVMFunction_MathDoubleRemap(SourceMaximum=140, SourceMinimum=120, TargetMaximum=1, TargetMinimum=0, Value=RigVMVariableNode() {variable=r calfLocal rz float}.Value)
   > &RigUnit_ModifyTransforms(
    ItemToModify=((Item=(Type=Bone,Name="base_R_skirt_c_2"),Transform=(Rotation=(X=-0.1149,Y=-0.5568,Z=0.1506,W=0.8088),Translation=(X=-3,Y=0,Z=0),Scale3D=(X=1,Y=1,Z=1))))
    WeightMinimum=0
    Weight=@n13.Result
   )
   > n14 = &RigUnit_ModifyTransforms(
    ItemToModify=((Item=(Type=Bone,Name="base_C_skirt_b_2"),Transform=(Rotation=(X=0,Y=-0.5,Z=0,W=0.866),Translation=(X=-3.5,Y=0,Z=-2.5),Scale3D=(X=1,Y=1,Z=1))))
    WeightMinimum=0
    Weight=&RigVMFunction_MathDoubleMul(B=@n12.Result, A=@n13.Result).Result
   )
   RigVMFunction_ControlFlowBranch = &RigVMFunction_ControlFlowBranch(ExecuteContext=@n14.ExecutePin, Condition=RigVMVariableNode() {variable=pub isRBANOn}.Value)
   n15 = &RigVMFunction_MathDoubleMul(
    A=&RigVMFunction_MathDoubleRemap(SourceMaximum=150, SourceMinimum=140, TargetMaximum=1, TargetMinimum=0, Value=RigVMVariableNode() {variable=r calfLocal rz float}.Value).Result
    B=RigVMVariableNode() {variable=isCrouched alpha}.Value
   )
   &RigUnit_ModifyTransforms(
    ItemToModify="((Item=(Type=Bone,Name=\"base_C_skirt_b_1\"),Transform=(Rotation=(X=-0.2164,Y=0,Z=0,W=0.9763),Translation=(X=-22.6906,Y=-1.3834,Z=-2.9672),Scale3D=(X=1,Y=1,Z=1))),(Item=(Type=Bone,Name=\"base_R_skirt_c_1\"),Transform=(Rotation=(X=0,Y=0,Z=0,W=1),Translation=(X= ...(truncated)"
    WeightMinimum=0
    Weight=@n15.Result
    ExecutePin=@RigVMFunction_ControlFlowBranch.False
   )
   &RigUnit_ModifyTransforms(
    ItemToModify="((Item=(Type=Bone,Name=\"base_C_skirt_b_1\"),Transform=(Rotation=(X=0,Y=0.2588,Z=0,W=-0.9659),Translation=(X=-34.2937,Y=0.0005,Z=-0.736),Scale3D=(X=1,Y=1,Z=1))),(Item=(Type=Bone,Name=\"base_R_skirt_c_1\"),Transform=(Rotation=(X=0,Y=-0.1736,Z=0,W=0.9848),Transl ...(truncated)"
    WeightMinimum=0
    Weight=@n15.Result
    ExecutePin=@RigVMFunction_ControlFlowBranch.True
   )
   RigVMFunction_MathBoolNot = &RigVMFunction_MathBoolNot(Value=RigVMVariableNode() {variable=pub isDBNO Carried}.Value)
   Add = &RigVMFunction_MathDoubleAdd(A=RigVMVariableNode() {variable=pub strafeRight}.Value, B=RigVMVariableNode() {variable=pub strafeLeft}.Value)
   n16 = RigVMVariableNode() {variable=pub moveForward}
   n17 = RigVMVariableNode() {variable=pub moveBackward}
   Multiply_5 = RigVMAggregateNode(
    C=&RigVMFunction_MathBoolToFloat(Value=@RigVMFunction_MathBoolNot.Result).Result
    B=&RigVMFunction_MathDoubleRemap(SourceMaximum=1, SourceMinimum=0, TargetMaximum=0, TargetMinimum=1, bClamp=False, Value=@Add.Result).Result
    A=&RigVMFunction_MathDoubleRemap(
     SourceMaximum=1
     SourceMinimum=0
     TargetMaximum=0.6
     TargetMinimum=1
     bClamp=False
     Value=&RigVMFunction_MathDoubleAdd(A=@n16.Value, B=@n17.Value).Result
    ).Result
   )
   AlphaInterp = &RigVMFunction_AlphaInterp(InterpSpeedDecreasing=10, InterpSpeedIncreasing=10, Scale=1, bClampResult=False, Value=@Multiply_5.Result)
   &RigUnit_ModifyTransforms(
    ItemToModify="((Item=(Type=Bone,Name=\"base_L_skirt_c_2\"),Transform=(Rotation=(X=0.2648,Y=0.4759,Z=0.094,W=-0.8334),Translation=(X=-4.2771,Y=-0.9038,Z=-12.8594),Scale3D=(X=1,Y=1,Z=1))),(Item=(Type=Bone,Name=\"base_C_skirt_b_2\"),Transform=(Rotation=(X=-0.1637,Y=0.5583,Z=0. ...(truncated)"
    WeightMinimum=0
    Weight=@AlphaInterp.Result
    ExecutePin=@n9.True
   )
   RigVMFunction_Sequence_2 = RigVMAggregateNode(ExecuteContext=@RigVMFunction_Sequence_1.C)
   n18 = &RigVMFunction_ControlFlowBranch(
    Condition=&RigVMFunction_MathBoolOr(A=RigVMVariableNode() {variable=pub isCrouching}.Value, B=RigVMVariableNode() {variable=pub isPlayingEasternBloc}.Value).Result
    ExecuteContext=@RigVMFunction_Sequence_2.A
   )
   RigVMVariableNode(ExecuteContext=@n18.False) {variable=isCrouched alpha}
   n19 = &RigUnit_GetRelativeTransformForItem(Child=(Type=Bone,Name="pelvis"), Parent=(Type=Bone,Name="root"), bChildInitial=False, bParentInitial=False)
   n20 = &RigVMFunction_AnimEvalRichCurve(
    Curve="(EditorCurveData=(Keys=((InterpMode=RCIM_Linear,TangentMode=RCTM_Auto,TangentWeightMode=RCTWM_WeightedNone,Time=0,Value=0,ArriveTangent=0,ArriveTangentWeight=0,LeaveTangent=0,LeaveTangentWeight=0),(InterpMode=RCIM_Cubic,TangentMode=RCTM_Auto,TangentWeightM ...(truncated)"
    SourceMaximum=45
    SourceMinimum=90
    TargetMinimum=0
    Value=@n19.RelativeTransform.Translation.Z
   )
   n21 = &RigVMFunction_MathDoubleMul(
    A=&RigVMFunction_MathDoubleRemap(
     SourceMaximum=1
     SourceMinimum=0
     TargetMaximum=0
     TargetMinimum=1
     bClamp=False
     Value=RigVMVariableNode() {variable=isBaseballSliding Lerp}.Value
    ).Result
    B=&RigVMFunction_MathDoubleRemap(
     SourceMaximum=1
     SourceMinimum=0
     TargetMaximum=0
     TargetMinimum=1
     bClamp=False
     Value=RigVMVariableNode() {variable=isKneeSliding Lerp}.Value
    ).Result
   )
   n22 = &RigVMFunction_MathDoubleMul(A=@n20.Result, B=@n21.Result)
   &RigUnit_ModifyTransforms(
    ItemToModify="((Item=(Type=Bone,Name=\"base_R_skirt_c_1\"),Transform=(Rotation=(X=0.0268,Y=-0.341,Z=0.0735,W=0.9368),Translation=(X=-3.6221,Y=-0.2475,Z=-5.882),Scale3D=(X=1,Y=1,Z=1))),(Item=(Type=Bone,Name=\"base_R_skirt_c_2\"),Transform=(Rotation=(X=0.2034,Y=0.6748,Z=-0.20 ...(truncated)"
    WeightMinimum=0
    Weight=@n22.Result
    ExecutePin=@n18.True
   )
   > &RigUnit_ModifyTransforms(
    ItemToModify="((Item=(Type=Bone,Name=\"base_L_skirt_b_1\"),Transform=(Rotation=(X=0.0062,Y=-0.0421,Z=0.057,W=-0.9975),Translation=(X=0,Y=0,Z=0),Scale3D=(X=1,Y=1,Z=1))),(Item=(Type=Bone,Name=\"base_L_skirt_b_2\"),Transform=(Rotation=(X=-0.0118,Y=0.2194,Z=-0.0604,W=0.9737),Tr ...(truncated)"
    WeightMinimum=0
    Weight=@n22.Result
   )
   > RigVMVariableNode(Value=@n22.Result) {variable=isCrouched alpha}
   > &RigUnit_ModifyTransforms(
    ItemToModify=((Item=(Type=Bone,Name="base_R_skirt_c_2"),Transform=(Rotation=(X=0,Y=0,Z=0,W=-1),Translation=(X=-1.9338,Y=-2.9433,Z=4.9456),Scale3D=(X=1,Y=1,Z=1))))
    WeightMinimum=0
    Weight=&RigVMFunction_MathDoubleMul(
     A=&RigVMFunction_MathDoubleRemap(SourceMaximum=44, SourceMinimum=39, TargetMaximum=0, TargetMinimum=1, Value=@n19.RelativeTransform.Translation.Z).Result
     B=@n21.Result
    ).Result
   )
   > n23 = &RigUnit_SphericalPoseReader(
    ActiveRegionScaleFactors=(PositiveWidth=1,NegativeWidth=1,PositiveHeight=1,NegativeHeight=1)
    ActiveRegionSize=0.1
    Debug=(bDrawDebug=false,bDraw2D=False,bDrawLocalAxes=False,DebugScale=25,DebugSegments=20,DebugThickness=0.25)
    DriverAxis=(X=1,Y=0,Z=0)
    DriverItem=(Type=Bone,Name="thigh_r")
    FalloffRegionScaleFactors=(PositiveWidth=0.6,NegativeWidth=1,PositiveHeight=0.7,NegativeHeight=1)
    FalloffSize=0.3
    FlipHeightScaling=False
    FlipWidthScaling=False
    OptionalParentItem=(Type=Bone,Name="None")
    RotationOffset=(X=-65,Y=-170,Z=0)
   )
   n24 = &RigVMFunction_AnimRichCurve(Curve="(EditorCurveData=(Keys=((InterpMode=RCIM_Linear,TangentMode=RCTM_Auto,TangentWeightMode=RCTWM_WeightedNone,Time=0,Value=0,ArriveTangent=0,ArriveTangentWeight=0,LeaveTangent=0,LeaveTangentWeight=0),(InterpMode=RCIM_Cubic,TangentMode=RCTM_Auto,TangentWeightM ...(truncated)")
   > RigVMVariableNode(Value=&RigVMFunction_AnimEvalRichCurve(SourceMinimum=0, TargetMinimum=0, Curve=@n24.Curve, Value=@n23.OutputParam).Result) {variable=r legSPR CrouchingStrafe float}
   > n25 = &RigUnit_SphericalPoseReader(
    ActiveRegionScaleFactors=(PositiveWidth=1,NegativeWidth=1,PositiveHeight=1,NegativeHeight=1)
    ActiveRegionSize=0.1
    Debug=(bDrawDebug=false,bDraw2D=False,bDrawLocalAxes=False,DebugScale=25,DebugSegments=20,DebugThickness=0.25)
    DriverAxis=(X=1,Y=0,Z=0)
    DriverItem=(Type=Bone,Name="thigh_l")
    FalloffRegionScaleFactors=(PositiveWidth=0.6,NegativeWidth=1,PositiveHeight=0.7,NegativeHeight=1)
    FalloffSize=0.3
    FlipHeightScaling=False
    FlipWidthScaling=False
    OptionalParentItem=(Type=Bone,Name="None")
    RotationOffset=(X=-65,Y=-170,Z=0)
   )
   > RigVMVariableNode(Value=&RigVMFunction_AnimEvalRichCurve(SourceMinimum=0, TargetMinimum=0, Value=@n25.OutputParam, Curve=@n24.Curve).Result) {variable=l legSPR CrouchingStrafe float}
   Multiply_2_2 = RigVMAggregateNode(
    B=&RigVMFunction_MathDoubleRemap(SourceMaximum=1, SourceMinimum=0.6, TargetMaximum=1, TargetMinimum=0, Value=RigVMVariableNode() {variable=pub strafeLeft}.Value).Result
    C=RigVMVariableNode() {variable=isCrouching Lerp}.Value
    A=RigVMVariableNode() {variable=r legSPR CrouchingStrafe float}.Value
   )
   > &RigUnit_ModifyTransforms(
    ItemToModify="((Item=(Type=Bone,Name=\"base_C_skirt_b_1\"),Transform=(Rotation=(X=0.21,Y=-0.1272,Z=-0.0113,W=0.9693),Translation=(X=-2.45,Y=0.2,Z=-0.6),Scale3D=(X=1,Y=1,Z=1))),(Item=(Type=Bone,Name=\"base_C_skirt_b_2\"),Transform=(Rotation=(X=-0.0011,Y=-0.1306,Z=0.0237,W=0. ...(truncated)"
    WeightMinimum=0
    Weight=@Multiply_2_2.Result
   )
   > &RigUnit_ModifyTransforms(
    ItemToModify="((Item=(Type=Bone,Name=\"base_R_skirt_c_1\"),Transform=(Rotation=(X=0.2042,Y=-0.2087,Z=-0.3856,W=0.8753),Translation=(X=-3.5,Y=2.65,Z=-0.5),Scale3D=(X=1,Y=1,Z=1))),(Item=(Type=Bone,Name=\"base_R_skirt_b_1\"),Transform=(Rotation=(X=0.1964,Y=-0.2401,Z=-0.0821,W= ...(truncated)"
    WeightMinimum=0
    Weight=@Multiply_2_2.Result
   )
   > &RigUnit_ModifyTransforms(
    ItemToModify="((Item=(Type=Bone,Name=\"base_L_skirt_c_1\"),Transform=(Rotation=(X=0,Y=0.0131,Z=0,W=0.9999),Translation=(X=0,Y=0,Z=0),Scale3D=(X=1,Y=1,Z=1))),(Item=(Type=Bone,Name=\"base_L_skirt_b_1\"),Transform=(Rotation=(X=0,Y=-0.0175,Z=0,W=0.9998),Translation=(X=0,Y=0,Z=0 ...(truncated)"
    WeightMinimum=0
    Weight=@Multiply_2_2.Result
   )
   Multiply_2_2_2_1 = RigVMAggregateNode(
    B=&RigVMFunction_MathDoubleRemap(SourceMaximum=1, SourceMinimum=0.6, TargetMaximum=1, TargetMinimum=0, Value=RigVMVariableNode() {variable=pub strafeLeft}.Value).Result
    C=RigVMVariableNode() {variable=isCrouching Lerp}.Value
    A=RigVMVariableNode() {variable=l legSPR CrouchingStrafe float}.Value
   )
   > &RigUnit_ModifyTransforms(
    ItemToModify="((Item=(Type=Bone,Name=\"base_C_skirt_b_1\"),Transform=(Rotation=(X=-0.0352,Y=-0.1756,Z=0.0453,W=0.9828),Translation=(X=0,Y=0,Z=0),Scale3D=(X=1,Y=1,Z=1))),(Item=(Type=Bone,Name=\"base_L_skirt_c_1\"),Transform=(Rotation=(X=-0.2428,Y=-0.0102,Z=0.1409,W=0.9597),T ...(truncated)"
    WeightMinimum=0
    Weight=@Multiply_2_2_2_1.Result
   )
   Multiply_2_2_2 = RigVMAggregateNode(
    B=&RigVMFunction_MathDoubleRemap(SourceMaximum=1, SourceMinimum=0.6, TargetMaximum=1, TargetMinimum=0, Value=RigVMVariableNode() {variable=pub strafeRight}.Value).Result
    C=RigVMVariableNode() {variable=isCrouching Lerp}.Value
    A=RigVMVariableNode() {variable=l legSPR CrouchingStrafe float}.Value
   )
   > &RigUnit_ModifyTransforms(
    ItemToModify="((Item=(Type=Bone,Name=\"base_L_skirt_c_1\"),Transform=(Rotation=(X=-0.0219,Y=-0.1003,Z=0.1256,W=0.9867),Translation=(X=0,Y=0,Z=0),Scale3D=(X=1,Y=1,Z=1))),(Item=(Type=Bone,Name=\"base_L_skirt_b_1\"),Transform=(Rotation=(X=-0.0008,Y=-0.0798,Z=-0.01,W=0.9968),Tr ...(truncated)"
    WeightMinimum=0
    Weight=@Multiply_2_2_2.Result
   )
   > n26 = &RigUnit_SphericalPoseReader(
    ActiveRegionScaleFactors=(PositiveWidth=1,NegativeWidth=1,PositiveHeight=1,NegativeHeight=1)
    ActiveRegionSize=0.1
    Debug=(bDrawDebug=false,bDraw2D=False,bDrawLocalAxes=False,DebugScale=25,DebugSegments=20,DebugThickness=0.1)
    DriverAxis=(X=1,Y=0,Z=0)
    DriverItem=(Type=Bone,Name="thigh_r")
    FalloffRegionScaleFactors=(PositiveWidth=1,NegativeWidth=1,PositiveHeight=1,NegativeHeight=1)
    FalloffSize=0.2
    FlipHeightScaling=False
    FlipWidthScaling=False
    OptionalParentItem=(Type=Bone,Name="None")
    RotationOffset=(X=-91,Y=0,Z=15)
   )
   n27 = &RigVMFunction_MathDoubleMul(A=@n26.OutputParam, B=RigVMVariableNode() {variable=isCrouching Lerp}.Value)
   > &RigUnit_ModifyTransforms(
    ItemToModify=((Item=(Type=Bone,Name="base_C_skirt_a_1"),Transform=(Rotation=(X=0,Y=0,Z=-0.1158,W=0.9933),Translation=(X=0,Y=0,Z=0),Scale3D=(X=1,Y=1,Z=1))))
    WeightMinimum=0
    Weight=@n27.Result
   )
   > &RigUnit_ModifyTransforms(
    ItemToModify="((Item=(Type=Bone,Name=\"base_C_skirt_b_1\"),Transform=(Rotation=(X=0,Y=0,Z=0.1136,W=-0.9935),Translation=(X=-0.2022,Y=-0.8727,Z=4.848),Scale3D=(X=1,Y=1,Z=1))),(Item=(Type=Bone,Name=\"base_R_skirt_c_1\"),Transform=(Rotation=(X=-0.0155,Y=-0.1016,Z=-0.1501,W=0.9 ...(truncated)"
    WeightMinimum=0
    Weight=@n27.Result
   )
   &RigUnit_ModifyTransforms(
    ItemToModify="((Item=(Type=Bone,Name=\"base_L_skirt_c_1\"),Transform=(Rotation=(X=0.0829,Y=0.0412,Z=-0.3127,W=-0.9453),Translation=(X=-4.7654,Y=0.5865,Z=-6.4388),Scale3D=(X=1,Y=1,Z=1))),(Item=(Type=Bone,Name=\"base_R_skirt_c_1\"),Transform=(Rotation=(X=0,Y=0,Z=0,W=1),Transl ...(truncated)"
    WeightMinimum=0
    Weight=RigVMVariableNode() {variable=isKneeSliding Lerp}.Value
    ExecutePin=@RigVMFunction_Sequence_2.B
   )
   > &RigUnit_ModifyTransforms(
    ItemToModify="((Item=(Type=Bone,Name=\"base_C_skirt_a_1\"),Transform=(Rotation=(X=0,Y=0,Z=0.2588,W=0.9659),Translation=(X=0,Y=0,Z=0),Scale3D=(X=1,Y=1,Z=1))),(Item=(Type=Bone,Name=\"base_L_skirt_a_1\"),Transform=(Rotation=(X=0,Y=0.0436,Z=0,W=0.999),Translation=(X=0,Y=0,Z=0), ...(truncated)"
    WeightMinimum=0
    Weight=RigVMVariableNode() {variable=isBaseballSliding Lerp}.Value
   )
   > DeltaRotation v4 = RigVMFunctionReferenceNode(
    Rotation Order=ZYX
    child=(Type=Bone,Name="calf_l")
    outputMinMax=(X=0,Y=1)
    parent=(Type=Null,Name="calf_l_null")
    premult=1
    rotationMinMax=(X=80,Y=128)
    triggerAxis=Z
    twistAxis=(X=1,Y=0,Z=0)
   )
   n28 = &RigVMFunction_AnimEvalRichCurve(
    Curve="(EditorCurveData=(Keys=((InterpMode=RCIM_Cubic),(InterpMode=RCIM_Cubic,TangentMode=RCTM_User,Time=0.6264,Value=0.3068,ArriveTangent=1.2597,LeaveTangent=1.2597),(InterpMode=RCIM_Cubic,TangentMode=RCTM_User,Time=1,Value=1,ArriveTangent=2.7015,LeaveTangent=2. ...(truncated)"
    SourceMaximum=38
    SourceMinimum=55
    TargetMinimum=0
    Value=&RigUnit_GetRelativeTransformForItem(Child=(Type=Bone,Name="pelvis"), Parent=(Type=Bone,Name="root"), bChildInitial=False, bParentInitial=False).RelativeTransform.Translation.Z
   )
   Multiply_3 = RigVMAggregateNode(A=@"DeltaRotation v4".outputWeight, C=@n28.Result, B=RigVMVariableNode() {variable=isGrindRail Lerp}.Value)
   > &RigUnit_ModifyTransforms(
    ItemToModify="((Item=(Type=Bone,Name=\"base_L_skirt_c_2\"),Transform=(Rotation=(X=0,Y=-0.3583,Z=0,W=-0.9336),Translation=(X=17.3309,Y=0,Z=4.6017),Scale3D=(X=1,Y=1,Z=1))),(Item=(Type=Bone,Name=\"base_L_skirt_c_1\"),Transform=(Rotation=(X=0,Y=0,Z=0,W=-1),Translation=(X=-7.508 ...(truncated)"
    WeightMinimum=0
    Weight=@Multiply_3.Result
   )
   > &RigUnit_ModifyTransforms(
    ItemToModify="((Item=(Type=Bone,Name=\"base_L_skirt_b_1\"),Transform=(Rotation=(X=0,Y=0.1736,Z=0,W=0.9848),Translation=(X=0,Y=0,Z=2),Scale3D=(X=1,Y=1,Z=1))),(Item=(Type=Bone,Name=\"base_L_skirt_b_2\"),Transform=(Rotation=(X=0,Y=0.1736,Z=0,W=0.9848),Translation=(X=-1,Y=0,Z=0 ...(truncated)"
    WeightMinimum=0
    Weight=RigVMVariableNode() {variable=isGrindRail Lerp}.Value
   )
   > n29 = &RigUnit_SphericalPoseReader(
    ActiveRegionScaleFactors=(PositiveWidth=1,NegativeWidth=1,PositiveHeight=1,NegativeHeight=1)
    ActiveRegionSize=0.08
    Debug=(bDrawDebug=false,bDraw2D=False,bDrawLocalAxes=False,DebugScale=25,DebugSegments=20,DebugThickness=0.25)
    DriverAxis=(X=1,Y=0,Z=0)
    DriverItem=(Type=Bone,Name="thigh_r")
    FalloffRegionScaleFactors=(PositiveWidth=1,NegativeWidth=1,PositiveHeight=1,NegativeHeight=1)
    FalloffSize=0.22
    FlipHeightScaling=False
    FlipWidthScaling=False
    OptionalParentItem=(Type=Bone,Name="None")
    RotationOffset=(X=-52,Y=-45,Z=0)
   )
   Cast = &RigVMFunction_MathBoolToFloat(Value=RigVMVariableNode() {variable=pub isPlayingEmote}.Value)
   > &RigUnit_ModifyTransforms(
    ItemToModify=((Item=(Type=Bone,Name="base_R_skirt_a_2"),Transform=(Rotation=(X=0,Y=-0.0054,Z=0,W=1),Translation=(X=0.25,Y=0,Z=2.4),Scale3D=(X=1,Y=1,Z=1))))
    WeightMinimum=0
    Weight=&RigVMFunction_MathDoubleMul(B=@Cast.Result, A=@n29.OutputParam).Result
   )
   /Script/ControlRig.RigUnit_SetCurveValue(Curve=crouchingAlpha, Value=RigVMVariableNode() {variable=isCrouched alpha}.Value, ExecutePin=@RigVMFunction_Sequence_2.C) {is=struct RigVMUnitNode}
   > &RigUnit_SetCurveValue(Curve=r_thighLocal_ry_float, Value=RigVMVariableNode() {variable=r thighLocal ry float}.Value)
   > &RigUnit_SetCurveValue(Curve=r_thighLocal_rz_float, Value=RigVMVariableNode() {variable=r thighLocal rz float}.Value)
   Remap = &RigVMFunction_MathDoubleRemap(SourceMaximum=-100, SourceMinimum=0, TargetMaximum=1, TargetMinimum=0, Value=RigVMVariableNode() {variable=l thighLocal rz float}.Value)
   &RigUnit_ModifyTransforms(
    ItemToModify="((Item=(Type=Bone,Name=\"base_L_skirt_a_1\"),Transform=(Rotation=(X=0,Y=-0.0175,Z=0,W=0.9998),Translation=(X=0,Y=0,Z=0),Scale3D=(X=1,Y=1,Z=1))),(Item=(Type=Bone,Name=\"base_L_skirt_b_1\"),Transform=(Rotation=(X=-0.0027,Y=-0.0523,Z=-0.0523,W=0.9973),Translation ...(truncated)"
    WeightMinimum=0
    Weight=@Remap.Result
    ExecutePin=@RigVMFunction_Sequence_1.A
   )
   n30 = &RigVMFunction_MathDoubleRemap(SourceMaximum=-100, SourceMinimum=0, TargetMaximum=1, TargetMinimum=0, Value=RigVMVariableNode() {variable=r thighLocal rz float}.Value)
   > &RigUnit_ModifyTransforms(
    ItemToModify="((Item=(Type=Bone,Name=\"base_R_skirt_a_1\"),Transform=(Rotation=(X=0,Y=-0.0175,Z=0,W=0.9998),Translation=(X=0,Y=0,Z=0),Scale3D=(X=1,Y=1,Z=1))),(Item=(Type=Bone,Name=\"base_R_skirt_b_1\"),Transform=(Rotation=(X=-0.0032,Y=-0.061,Z=-0.0522,W=0.9968),Translation= ...(truncated)"
    WeightMinimum=0
    Weight=@n30.Result
   )
   Multiply = &RigVMFunction_MathDoubleMul(B=@Remap.Result, A=@n30.Result)
   n31 = &RigVMFunction_AnimEvalRichCurve(
    Curve="(EditorCurveData=(Keys=((InterpMode=RCIM_Cubic,TangentMode=RCTM_User,TangentWeightMode=RCTWM_WeightedNone,Time=0,Value=0,ArriveTangent=1.6904,ArriveTangentWeight=0,LeaveTangent=1.6904,LeaveTangentWeight=0),(InterpMode=RCIM_Cubic,TangentMode=RCTM_User,Tange ...(truncated)"
    SourceMinimum=0
    TargetMinimum=0
    Value=@Multiply.Result
   )
   > &RigUnit_ModifyTransforms(
    ItemToModify="((Item=(Type=Bone,Name=\"base_C_skirt_a_1\"),Transform=(Rotation=(X=0,Y=0.6428,Z=0,W=0.766),Translation=(X=0,Y=0,Z=0),Scale3D=(X=1,Y=1,Z=1))),(Item=(Type=Bone,Name=\"base_C_skirt_b_1\"),Transform=(Rotation=(X=0,Y=0.342,Z=0,W=0.9397),Translation=(X=0,Y=0,Z=0),S ...(truncated)"
    WeightMinimum=0
    Weight=@n31.Result
   )
   n32 = &RigVMFunction_MathDoubleRemap(SourceMaximum=50, SourceMinimum=0, TargetMaximum=1, TargetMinimum=0, Value=RigVMVariableNode() {variable=l thighLocal rz float}.Value)
   > &RigUnit_ModifyTransforms(
    ItemToModify="((Item=(Type=Bone,Name=\"base_L_skirt_a_1\"),Transform=(Rotation=(X=0,Y=-0.1736,Z=0,W=0.9848),Translation=(X=0,Y=0,Z=0),Scale3D=(X=1,Y=1,Z=1))),(Item=(Type=Bone,Name=\"base_L_skirt_c_1\"),Transform=(Rotation=(X=0,Y=-0.0523,Z=0,W=0.9986),Translation=(X=-0.5,Y=0 ...(truncated)"
    WeightMinimum=0
    Weight=@n32.Result
   )
   n33 = &RigVMFunction_MathDoubleRemap(SourceMaximum=50, SourceMinimum=0, TargetMaximum=1, TargetMinimum=0, Value=RigVMVariableNode() {variable=r thighLocal rz float}.Value)
   > &RigUnit_ModifyTransforms(
    ItemToModify="((Item=(Type=Bone,Name=\"base_R_skirt_a_1\"),Transform=(Rotation=(X=0,Y=-0.1736,Z=0,W=0.9848),Translation=(X=0,Y=0,Z=0),Scale3D=(X=1,Y=1,Z=1))),(Item=(Type=Bone,Name=\"base_R_skirt_c_1\"),Transform=(Rotation=(X=0,Y=-0.0523,Z=0,W=0.9986),Translation=(X=-0.5,Y=0 ...(truncated)"
    WeightMinimum=0
    Weight=@n33.Result
   )
   > &RigUnit_ModifyTransforms(
    ItemToModify="((Item=(Type=Bone,Name=\"base_C_skirt_a_1\"),Transform=(Rotation=(X=0,Y=0.4226,Z=0,W=0.9063),Translation=(X=0,Y=0,Z=-2),Scale3D=(X=1,Y=1,Z=1))),(Item=(Type=Bone,Name=\"base_C_skirt_b_1\"),Transform=(Rotation=(X=0,Y=0.2164,Z=0,W=0.9763),Translation=(X=-0.8,Y=0, ...(truncated)"
    WeightMinimum=0
    Weight=&RigVMFunction_MathDoubleMul(B=@n32.Result, A=@n33.Result).Result
   )
   n34 = RigVMVariableNode() {variable=l thighLocal rz float}
   n35 = RigVMVariableNode() {variable=r thighLocal rz float}
   > &RigUnit_ModifyTransforms(
    ItemToModify="((Item=(Type=Bone,Name=\"base_C_skirt_a_1\"),Transform=(Rotation=(X=0,Y=-0.0436,Z=0,W=0.999),Translation=(X=0,Y=0,Z=0),Scale3D=(X=1,Y=1,Z=1))),(Item=(Type=Bone,Name=\"base_C_skirt_b_1\"),Transform=(Rotation=(X=0,Y=-0.0872,Z=0,W=0.9962),Translation=(X=0,Y=-4,Z= ...(truncated)"
    WeightMinimum=0
    Weight=&RigVMFunction_MathDoubleMul(
     A=&RigVMFunction_MathDoubleRemap(SourceMaximum=45, SourceMinimum=0, TargetMaximum=1, TargetMinimum=0, Value=@n34.Value).Result
     B=&RigVMFunction_MathDoubleRemap(SourceMaximum=-45, SourceMinimum=0, TargetMaximum=1, TargetMinimum=0, Value=@n35.Value).Result
    ).Result
   )
   > &RigUnit_ModifyTransforms(
    ItemToModify="((Item=(Type=Bone,Name=\"base_C_skirt_a_1\"),Transform=(Rotation=(X=0,Y=-0.0436,Z=0,W=0.999),Translation=(X=0,Y=0,Z=0),Scale3D=(X=1,Y=1,Z=1))),(Item=(Type=Bone,Name=\"base_C_skirt_b_1\"),Transform=(Rotation=(X=0,Y=-0.0872,Z=0,W=0.9962),Translation=(X=0,Y=4,Z=4 ...(truncated)"
    WeightMinimum=0
    Weight=&RigVMFunction_MathDoubleMul(
     B=&RigVMFunction_MathDoubleRemap(SourceMaximum=45, SourceMinimum=0, TargetMaximum=1, TargetMinimum=0, Value=@n35.Value).Result
     A=&RigVMFunction_MathDoubleRemap(SourceMaximum=-45, SourceMinimum=0, TargetMaximum=1, TargetMinimum=0, Value=@n34.Value).Result
    ).Result
   )
   n36 = RigVMVariableNode() {variable=l thighLocal ry float}
   n37 = RigVMVariableNode() {variable=l thighLocal rz float}
   > &RigUnit_ModifyTransforms(
    ItemToModify="((Item=(Type=Bone,Name=\"base_C_skirt_a_1\"),Transform=(Rotation=(X=0,Y=0,Z=0.1045,W=0.9945),Translation=(X=0,Y=0,Z=0),Scale3D=(X=1,Y=1,Z=1))),(Item=(Type=Bone,Name=\"base_C_skirt_b_1\"),Transform=(Rotation=(X=0,Y=0,Z=0.1305,W=0.9914),Translation=(X=0,Y=0,Z=0) ...(truncated)"
    WeightMinimum=0
    Weight=&RigVMFunction_MathDoubleMul(
     A=&RigVMFunction_MathDoubleRemap(SourceMaximum=24, SourceMinimum=0, TargetMaximum=1, TargetMinimum=0, Value=@n36.Value).Result
     B=&RigVMFunction_MathDoubleRemap(SourceMaximum=-70, SourceMinimum=0, TargetMaximum=1, TargetMinimum=0, Value=@n37.Value).Result
    ).Result
   )
   n38 = RigVMVariableNode() {variable=r thighLocal ry float}
   n39 = RigVMVariableNode() {variable=r thighLocal rz float}
   > &RigUnit_ModifyTransforms(
    ItemToModify="((Item=(Type=Bone,Name=\"base_C_skirt_a_1\"),Transform=(Rotation=(X=0,Y=0,Z=-0.1045,W=0.9945),Translation=(X=0,Y=0,Z=0),Scale3D=(X=1,Y=1,Z=1))),(Item=(Type=Bone,Name=\"base_C_skirt_b_1\"),Transform=(Rotation=(X=0,Y=0,Z=-0.1305,W=0.9914),Translation=(X=0,Y=0,Z= ...(truncated)"
    WeightMinimum=0
    Weight=&RigVMFunction_MathDoubleMul(
     A=&RigVMFunction_MathDoubleRemap(SourceMaximum=24, SourceMinimum=0, TargetMaximum=1, TargetMinimum=0, Value=@n38.Value).Result
     B=&RigVMFunction_MathDoubleRemap(SourceMaximum=-70, SourceMinimum=0, TargetMaximum=1, TargetMinimum=0, Value=@n39.Value).Result
    ).Result
   )
   RigVMFunction_MathFloatMake = /Script/RigVM.RigVMFunction_MathFloatMake(Value=0.9) {is=struct RigVMUnitNode}
   > &RigUnit_OffsetTransformForItem(
    Item=(Type=Bone,Name="deform_thigh_twist_01_l")
    OffsetTransform=(Rotation=(X=0,Y=0,Z=0,W=1),Translation=(X=0,Y=0,Z=0),Scale3D=(X=1,Y=0.8,Z=0.8))
    Weight=1
    bPropagateToChildren=false
    OffsetTransform.Scale3D.Y=@RigVMFunction_MathFloatMake.Value
    OffsetTransform.Scale3D.Z=@RigVMFunction_MathFloatMake.Value
   )
   > &RigUnit_OffsetTransformForItem(
    Item=(Type=Bone,Name="deform_thigh_twist_01_r")
    OffsetTransform=(Rotation=(X=0,Y=0,Z=0,W=1),Translation=(X=0,Y=0,Z=0),Scale3D=(X=1,Y=0.8,Z=0.8))
    Weight=1
    bPropagateToChildren=false
    OffsetTransform.Scale3D.Y=@RigVMFunction_MathFloatMake.Value
    OffsetTransform.Scale3D.Z=@RigVMFunction_MathFloatMake.Value
   )
   RigVMFunctionReturnNode(ExecuteContext=@RigVMFunction_Sequence.B)
   &RigUnit_ModifyTransforms(
    ExecutePin=()
    ItemToModify="((Item=(Type=Bone,Name=\"base_L_skirt_b_1\"),Transform=(Rotation=(X=0.0062,Y=-0.0421,Z=0.057,W=-0.9975),Translation=(X=0,Y=0,Z=0),Scale3D=(X=1,Y=1,Z=1))),(Item=(Type=Bone,Name=\"base_L_skirt_b_2\"),Transform=(Rotation=(X=0.0188,Y=0.2187,Z=0.0762,W=0.9726),Tran ...(truncated)"
    Weight=1
    WeightMinimum=0
   )
  ]}
 New Function_ContainedGraph {nodes [
   RigVMFunctionEntryNode()
   > &RigUnit_ParentConstraint(
    AdvancedSettings=(RotationOrderForFilter=XZY)
    Child=(Type=Bone,Name="kin_holsterPistol_l")
    Parents=((Item=(Type=Bone,Name="calf_l"),Weight=0.1003),(Item=(Type=Bone,Name="calf_twist_01_l"),Weight=0.2041),(Item=(Type=Bone,Name="calf_twist_02_l"),Weight=0.6956))
    Weight=1
   )
   > &RigUnit_ParentConstraint(
    AdvancedSettings=(RotationOrderForFilter=XZY)
    Child=(Type=Bone,Name="base_slipperPuff_l")
    Parents=((Item=(Type=Bone,Name="ball_l"),Weight=0.0492),(Item=(Type=Bone,Name="foot_l"),Weight=0.9508))
    Weight=1
   )
   > &RigUnit_ParentConstraint(
    AdvancedSettings=(RotationOrderForFilter=XZY)
    Child=(Type=Bone,Name="base_slipperPuff_r")
    Parents=((Item=(Type=Bone,Name="ball_r"),Weight=0.0492),(Item=(Type=Bone,Name="foot_r"),Weight=0.9508))
    Weight=1
   )
   > DeltaQuatExtraction_1 = RigVMFunctionReferenceNode(child=(Type=Bone,Name="ball_l"), onlyTwist=true, parent=(Type=Null,Name="ball_l_null"), twistAxis=(X=0,Y=0,Z=1))
   Clamp = &RigVMFunction_MathDoubleClamp(
    Maximum=0
    Minimum=-30
    Value=&RigVMFunction_MathQuaternionToEuler(RotationOrder=ZYX, Value=@DeltaQuatExtraction_1.outputQuat).Result.Z
   )
   RigVMFunction_MathQuaternionFromEuler = &RigVMFunction_MathQuaternionFromEuler(Euler=(X=0,Y=0,Z=0), Euler.Z=@Clamp.Result)
   > &RigUnit_OffsetTransformForItem(
    Item=(Type=Bone,Name="pvt_ballSole_l")
    OffsetTransform=(Rotation=(X=0,Y=0,Z=-0.5,W=0.866),Translation=(X=0,Y=0,Z=0),Scale3D=(X=1,Y=1,Z=1))
    Weight=1
    bPropagateToChildren=true
    OffsetTransform.Rotation=@RigVMFunction_MathQuaternionFromEuler.Result
   )
   > DeltaQuatExtraction_2 = RigVMFunctionReferenceNode(child=(Type=Bone,Name="ball_r"), onlyTwist=true, parent=(Type=Null,Name="ball_r_null"), twistAxis=(X=0,Y=0,Z=1))
   n1 = &RigVMFunction_MathQuaternionToEuler(RotationOrder=ZYX, Value=@DeltaQuatExtraction_2.outputQuat)
   > &RigUnit_OffsetTransformForItem(
    Item=(Type=Bone,Name="pvt_ballSole_r")
    OffsetTransform=(Rotation=(X=0,Y=0,Z=-0.5,W=0.866),Translation=(X=0,Y=0,Z=0),Scale3D=(X=1,Y=1,Z=1))
    Weight=1
    bPropagateToChildren=True
    OffsetTransform.Rotation=&RigVMFunction_MathQuaternionFromEuler(Euler=(X=0,Y=0,Z=0), Euler.Z=&RigVMFunction_MathDoubleClamp(Maximum=0, Minimum=-30, Value=@n1.Result.Z).Result).Result
   )
   > &RigUnit_ParentConstraint(
    AdvancedSettings=(RotationOrderForFilter=XZY)
    Child=(Type=Bone,Name="kin_duckNadeHandle_A")
    Parents="((Item=(Type=Bone,Name=\"clavicle_l\"),Weight=0.0084),(Item=(Type=Bone,Name=\"spine_04\"),Weight=0.0113),(Item=(Type=Bone,Name=\"pelvis\"),Weight=0.0739),(Item=(Type=Bone,Name=\"spine_03\"),Weight=0.1301),(Item=(Type=Bone,Name=\"spine_01\"),Weight=0.3245),(Item=(Typ ...(truncated)"
    Weight=1
   )
   > &RigUnit_ParentConstraint(
    AdvancedSettings=(RotationOrderForFilter=XZY)
    Child=(Type=Bone,Name="kin_duckNadeHandle_B")
    Parents="((Item=(Type=Bone,Name=\"clavicle_l\"),Weight=0.0068),(Item=(Type=Bone,Name=\"spine_04\"),Weight=0.008),(Item=(Type=Bone,Name=\"pelvis\"),Weight=0.1032),(Item=(Type=Bone,Name=\"spine_03\"),Weight=0.1181),(Item=(Type=Bone,Name=\"spine_01\"),Weight=0.3491),(Item=(Type ...(truncated)"
    Weight=1
   )
   > DeltaQuatExtraction = RigVMFunctionReferenceNode(child=(Type=Bone,Name="spine_02"), parent=(Type=Null,Name="spine02_null"), twistAxis=(X=1,Y=0,Z=0))
   RigVMFunction_MathQuaternionToEuler = &RigVMFunction_MathQuaternionToEuler(RotationOrder=ZYX, Value=@DeltaQuatExtraction.outputQuat)
   > &RigUnit_ModifyTransforms(
    ItemToModify="((Item=(Type=Bone,Name=\"base_duckNade_B\"),Transform=(Rotation=(X=0,Y=0.0958,Z=0,W=0.9954),Translation=(X=0,Y=0,Z=0),Scale3D=(X=1,Y=1,Z=1))),(Item=(Type=Bone,Name=\"base_duckNade_A\"),Transform=(Rotation=(X=0,Y=0.0958,Z=0,W=0.9954),Translation=(X=0,Y=0,Z=0),S ...(truncated)"
    WeightMinimum=0
    Weight=&RigVMFunction_MathDoubleRemap(SourceMaximum=30, SourceMinimum=0, TargetMaximum=1, TargetMinimum=0, Value=@RigVMFunction_MathQuaternionToEuler.Result.Z).Result
   )
   Remap = &RigVMFunction_MathDoubleRemap(SourceMaximum=-25, SourceMinimum=0, TargetMaximum=1, TargetMinimum=0, Value=@RigVMFunction_MathQuaternionToEuler.Result.Z)
   > &RigUnit_ModifyTransforms(
    ItemToModify="((Item=(Type=Bone,Name=\"base_duckNade_B\"),Transform=(Rotation=(X=0,Y=-0.1736,Z=0,W=0.9848),Translation=(X=0,Y=0,Z=0),Scale3D=(X=1,Y=1,Z=1))),(Item=(Type=Bone,Name=\"base_duckNade_A\"),Transform=(Rotation=(X=0,Y=-0.1736,Z=0,W=0.9848),Translation=(X=0,Y=0,Z=0) ...(truncated)"
    WeightMinimum=0
    Weight=@Remap.Result
   )
   > &RigUnit_ParentConstraint(
    AdvancedSettings=(RotationOrderForFilter=XZY)
    Child=(Type=Bone,Name="kin_bandAids_knee_r")
    Filter=(TranslationFilter=(bX=True,bY=True,bZ=True),RotationFilter=(bX=false,bY=false,bZ=false),ScaleFilter=(bX=false,bY=false,bZ=false))
    Parents="((Item=(Type=Bone,Name=\"deform_knee_in_r\"),Weight=0.0004),(Item=(Type=Bone,Name=\"calf_twist_01_r\"),Weight=0.0017),(Item=(Type=Bone,Name=\"thigh_r\"),Weight=0.073),(Item=(Type=Bone,Name=\"calf_twist_02_r\"),Weight=0.1435),(Item=(Type=Bone,Name=\"deform_knee_out_ ...(truncated)"
    Weight=1
   )
   > DeltaRotation v4 = RigVMFunctionReferenceNode(
    Rotation Order=ZYX
    child=(Type=Bone,Name="calf_r")
    outputMinMax=(X=0,Y=1)
    parent=(Type=Null,Name="calf_r_null")
    premult=1
    rotationMinMax=(X=50,Y=150)
    triggerAxis=Z
    twistAxis=(X=1,Y=0,Z=0)
   )
   > &RigUnit_ModifyTransforms(
    ItemToModify=((Item=(Type=Bone,Name="kin_bandAids_knee_r"),Transform=(Rotation=(X=-0.0871,Y=0.0435,Z=0.0038,W=0.9952),Translation=(X=0,Y=0,Z=1),Scale3D=(X=1,Y=1,Z=1))))
    WeightMinimum=0
    Weight=@"DeltaRotation v4".outputWeight
   )
   > RigVMFunctionReturnNode()
  ]}
 New Function_ContainedGraph {nodes [
   n1 = RigVMFunctionEntryNode()
   RigVMFunction_MathBoolToInteger = &RigVMFunction_MathBoolToInteger(Value=@n1.onlyTwist)
   RigVMFunction_MathQuaternionSwingTwist = &RigVMFunction_MathQuaternionSwingTwist(
    TwistAxis=@n1.twistAxis
    Input=&RigUnit_GetRelativeTransformForItem(bChildInitial=False, bParentInitial=False, Child=@n1.child, Parent=@n1.parent).RelativeTransform.Rotation
   )
   n2 = RigVMDispatchNode(
    Values=((),())
    Index=@RigVMFunction_MathBoolToInteger.Result
    Values.0=@RigVMFunction_MathQuaternionSwingTwist.Swing
    Values.1=@RigVMFunction_MathQuaternionSwingTwist.Twist
   ) {template="DISPATCH_RigVMDispatch_SelectInt32(in Index,in Values,out Result)"}
   Inverse = /Script/RigVM.RigVMFunction_MathQuaternionInverse(Value=@n2.Result) {is=struct RigVMUnitNode, template="Inverse::Execute(in Value,out Result)"}
   > RigVMFunctionReturnNode(outputQuat=RigVMDispatchNode(True=@Inverse.Result, Condition=@n1.invert, False=@n2.Result) {template="DISPATCH_RigVMDispatch_If(in Condition,in True,in False,out Result)"}.Result)
  ]}
 New Function_ContainedGraph {nodes [
   RigVMFunctionEntryNode()
   > &RigUnit_ParentConstraint(
    AdvancedSettings=(RotationOrderForFilter=XZY)
    Child=(Type=Bone,Name="kin_beltRope")
    Parents="((Item=(Type=Bone,Name=\"thigh_twist_01_l\"),Weight=0.0003),(Item=(Type=Bone,Name=\"spine_04\"),Weight=0.0154),(Item=(Type=Bone,Name=\"pelvis\"),Weight=0.0757),(Item=(Type=Bone,Name=\"spine_03\"),Weight=0.1141),(Item=(Type=Bone,Name=\"spine_01\"),Weight=0.3592),(Ite ...(truncated)"
    Weight=1
   )
   RigVMFunction_AnimEvalRichCurve = &RigVMFunction_AnimEvalRichCurve(
    Curve="(EditorCurveData=(Keys=((InterpMode=RCIM_Cubic),(InterpMode=RCIM_Cubic,Time=0.3985,Value=0.2244,ArriveTangent=0.8024,LeaveTangent=0.8024),(InterpMode=RCIM_Cubic,TangentMode=RCTM_User,Time=0.7367,Value=0.5911,ArriveTangent=1.4878,LeaveTangent=1.4878),(Time= ...(truncated)"
    SourceMinimum=0
    TargetMinimum=0
    Value=&RigVMFunction_MathDoubleRemap(SourceMaximum=-100, SourceMinimum=-10, TargetMaximum=1, TargetMinimum=0, Value=RigVMVariableNode() {variable=l thighLocal rz float}.Value).Result
   )
   > &RigUnit_ModifyTransforms(
    ItemToModify="((Item=(Type=Bone,Name=\"base_robeEnd_r_1\"),Transform=(Rotation=(X=-0.0609,Y=0.2252,Z=0.2537,W=-0.9387),Translation=(X=0,Y=0,Z=0),Scale3D=(X=1,Y=1,Z=1))),(Item=(Type=Bone,Name=\"None\"),Transform=(Rotation=(X=0,Y=0,Z=0,W=1),Translation=(X=0,Y=0,Z=0),Scale3D=( ...(truncated)"
    WeightMinimum=0
    Weight=@RigVMFunction_AnimEvalRichCurve.Result
   )
   > &RigUnit_ModifyTransforms(
    ItemToModify=((Item=(Type=Bone,Name="kin_beltRope"),Transform=(Rotation=(X=0,Y=0,Z=-0.2588,W=0.9659),Translation=(X=-1.9,Y=-6.3,Z=1),Scale3D=(X=1,Y=1,Z=1))))
    WeightMinimum=0
    Weight=@RigVMFunction_AnimEvalRichCurve.Result
   )
   > RigVMFunctionReturnNode()
  ]}
 New Function_ContainedGraph {nodes [
   RigVMFunctionEntryNode()
   > &RigUnit_ParentConstraint(
    AdvancedSettings=(InterpolationType=Average,RotationOrderForFilter=XZY)
    Child=(Type=Bone,Name="deform_ankle_in_l")
    Filter=(TranslationFilter=(bX=True,bY=True,bZ=True),RotationFilter=(bX=True,bY=True,bZ=True),ScaleFilter=(bX=False,bY=False,bZ=False))
    Parents=((Item=(Type=Bone,Name="calf_l"),Weight=0.5),(Item=(Type=Bone,Name="foot_l"),Weight=0.5))
    Weight=1
   )
   > &RigUnit_ParentConstraint(
    AdvancedSettings=(InterpolationType=Average,RotationOrderForFilter=XZY)
    Child=(Type=Bone,Name="deform_ankle_out_l")
    Filter=(TranslationFilter=(bX=True,bY=True,bZ=True),RotationFilter=(bX=True,bY=True,bZ=True),ScaleFilter=(bX=False,bY=False,bZ=False))
    Parents=((Item=(Type=Bone,Name="calf_l"),Weight=0.5),(Item=(Type=Bone,Name="foot_l"),Weight=0.5))
    Weight=1
   )
   > &RigUnit_ParentConstraint(
    AdvancedSettings=(InterpolationType=Average,RotationOrderForFilter=XZY)
    Child=(Type=Bone,Name="deform_ankle_in_r")
    Filter=(TranslationFilter=(bX=True,bY=True,bZ=True),RotationFilter=(bX=True,bY=True,bZ=True),ScaleFilter=(bX=False,bY=False,bZ=False))
    Parents=((Item=(Type=Bone,Name="calf_r"),Weight=0.5),(Item=(Type=Bone,Name="foot_r"),Weight=0.5))
    Weight=1
   )
   > &RigUnit_ParentConstraint(
    AdvancedSettings=(InterpolationType=Average,RotationOrderForFilter=XZY)
    Child=(Type=Bone,Name="deform_ankle_out_r")
    Filter=(TranslationFilter=(bX=True,bY=True,bZ=True),RotationFilter=(bX=True,bY=True,bZ=True),ScaleFilter=(bX=False,bY=False,bZ=False))
    Parents=((Item=(Type=Bone,Name="calf_r"),Weight=0.5),(Item=(Type=Bone,Name="foot_r"),Weight=0.5))
    Weight=1
   )
   > DeltaQuatExtraction = RigVMFunctionReferenceNode(child=(Type=Bone,Name="foot_l"), onlyTwist=false, parent=(Type=Null,Name="foot_l_null"), twistAxis=(X=1,Y=0,Z=0))
   RigVMFunction_MathQuaternionToEuler = &RigVMFunction_MathQuaternionToEuler(RotationOrder=ZYX, Value=@DeltaQuatExtraction.outputQuat)
   Remap = &RigVMFunction_MathDoubleRemap(SourceMaximum=-60, SourceMinimum=0, TargetMaximum=1, TargetMinimum=0, Value=@RigVMFunction_MathQuaternionToEuler.Result.Z)
   > &RigUnit_ModifyTransforms(
    ItemToModify="((Item=(Type=Bone,Name=\"deform_ankle_in_l\"),Transform=(Rotation=(X=0,Y=0,Z=0.3907,W=0.9205),Translation=(X=2,Y=3,Z=0),Scale3D=(X=0.2,Y=1,Z=1))),(Item=(Type=Bone,Name=\"deform_ankle_out_l\"),Transform=(Rotation=(X=0,Y=0,Z=0.2164,W=0.9763),Translation=(X=0.6,Y ...(truncated)"
    WeightMinimum=0
    Weight=@Remap.Result
   )
   Power = /Script/RigVM.RigVMFunction_MathDoublePow(
    B=1.5
    A=&RigVMFunction_MathDoubleRemap(SourceMaximum=80, SourceMinimum=0, TargetMaximum=1, TargetMinimum=0, Value=@RigVMFunction_MathQuaternionToEuler.Result.Z).Result
   ) {is=struct RigVMUnitNode, template="Power::Execute(in A,in B,out Result)"}
   > &RigUnit_ModifyTransforms(
    ItemToModify="((Item=(Type=Bone,Name=\"deform_ankle_in_l\"),Transform=(Rotation=(X=0,Y=0,Z=-0.7071,W=0.7071),Translation=(X=-0.2,Y=-3,Z=0),Scale3D=(X=1,Y=1,Z=1))),(Item=(Type=Bone,Name=\"deform_ankle_out_l\"),Transform=(Rotation=(X=0,Y=0,Z=-0.2588,W=0.9659),Translation=(X=1 ...(truncated)"
    WeightMinimum=0
    Weight=@Power.Result
   )
   > DeltaQuatExtraction_1 = RigVMFunctionReferenceNode(child=(Type=Bone,Name="foot_r"), onlyTwist=false, parent=(Type=Null,Name="foot_r_null"), twistAxis=(X=1,Y=0,Z=0))
   n1 = &RigVMFunction_MathQuaternionToEuler(RotationOrder=ZYX, Value=@DeltaQuatExtraction_1.outputQuat)
   > &RigUnit_ModifyTransforms(
    ItemToModify="((Item=(Type=Bone,Name=\"deform_ankle_in_r\"),Transform=(Rotation=(X=0,Y=0,Z=0.3907,W=0.9205),Translation=(X=-2,Y=-3,Z=0),Scale3D=(X=0.2,Y=1,Z=1))),(Item=(Type=Bone,Name=\"deform_ankle_out_r\"),Transform=(Rotation=(X=0,Y=0,Z=0.2164,W=0.9763),Translation=(X=-0. ...(truncated)"
    WeightMinimum=0
    Weight=&RigVMFunction_MathDoubleRemap(SourceMaximum=-60, SourceMinimum=0, TargetMaximum=1, TargetMinimum=0, Value=@n1.Result.Z).Result
   )
   > &RigUnit_ModifyTransforms(
    ItemToModify="((Item=(Type=Bone,Name=\"deform_ankle_in_r\"),Transform=(Rotation=(X=0,Y=0,Z=-0.7071,W=0.7071),Translation=(X=0.2,Y=3,Z=0),Scale3D=(X=1,Y=1,Z=1))),(Item=(Type=Bone,Name=\"deform_ankle_out_r\"),Transform=(Rotation=(X=0,Y=0,Z=-0.2588,W=0.9659),Translation=(X=-1. ...(truncated)"
    WeightMinimum=0
    Weight=&RigVMFunction_MathDoublePow(
     B=1.5
     A=&RigVMFunction_MathDoubleRemap(SourceMaximum=80, SourceMinimum=0, TargetMaximum=1, TargetMinimum=0, Value=@n1.Result.Z).Result
    ) {template="Power::Execute(in A,in B,out Result)"}.Result
   )
   > RigVMFunctionReturnNode()
  ]}
 RigVMFunctionLibrary {nodes [
   DeltaQuatExtraction = RigVMCollapseNode(ExecuteContext=(), child=(Type=Bone,Name="None"), parent=(Type=Bone,Name="None"), twistAxis=(X=1,Y=0,Z=0))
   DeltaRotation v4 = RigVMCollapseNode(
    ExecuteContext=()
    Rotation Order=ZYX
    child=(Type=Bone,Name="None")
    outputMinMax=(X=0,Y=1)
    parent=(Type=Bone,Name="None")
    premult=1
    rotationMinMax=(X=0,Y=0)
    twistAxis=(X=1,Y=0,Z=0)
   )
   ankleCorrectives = RigVMCollapseNode(ExecuteContext=())
   bathRobe = RigVMCollapseNode(ExecuteContext=())
   knot = RigVMCollapseNode(ExecuteContext=())
   misc = RigVMCollapseNode(ExecuteContext=())
   twistFix = RigVMCollapseNode(ExecuteContext=())
  ]}
 RigVMFunction_Sequence_2_ContainedGraph {nodes [
   n1 = RigVMFunctionEntryNode()
   n2 = &RigVMFunction_Sequence(ExecuteContext=@n1.ExecuteContext)
   n3 = &RigVMFunction_Sequence(ExecuteContext=@n2.B)
   RigVMFunctionReturnNode(A=@n2.A, B=@n3.A, C=@n3.B)
  ]}
 RigVMFunction_Sequence_3_ContainedGraph {nodes [
   n1 = RigVMFunctionEntryNode()
   n2 = &RigVMFunction_Sequence(ExecuteContext=@n1.ExecuteContext)
   n3 = &RigVMFunction_Sequence(ExecuteContext=@n2.B)
   RigVMFunctionReturnNode(B=@n3.A, C=@n3.B, A=@n2.A)
  ]}
 RigVMModel {nodes [
   DeltaRotation v4 = RigVMFunctionReferenceNode(
    ExecuteContext=()
    Rotation Order=ZYX
    child=(Type=Bone,Name="foot_r")
    debugRotation=false
    outputMinMax=(X=0,Y=1)
    parent=(Type=Null,Name="foot_r_null")
    premult=1
    rotationMinMax=(X=0,Y=-60)
    triggerAxis=Z
    twistAxis=(X=1,Y=0,Z=0)
   )
   > &RigUnit_ModifyTransforms(
    ItemToModify="((Item=(Type=Bone,Name=\"deform_ankle_in_r\"),Transform=(Rotation=(X=0,Y=0,Z=0.3907,W=0.9205),Translation=(X=-1.15,Y=-6.35,Z=0),Scale3D=(X=0.2,Y=1,Z=1))),(Item=(Type=Bone,Name=\"deform_ankle_out_r\"),Transform=(Rotation=(X=0,Y=0,Z=0.2588,W=0.9659),Translation= ...(truncated)"
    WeightMinimum=0
    Weight=@"DeltaRotation v4".outputWeight
   )
   > DeltaRotation v4_1 = RigVMFunctionReferenceNode(
    Rotation Order=ZYX
    child=(Type=Bone,Name="foot_r")
    debugRotation=true
    outputMinMax=(X=0,Y=1)
    parent=(Type=Null,Name="foot_r_null")
    premult=1
    rotationMinMax=(X=0,Y=80)
    triggerAxis=Z
    twistAxis=(X=1,Y=0,Z=0)
   )
   > &RigUnit_ModifyTransforms(
    ItemToModify="((Item=(Type=Bone,Name=\"deform_ankle_in_r\"),Transform=(Rotation=(X=0,Y=0,Z=-0.4848,W=0.8746),Translation=(X=0.6,Y=3,Z=0),Scale3D=(X=1,Y=1,Z=1))),(Item=(Type=Bone,Name=\"deform_ankle_out_r\"),Transform=(Rotation=(X=0,Y=0,Z=-0.1564,W=0.9877),Translation=(X=2.6 ...(truncated)"
    WeightMinimum=0
    Weight=@"DeltaRotation v4_1".outputWeight
   )
   RigUnit_BeginExecution = &RigUnit_BeginExecution()
   twistFix = RigVMFunctionReferenceNode(ExecuteContext=@RigUnit_BeginExecution.ExecutePin)
   > FN_DeformRig_FNC = RigVMFunctionReferenceNode(
    DeformRig_ Curve="(deformRig_l_upperarm_up_alpha_72_#1=False,deformRig_r_upperarm_up_alpha_73_#2=False,deformRig_l_upperarm_fr_alpha_74_#3=False,deformRig_r_upperarm_fr_alpha_75_CB51B8 ...(truncated)"
    DeformRig_Bool="(upperarm_1_#4=True,elbow_3_#5=True,wrist_7_#6=True,groin_8_#7=true,glute_17_#8=true,knee_9_40C408324CA261 ...(truncated)"
    DeformRig_Debug=(upperarm_debug_18_#4=False,wrist_debug_20_#6=False,lat_debug_26_#9=False,groinGlute_debug_27_#7=False)
    DeformRig_Value="(upperarm_up_18_#10=(Rotation=(X=0,Y=0,Z=0,W=1),Translation=(X=0,Y=0,Z=3),Scale3D=(X=0.7,Y=1,Z=1)),upperarm_fwd_fr_19_#11=(Rotation=(X=0,Y=0,Z=0,W=1),Translation=(X=0,Y=10,Z=0),Scale3D=(X=0.5,Y=1,Z= ...(truncated)"
   )
   > bathRobe = RigVMFunctionReferenceNode()
   > misc = RigVMFunctionReferenceNode()
   > knot = RigVMFunctionReferenceNode()
   > ankleCorrectives = RigVMFunctionReferenceNode()
  ]}
 RigVMModel Rig {nodes [
   n1 = &RigUnit_PrepareForExecution()
   &RigUnit_HierarchyImportFromSkeleton(NameSpace=None, bIncludeCurves=true, bIncludeMeshSockets=False, bIncludeVirtualBones=True, ExecutePin=@n1.ExecutePin)
   ItemArray = RigVMDispatchNode(Value="((Type=Bone,Name=\"thigh_l\"),(Type=Bone,Name=\"calf_l\"),(Type=Bone,Name=\"foot_l\"),(Type=Bone,Name=\"thigh_r\"),(Type=Bone,Name=\"calf_r\"),(Type=Bone,Name=\"foot_r\"),(Type=Bone,Name=\"clavicle_l\"),(Type=Bone,Name=\"upperarm_l\"),(Type=Bone,Name=\"hand_l\"),(Type=Bone, ...(truncated)") {template="DISPATCH_RigVMDispatch_Constant(io Value)"}
   > addNullObject = RigVMFunctionReferenceNode(
    Options_Null="(pelvis_1_#12=False,spine_3_#13=False,clavicle_5_#14=False,arm_7_#15=False,leg_9_#16=False,head_11_46DC15EF465C ...(truncated)"
    Custom_Joints=@ItemArray.Value
   )
   > &RigUnit_HierarchyAddNull(
    Name=spine02_null
    Parent=(Type=Bone,Name="pelvis")
    Space=LocalSpace
    Transform=&RigUnit_GetRelativeTransformForItem(Child=(Type=Bone,Name="spine_02"), Parent=(Type=Bone,Name="pelvis"), bChildInitial=true, bParentInitial=true).RelativeTransform
   )
  ]}
]
hierarchy [
 crouching alpha {type=Curve, loc=<0 0 0>, rot=<0 0 0 1>}
 crouchingAlpha {type=Curve, loc=<0 0 0>, rot=<0 0 0 1>}
 r thighLocal ry float {type=Curve, loc=<0 0 0>, rot=<0 0 0 1>}
 r thighLocal rz float {type=Curve, loc=<0 0 0>, rot=<0 0 0 1>}
 r_thighLocal_ry_float {type=Curve, loc=<0 0 0>, rot=<0 0 0 1>}
 r_thighLocal_rz_float {type=Curve, loc=<0 0 0>, rot=<0 0 0 1>}
]
variables [
 isCrouched alpha {type=double, default=0}
 l thighLocal ry float {type=double, default=0}
 l thighLocal rz float {type=double, default=0}
 pub isBaseballSliding {type=bool, default=False}
 pub isCrouching {type=bool, default=False}
 pub isDBNO {type=bool, default=False}
 pub isDBNO Carried {type=bool, default=False}
 pub isGrindRail {type=bool, default=False}
 pub isKneeSliding {type=bool, default=False}
 pub isPlayingEasternBloc {type=bool, default=False}
 pub isPlayingEmote {type=bool, default=False}
 pub isRBANOn {type=bool, default=False}
 pub isZipline {type=bool, default=False}
 pub moveBackward {type=double, default=0}
 pub moveForward {type=double, default=0}
 pub strafeLeft {type=double, default=0}
 pub strafeRight {type=double, default=0}
 r thighLocal ry float {type=double, default=0}
 r thighLocal rz float {type=double, default=0}
]
