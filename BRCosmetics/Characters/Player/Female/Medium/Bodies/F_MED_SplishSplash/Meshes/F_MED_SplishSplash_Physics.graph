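# Physics asset for the F_MED_SplishSplash female player body cosmetic, set up for motion of its skirt, belt rope and robe-end bones.
# Eight two-link skirt chains hang from small base capsules constrained to the pelvis; belt rope and robe-end chains hang from bases on spine_01.
# Dynamic links have limited swing and twist, while base and body constraints rotate freely.
# Pelvis, spine, thigh and calf capsules supply body volumes, with a thigh twist capsule only on the left.
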PhysicsAsset /BRCosmetics/Characters/Player/Female/Medium/Bodies/F_MED_SplishSplash/Meshes/F_MED_SplishSplash_Physics
bodies {
 base_C_skirt_a_2 [{shapeType=Capsule, length=2, radius=1, shapeName=base_C_skirt_a_2_capsule}]
 base_C_skirt_b_2 [{shapeType=Capsule, length=2, radius=1, shapeName=base_C_skirt_b_2_capsule}]
 base_L_skirt_a_2 [{shapeType=Capsule, length=2, radius=1, shapeName=base_L_skirt_a_2_capsule}]
 base_L_skirt_b_2 [{shapeType=Capsule, length=2, radius=1, shapeName=base_L_skirt_b_2_capsule}]
 base_L_skirt_c_2 [{shapeType=Capsule, length=2, radius=1, shapeName=base_L_skirt_c_2_capsule}]
 base_R_skirt_a_2 [{shapeType=Capsule, length=2, radius=1, shapeName=base_R_skirt_a_2_capsule}]
 base_R_skirt_b_2 [{shapeType=Capsule, length=2, radius=1, shapeName=base_R_skirt_b_2_capsule}]
 base_R_skirt_c_2 [{shapeType=Capsule, length=2, radius=1, shapeName=base_R_skirt_c_2_capsule}]
 base_beltRope_1 [{shapeType=Capsule, length=0.505, radius=0.505, shapeName=base_robeEnd_r_1_capsule, rotation=<rot 0 -90 0>}]
 base_robeEnd_l_1 [{shapeType=Capsule, length=0.505, radius=0.505, shapeName=base_robeEnd_r_1_capsule, rotation=<rot 0 -90 0>}]
 base_robeEnd_r_1 [{shapeType=Capsule, length=0.505, radius=0.505, shapeName=base_robeEnd_r_1_capsule, rotation=<rot 0 -90 0>}]
 calf_l [{shapeType=Capsule, length=39.7913, radius=8.2847, shapeName=calf_l_capsule, center=<-23.8013 -2.3204 0>, rotation=<rot -90 180 180>}]
 calf_r [{shapeType=Capsule, length=39.7913, radius=8.2847, shapeName=calf_r_capsule, center=<23.8013 2.3204 0>, rotation=<rot -90 180 0>}]
 dyn_C_skirt_a_2 [{shapeType=Capsule, length=5.4412, radius=4.8087, shapeName=dyn_C_skirt_a_2_capsule, center=<6.6097 0 2.795>, rotation=<rot 90 0 0>}]
 dyn_C_skirt_a_3 [{shapeType=Capsule, length=4.1909, radius=4, shapeName=dyn_C_skirt_a_2_capsule, center=<5.6161 0 2.3962>, rotation=<rot 90 0 0>}]
 dyn_C_skirt_b_2 [{shapeType=Capsule, length=4.4102, radius=5.5913, shapeName=dyn_C_skirt_a_2_capsule, center=<7.3572 0 0.5141>, rotation=<rot 90 0 0>}]
 dyn_C_skirt_b_3 [{shapeType=Capsule, length=3.8813, radius=4, shapeName=dyn_C_skirt_a_2_capsule, center=<7.4643 0 2.3962>, rotation=<rot 90 0 0>}]
 dyn_L_skirt_a_2 [
  {shapeType=Capsule, length=7.3, radius=4, shapeName=dyn_C_skirt_a_2_capsule, center=<7.38 -4.64 3.4653>, rotation=<rot -90 0 0>}
  {shapeType=Capsule, length=7.3, radius=4, shapeName=dyn_C_skirt_a_2_capsule, center=<7.5 2.42 3.1153>, rotation=<rot -90 0 0>}
 ]
 dyn_L_skirt_a_3 [{shapeType=Capsule, length=9.3331, radius=4.2039, shapeName=dyn_R_skirt_a_3_capsule, center=<6.8323 0 1.9836>, rotation=<rot 0 -4.5 -82.9>}]
 dyn_L_skirt_b_2 [{shapeType=Capsule, length=10.5, radius=4, shapeName=dyn_C_skirt_a_2_capsule, center=<7.5 -1.2 3.1>, rotation=<rot 0 0 -90>}]
 dyn_L_skirt_b_3 [{shapeType=Capsule, length=10.55, radius=4, shapeName=dyn_C_skirt_a_2_capsule, center=<2.1 -1.2 3.5576>, rotation=<rot 0 0 -90>}]
 dyn_L_skirt_c_2 [
  {shapeType=Capsule, length=7, radius=4, shapeName=dyn_C_skirt_a_2_capsule, center=<7.5 4.75 1>, rotation=<rot -90 0 0>}
  {shapeType=Capsule, length=7, radius=4, shapeName=dyn_C_skirt_a_2_capsule, center=<7.5 -3.2 0.9889>, rotation=<rot -90 0 0>}
 ]
 dyn_L_skirt_c_3 [{shapeType=Capsule, length=18, radius=4, shapeName=dyn_C_skirt_c_3_capsule, center=<5.55 0 1.6>, rotation=<rot 0 0 -88>}]
 dyn_R_skirt_a_2 [
  {shapeType=Capsule, length=7.3, radius=4, shapeName=dyn_C_skirt_a_2_capsule, center=<7.38 -4.64 2.5>, rotation=<rot -90 0 0>}
  {shapeType=Capsule, length=7.3, radius=4, shapeName=dyn_C_skirt_a_2_capsule, center=<7.5 2.42 2.15>, rotation=<rot -90 0 0>}
 ]
 dyn_R_skirt_a_3 [{shapeType=Capsule, length=9.3331, radius=3.8705, shapeName=dyn_R_skirt_a_3_capsule, center=<6.8323 0 1.9836>, rotation=<rot 0 4.5 -82.9>}]
 dyn_R_skirt_b_2 [{shapeType=Capsule, length=10.5, radius=4, shapeName=dyn_C_skirt_a_2_capsule, center=<7.5 1.2 3.1>, rotation=<rot 0 0 -90>}]
 dyn_R_skirt_b_3 [{shapeType=Capsule, length=8.4311, radius=6.119, shapeName=dyn_C_skirt_a_2_capsule, center=<4.823 1.2 4.8554>, rotation=<rot 0 0 -90>}]
 dyn_R_skirt_c_2 [
  {shapeType=Capsule, length=7, radius=4, shapeName=dyn_C_skirt_a_2_capsule, center=<7.5 4.75 0.7589>, rotation=<rot -90 0 0>}
  {shapeType=Capsule, length=7, radius=4, shapeName=dyn_C_skirt_a_2_capsule, center=<7.5 -3.2 2.23>, rotation=<rot -90 0 0>}
 ]
 dyn_R_skirt_c_3 [{shapeType=Capsule, length=18, radius=4.4766, shapeName=dyn_C_skirt_c_3_capsule, center=<6.2776 0.0381 2.6899>, rotation=<rot 0 0 -88>}]
 dyn_beltRope_1 [{
   shapeType=Box
   extentX=4.1727
   extentY=6.6349
   extentZ=1.2594
   shapeName=dyn_beltRope_1_box
   center=<2.5381 -0.3586 0.5894>
   rotation=<rot -3.6085 0.0001 0>
  }]
 dyn_beltRope_2 [{
   shapeType=Box
   extentX=4.637
   extentY=6.461
   extentZ=1.9842
   shapeName=dyn_beltRope_2_box
   center=<2.7672 -0.3356 0.0577>
   rotation=<rot -2.6866 0 0>
  }]
 dyn_beltRope_3 [{shapeType=Box, extentX=5.4877, extentY=6.4353, extentZ=1.6428, shapeName=dyn_beltRope_3_box, center=<3.0877 -0.3773 0.0289>}]
 dyn_robeEnd_l_1 [{
   shapeType=Capsule
   length=2.6103
   radius=2.0615
   shapeName=dyn_robeEnd_l_1_capsule
   center=<3.5801 0.1478 0.563>
   rotation=<rot 9.1924 -88.0192 -78.4487>
  }]
 dyn_robeEnd_l_2 [{
   shapeType=Capsule
   length=3.5855
   radius=1.9347
   shapeName=dyn_robeEnd_l_2_capsule
   center=<3.4929 0.0568 0.4189>
   rotation=<rot -1.2598 -174.9316 -85.488>
  }]
 dyn_robeEnd_r_1 [{
   shapeType=Capsule
   length=2.1412
   radius=2.6456
   shapeName=dyn_robeEnd_r_1_capsule
   center=<2.6501 -0.0299 0.3996>
   rotation=<rot 2.087 86.9158 -94.689>
  }]
 dyn_robeEnd_r_2 [{
   shapeType=Capsule
   length=3.4785
   radius=2.136
   shapeName=dyn_robeEnd_r_2_capsule
   center=<2.9159 0.0664 -0.1777>
   rotation=<rot 0 -5.5141 -89.3768>
  }]
 pelvis [{shapeType=Capsule, length=12.6204, radius=10.2578, shapeName=pelvis_capsule, center=<0 -0.8599 -0.1713>, rotation=<rot 0 -0.0002 9.8849>}]
 spine_01 [{
   shapeType=Capsule
   length=8.4598
   radius=7.4497
   shapeName=spine_01_capsule
   center=<1.3817 1.8615 -1.247>
   rotation=<rot 16.8561 -89.6387 0.8505>
  }]
 thigh_l [
  {
   shapeType=Capsule
   length=18.6599
   radius=8.361
   shapeName=thigh_l_capsule
   center=<-26.0211 0.1494 0.8187>
   rotation=<rot 88.7672 177.9805 177.7048>
  }
  {
   shapeType=Capsule
   length=19.5414
   radius=7.4795
   shapeName=thigh_l_capsule
   center=<-26.0155 -1.0069 0.8197>
   rotation=<rot 88.7672 177.9805 177.7048>
  }
 ]
 thigh_r [
  {
   shapeType=Capsule
   length=18.6599
   radius=8.361
   shapeName=thigh_l_capsule
   center=<26.0211 -0.1494 -0.8187>
   rotation=<rot 88.7672 177.9805 177.7048>
  }
  {
   shapeType=Capsule
   length=19.5414
   radius=7.4795
   shapeName=thigh_l_capsule
   center=<26.0155 1.0069 -0.8197>
   rotation=<rot 88.7672 177.9805 177.7048>
  }
 ]
 thigh_twist_01_l [{
   shapeType=Capsule
   length=25.3732
   radius=9.7817
   shapeName=thigh_l_capsule
   center=<13.9058 -0.5381 1.7759>
   rotation=<rot -88.2395 -0.0018 -5.0119>
  }]
}
constraints [
 {bone1Name=dyn_C_skirt_a_3, bone2Name=dyn_C_skirt_a_2, swing1LimitDegrees=15, swing2LimitDegrees=35, twistLimitDegrees=10}
 {bone1Name=dyn_C_skirt_a_2, bone2Name=base_C_skirt_a_2, swing1LimitDegrees=5, swing2LimitDegrees=22, twistLimitDegrees=15}
 {bone1Name=dyn_C_skirt_b_2, bone2Name=base_C_skirt_b_2, swing1LimitDegrees=55, swing2LimitDegrees=22, twistLimitDegrees=0}
 {bone1Name=dyn_C_skirt_b_3, bone2Name=dyn_C_skirt_b_2, swing1LimitDegrees=15, swing2LimitDegrees=35, twistLimitDegrees=0}
 {bone1Name=dyn_L_skirt_a_2, bone2Name=base_L_skirt_a_2, swing1LimitDegrees=5, swing2LimitDegrees=22, twistLimitDegrees=0}
 {bone1Name=dyn_L_skirt_a_3, bone2Name=dyn_L_skirt_a_2, swing1LimitDegrees=15, swing2LimitDegrees=35, twistLimitDegrees=0}
 {bone1Name=dyn_L_skirt_b_2, bone2Name=base_L_skirt_b_2, swing1LimitDegrees=5, swing2LimitDegrees=22, twistLimitDegrees=0}
 {bone1Name=dyn_L_skirt_b_3, bone2Name=dyn_L_skirt_b_2, swing1LimitDegrees=15, swing2LimitDegrees=35, twistLimitDegrees=0}
 {bone1Name=dyn_L_skirt_c_3, bone2Name=dyn_L_skirt_c_2, swing1LimitDegrees=15, swing2LimitDegrees=35, twistLimitDegrees=0}
 {bone1Name=dyn_L_skirt_c_2, bone2Name=base_L_skirt_c_2, swing1LimitDegrees=5, swing2LimitDegrees=22, twistLimitDegrees=0}
 {bone1Name=dyn_R_skirt_a_2, bone2Name=base_R_skirt_a_2, swing1LimitDegrees=5, swing2LimitDegrees=22, twistLimitDegrees=0}
 {bone1Name=dyn_R_skirt_a_3, bone2Name=dyn_R_skirt_a_2, swing1LimitDegrees=15, swing2LimitDegrees=35, twistLimitDegrees=0}
 {bone1Name=dyn_R_skirt_b_2, bone2Name=base_R_skirt_b_2, swing1LimitDegrees=5, swing2LimitDegrees=22, twistLimitDegrees=0}
 {bone1Name=dyn_R_skirt_b_3, bone2Name=dyn_R_skirt_b_2, swing1LimitDegrees=15, swing2LimitDegrees=35, twistLimitDegrees=0}
 {bone1Name=dyn_R_skirt_c_2, bone2Name=base_R_skirt_c_2, swing1LimitDegrees=5, swing2LimitDegrees=22, twistLimitDegrees=0}
 {bone1Name=dyn_R_skirt_c_3, bone2Name=dyn_R_skirt_c_2, swing1LimitDegrees=15, swing2LimitDegrees=35, twistLimitDegrees=0}
 {
  bone1Name=calf_r
  bone2Name=thigh_r
  swing1LimitDegrees=45
  swing1Motion=Free
  swing2LimitDegrees=45
  swing2Motion=Free
  twistLimitDegrees=45
  twistMotion=Free
 }
 {
  bone1Name=calf_l
  bone2Name=thigh_l
  swing1LimitDegrees=45
  swing1Motion=Free
  swing2LimitDegrees=45
  swing2Motion=Free
  twistLimitDegrees=45
  twistMotion=Free
 }
 {
  bone1Name=thigh_l
  bone2Name=pelvis
  swing1LimitDegrees=45
  swing1Motion=Free
  swing2LimitDegrees=45
  swing2Motion=Free
  twistLimitDegrees=45
  twistMotion=Free
 }
 {
  bone1Name=thigh_r
  bone2Name=pelvis
  swing1LimitDegrees=45
  swing1Motion=Free
  swing2LimitDegrees=45
  swing2Motion=Free
  twistLimitDegrees=45
  twistMotion=Free
 }
 {
  bone1Name=base_C_skirt_a_2
  bone2Name=pelvis
  swing1LimitDegrees=45
  swing1Motion=Free
  swing2LimitDegrees=45
  swing2Motion=Free
  twistLimitDegrees=45
  twistMotion=Free
 }
 {
  bone1Name=base_C_skirt_b_2
  bone2Name=pelvis
  swing1LimitDegrees=45
  swing1Motion=Free
  swing2LimitDegrees=45
  swing2Motion=Free
  twistLimitDegrees=45
  twistMotion=Free
 }
 {
  bone1Name=base_L_skirt_a_2
  bone2Name=pelvis
  swing1LimitDegrees=45
  swing1Motion=Free
  swing2LimitDegrees=45
  swing2Motion=Free
  twistLimitDegrees=45
  twistMotion=Free
 }
 {
  bone1Name=base_L_skirt_b_2
  bone2Name=pelvis
  swing1LimitDegrees=45
  swing1Motion=Free
  swing2LimitDegrees=45
  swing2Motion=Free
  twistLimitDegrees=45
  twistMotion=Free
 }
 {
  bone1Name=base_L_skirt_c_2
  bone2Name=pelvis
  swing1LimitDegrees=45
  swing1Motion=Free
  swing2LimitDegrees=45
  swing2Motion=Free
  twistLimitDegrees=45
  twistMotion=Free
 }
 {
  bone1Name=base_R_skirt_a_2
  bone2Name=pelvis
  swing1LimitDegrees=45
  swing1Motion=Free
  swing2LimitDegrees=45
  swing2Motion=Free
  twistLimitDegrees=45
  twistMotion=Free
 }
 {
  bone1Name=base_R_skirt_b_2
  bone2Name=pelvis
  swing1LimitDegrees=45
  swing1Motion=Free
  swing2LimitDegrees=45
  swing2Motion=Free
  twistLimitDegrees=45
  twistMotion=Free
 }
 {
  bone1Name=base_R_skirt_c_2
  bone2Name=pelvis
  swing1LimitDegrees=45
  swing1Motion=Free
  swing2LimitDegrees=45
  swing2Motion=Free
  twistLimitDegrees=45
  twistMotion=Free
 }
 {
  bone1Name=spine_01
  bone2Name=pelvis
  swing1LimitDegrees=45
  swing1Motion=Free
  swing2LimitDegrees=45
  swing2Motion=Free
  twistLimitDegrees=45
  twistMotion=Free
 }
 {
  bone1Name=base_beltRope_1
  bone2Name=spine_01
  swing1LimitDegrees=45
  swing1Motion=Free
  swing2LimitDegrees=45
  swing2Motion=Free
  twistLimitDegrees=45
  twistMotion=Free
 }
 {bone1Name=dyn_beltRope_1, bone2Name=base_beltRope_1, swing1LimitDegrees=20, swing2LimitDegrees=20, twistLimitDegrees=10}
 {bone1Name=dyn_beltRope_2, bone2Name=dyn_beltRope_1, swing1LimitDegrees=20, swing2LimitDegrees=20, twistLimitDegrees=10}
 {bone1Name=dyn_beltRope_3, bone2Name=dyn_beltRope_2, swing1LimitDegrees=20, swing2LimitDegrees=20, twistLimitDegrees=10}
 {
  bone1Name=base_robeEnd_l_1
  bone2Name=spine_01
  swing1LimitDegrees=45
  swing1Motion=Free
  swing2LimitDegrees=45
  swing2Motion=Free
  twistLimitDegrees=45
  twistMotion=Free
 }
 {bone1Name=dyn_robeEnd_l_1, bone2Name=base_robeEnd_l_1, swing1LimitDegrees=20, swing2LimitDegrees=20, twistLimitDegrees=0}
 {bone1Name=dyn_robeEnd_l_2, bone2Name=dyn_robeEnd_l_1, swing1LimitDegrees=20, swing2LimitDegrees=20, twistLimitDegrees=0}
 {
  bone1Name=base_robeEnd_r_1
  bone2Name=spine_01
  swing1LimitDegrees=45
  swing1Motion=Free
  swing2LimitDegrees=45
  swing2Motion=Free
  twistLimitDegrees=45
  twistMotion=Free
 }
 {bone1Name=dyn_robeEnd_r_1, bone2Name=base_robeEnd_r_1, swing1LimitDegrees=20, swing2LimitDegrees=20, twistLimitDegrees=0}
 {bone1Name=dyn_robeEnd_r_2, bone2Name=dyn_robeEnd_r_1, swing1LimitDegrees=20, swing2LimitDegrees=20, twistLimitDegrees=0}
 {
  bone1Name=thigh_twist_01_l
  bone2Name=thigh_l
  swing1LimitDegrees=45
  swing1Motion=Free
  swing2LimitDegrees=45
  swing2Motion=Free
  twistLimitDegrees=45
  twistMotion=Free
 }
]
