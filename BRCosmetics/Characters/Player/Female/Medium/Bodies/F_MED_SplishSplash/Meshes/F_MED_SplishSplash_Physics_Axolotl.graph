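# Physics asset for the Axolotl plushie on the medium female SplishSplash cosmetic body, covering the plushie's head and arms.
# Each dyn_ head and arm capsule is constrained to a small base_ anchor capsule with narrow swing limits and little or no twist.
# Pelvis, spine and four coll_plushieArms capsules supply body collision shapes.
# Anchors and arm colliders link to the pelvis, and the spine chains up from it, all with free angular motion.
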PhysicsAsset /BRCosmetics/Characters/Player/Female/Medium/Bodies/F_MED_SplishSplash/Meshes/F_MED_SplishSplash_Physics_Axolotl
bodies {
 base_plushieArm_l [{shapeType=Capsule, length=2, radius=0.5, shapeName=base_plushieArm_l_capsule}]
 base_plushieArm_r [{shapeType=Capsule, length=2, radius=0.5, shapeName=base_plushieArm_r_capsule}]
 base_plushieHead [{shapeType=Capsule, length=1, radius=0.5, shapeName=base_plushieHead_capsule}]
 coll_plushieArms_1 [{
   shapeType=Capsule
   length=14.2198
   radius=6.0122
   shapeName=coll_plushieArms_1_capsule
   center=<4.5732 -1.191 4.8405>
   rotation=<rot 89.442 0.0006 0.0004>
  }]
 coll_plushieArms_2 [{
   shapeType=Capsule
   length=10.8281
   radius=3.1173
   shapeName=coll_plushieArms_2_capsule
   center=<3.7403 -2.1105 0.393>
   rotation=<rot 0.6073 -84.4858 83.7338>
  }]
 coll_plushieArms_3 [{
   shapeType=Capsule
   length=6.4801
   radius=4.0315
   shapeName=coll_plushieArms_3_capsule
   center=<3.7759 0.5309 1.7675>
   rotation=<rot 86.5396 -177.6974 178.3169>
  }]
 coll_plushieArms_4 [{
   shapeType=Capsule
   length=5.4394
   radius=4.3541
   shapeName=coll_plushieArms_4_capsule
   center=<4.3265 -2.0918 1.5255>
   rotation=<rot 0.0459 -86.5147 -90.7537>
  }]
 dyn_plushieArm_l [{
   shapeType=Capsule
   length=4.0421
   radius=2.2027
   shapeName=dyn_plushieArm_l_capsule
   center=<4.5947 -0.0968 0.5275>
   rotation=<rot 0 -87.9723 -86.7313>
  }]
 dyn_plushieArm_r [{
   shapeType=Capsule
   length=5.5162
   radius=2.0311
   shapeName=dyn_plushieArm_r_capsule
   center=<4.2771 0.1818 0.5766>
   rotation=<rot -0.0042 82.936 -95.9813>
  }]
 dyn_plushieHead [{
   shapeType=Capsule
   length=13.9263
   radius=6.6702
   shapeName=dyn_plushieHead_capsule
   center=<6.1869 3.6011 -0.2017>
   rotation=<rot -6.8716 -89.5718 -2.6341>
  }]
 pelvis [{shapeType=Capsule, length=8.4462, radius=10.5826, shapeName=pelvis_capsule}]
 spine_01 [{shapeType=Capsule, length=5.4643, radius=9.2001, shapeName=spine_01_capsule, center=<0.1588 1.5729 0.8808>, rotation=<rot -8.0292 -90 0.822>}]
 spine_02 [{shapeType=Capsule, length=11.8271, radius=8.6332, shapeName=spine_02_capsule, center=<0 1.5355 -3.1457>}]
 spine_03 [{shapeType=Capsule, length=7.2174, radius=11.4149, shapeName=spine_03_capsule, center=<2.9929 1.6937 -0.6202>, rotation=<rot 90 0 0>}]
}
constraints [
 {bone1Name=dyn_plushieHead, bone2Name=base_plushieHead, swing1LimitDegrees=20, swing2LimitDegrees=20, twistLimitDegrees=10}
 {
  bone1Name=base_plushieHead
  bone2Name=pelvis
  swing1LimitDegrees=45
  swing1Motion=Free
  swing2LimitDegrees=45
  swing2Motion=Free
  twistLimitDegrees=45
  twistMotion=Free
 }
 {
  bone1Name=spine_02
  bone2Name=spine_01
  swing1LimitDegrees=45
  swing1Motion=Free
  swing2LimitDegrees=45
  swing2Motion=Free
  twistLimitDegrees=45
  twistMotion=Free
 }
 {
  bone1Name=spine_01
  bone2Name=pelvis
  swing1LimitDegrees=45
  swing1Motion=Free
  swing2LimitDegrees=45
  swing2Motion=Free
  twistLimitDegrees=45
  twistMotion=Free
 }
 {
  bone1Name=coll_plushieArms_4
  bone2Name=pelvis
  swing1LimitDegrees=45
  swing1Motion=Free
  swing2LimitDegrees=45
  swing2Motion=Free
  twistLimitDegrees=45
  twistMotion=Free
 }
 {
  bone1Name=coll_plushieArms_1
  bone2Name=pelvis
  swing1LimitDegrees=45
  swing1Motion=Free
  swing2LimitDegrees=45
  swing2Motion=Free
  twistLimitDegrees=45
  twistMotion=Free
 }
 {
  bone1Name=coll_plushieArms_2
  bone2Name=pelvis
  swing1LimitDegrees=45
  swing1Motion=Free
  swing2LimitDegrees=45
  swing2Motion=Free
  twistLimitDegrees=45
  twistMotion=Free
 }
 {
  bone1Name=coll_plushieArms_3
  bone2Name=pelvis
  swing1LimitDegrees=45
  swing1Motion=Free
  swing2LimitDegrees=45
  swing2Motion=Free
  twistLimitDegrees=45
  twistMotion=Free
 }
 {
  bone1Name=spine_03
  bone2Name=spine_02
  swing1LimitDegrees=45
  swing1Motion=Free
  swing2LimitDegrees=45
  swing2Motion=Free
  twistLimitDegrees=45
  twistMotion=Free
 }
 {bone1Name=dyn_plushieArm_r, bone2Name=base_plushieArm_r, swing1LimitDegrees=20, swing2LimitDegrees=20, twistLimitDegrees=0}
 {
  bone1Name=base_plushieArm_l
  bone2Name=pelvis
  swing1LimitDegrees=45
  swing1Motion=Free
  swing2LimitDegrees=45
  swing2Motion=Free
  twistLimitDegrees=45
  twistMotion=Free
 }
 {bone1Name=dyn_plushieArm_l, bone2Name=base_plushieArm_l, swing1LimitDegrees=20, swing2LimitDegrees=20, twistLimitDegrees=0}
 {
  bone1Name=base_plushieArm_r
  bone2Name=pelvis
  swing1LimitDegrees=45
  swing1Motion=Free
  swing2LimitDegrees=45
  swing2Motion=Free
  twistLimitDegrees=45
  twistMotion=Free
 }
]
