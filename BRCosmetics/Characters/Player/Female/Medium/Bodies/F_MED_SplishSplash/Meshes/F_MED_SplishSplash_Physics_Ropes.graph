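# Physics asset for the belt rope and left and right robe ends on the SplishSplash female medium body cosmetic.
# A three-box belt chain and two two-capsule end chains each hang from a small base capsule tied to spine_01.
# Chain joints have narrow swing limits, with tight twist on the belt and zero twist on the ends.
# Pelvis, spine_01 and left thigh capsules supply body shapes, joined by angularly free constraints; the right thigh has none.
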PhysicsAsset /BRCosmetics/Characters/Player/Female/Medium/Bodies/F_MED_SplishSplash/Meshes/F_MED_SplishSplash_Physics_Ropes
bodies {
 base_beltRope_1 [{shapeType=Capsule, length=2, radius=0.5, shapeName=base_beltRope_1_capsule, rotation=<rot 0 0 90>}]
 base_robeEnd_l_1 [{shapeType=Capsule, length=0.505, radius=0.505, shapeName=base_robeEnd_l_1_capsule, rotation=<rot 0 -90 0>}]
 base_robeEnd_r_1 [{shapeType=Capsule, length=0.505, radius=0.505, shapeName=base_robeEnd_r_1_capsule, rotation=<rot 0 -90 0>}]
 dyn_beltRope_1 [{
   shapeType=Box
   extentX=4.1727
   extentY=6.6349
   extentZ=1.2594
   shapeName=dyn_beltRope_1_box
   center=<2.5381 -0.3586 0.5894>
   rotation=<rot -3.6085 0.0001 0>
  }]
 dyn_beltRope_2 [{
   shapeType=Box
   extentX=4.637
   extentY=6.461
   extentZ=1.9842
   shapeName=dyn_beltRope_2_box
   center=<2.7672 -0.3356 0.0577>
   rotation=<rot -2.6866 0 0>
  }]
 dyn_beltRope_3 [{shapeType=Box, extentX=5.4877, extentY=6.4353, extentZ=1.6428, shapeName=dyn_beltRope_3_box, center=<3.0877 -0.3773 0.0289>}]
 dyn_robeEnd_l_1 [{
   shapeType=Capsule
   length=2.6103
   radius=2.0615
   shapeName=dyn_robeEnd_l_1_capsule
   center=<3.5801 0.1478 0.563>
   rotation=<rot 9.1924 -88.0192 -78.4487>
  }]
 dyn_robeEnd_l_2 [{
   shapeType=Capsule
   length=3.5855
   radius=1.9347
   shapeName=dyn_robeEnd_l_2_capsule
   center=<3.4929 0.0568 0.4189>
   rotation=<rot -1.2598 -174.9316 -85.488>
  }]
 dyn_robeEnd_r_1 [{
   shapeType=Capsule
   length=2.1412
   radius=2.6456
   shapeName=dyn_robeEnd_r_1_capsule
   center=<2.6501 -0.0299 0.3996>
   rotation=<rot 2.087 86.9158 -94.689>
  }]
 dyn_robeEnd_r_2 [{
   shapeType=Capsule
   length=3.4785
   radius=2.136
   shapeName=dyn_robeEnd_r_2_capsule
   center=<2.9159 0.0664 -0.1777>
   rotation=<rot 0 -5.5141 -89.3768>
  }]
 pelvis [{shapeType=Capsule, length=12.6204, radius=10.2578, shapeName=pelvis_capsule, center=<0 -0.8599 -0.1713>, rotation=<rot 0 -0.0002 9.8849>}]
 spine_01 [{
   shapeType=Capsule
   length=8.4598
   radius=7.4497
   shapeName=spine_01_capsule
   center=<1.3817 1.8615 -1.247>
   rotation=<rot 16.8561 -89.6387 0.8505>
  }]
 thigh_l [{
   shapeType=Capsule
   length=25.3732
   radius=9.7817
   shapeName=thigh_l_capsule
   center=<-14.0461 -0.5381 1.7759>
   rotation=<rot -88.2395 -0.0018 -5.0119>
  }]
}
constraints [
 {
  bone1Name=thigh_l
  bone2Name=pelvis
  swing1LimitDegrees=45
  swing1Motion=Free
  swing2LimitDegrees=45
  swing2Motion=Free
  twistLimitDegrees=45
  twistMotion=Free
 }
 {
  bone1Name=base_beltRope_1
  bone2Name=spine_01
  swing1LimitDegrees=45
  swing1Motion=Free
  swing2LimitDegrees=45
  swing2Motion=Free
  twistLimitDegrees=45
  twistMotion=Free
 }
 {bone1Name=dyn_beltRope_1, bone2Name=base_beltRope_1, swing1LimitDegrees=20, swing2LimitDegrees=20, twistLimitDegrees=10}
 {bone1Name=dyn_beltRope_2, bone2Name=dyn_beltRope_1, swing1LimitDegrees=20, swing2LimitDegrees=20, twistLimitDegrees=10}
 {bone1Name=dyn_beltRope_3, bone2Name=dyn_beltRope_2, swing1LimitDegrees=20, swing2LimitDegrees=20, twistLimitDegrees=10}
 {
  bone1Name=base_robeEnd_l_1
  bone2Name=spine_01
  swing1LimitDegrees=45
  swing1Motion=Free
  swing2LimitDegrees=45
  swing2Motion=Free
  twistLimitDegrees=45
  twistMotion=Free
 }
 {bone1Name=dyn_robeEnd_l_1, bone2Name=base_robeEnd_l_1, swing1LimitDegrees=20, swing2LimitDegrees=20, twistLimitDegrees=0}
 {bone1Name=dyn_robeEnd_l_2, bone2Name=dyn_robeEnd_l_1, swing1LimitDegrees=20, swing2LimitDegrees=20, twistLimitDegrees=0}
 {
  bone1Name=base_robeEnd_r_1
  bone2Name=spine_01
  swing1LimitDegrees=45
  swing1Motion=Free
  swing2LimitDegrees=45
  swing2Motion=Free
  twistLimitDegrees=45
  twistMotion=Free
 }
 {bone1Name=dyn_robeEnd_r_1, bone2Name=base_robeEnd_r_1, swing1LimitDegrees=20, swing2LimitDegrees=20, twistLimitDegrees=0}
 {bone1Name=dyn_robeEnd_r_2, bone2Name=dyn_robeEnd_r_1, swing1LimitDegrees=20, swing2LimitDegrees=20, twistLimitDegrees=0}
 {
  bone1Name=spine_01
  bone2Name=pelvis
  swing1LimitDegrees=45
  swing1Motion=Free
  swing2LimitDegrees=45
  swing2Motion=Free
  twistLimitDegrees=45
  twistMotion=Free
 }
]
