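# Physics asset for the left side of the big skirt on the F_MED_SplishSplash female player body.
# Three skirt chains (a, b, c) each link a small base capsule to two larger dyn segments, with zero twist and wider swing limits at the tip than at the root.
# Left thigh and calf capsules supply the leg's shape, joined by a constraint with free swing and twist.
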PhysicsAsset /BRCosmetics/Characters/Player/Female/Medium/Bodies/F_MED_SplishSplash/Meshes/F_MED_SplishSplash_Physics_SkirtBig_2
bodies {
 base_L_skirt_a_2 [{shapeType=Capsule, length=2, radius=1, shapeName=base_L_skirt_a_2_capsule}]
 base_L_skirt_b_2 [{shapeType=Capsule, length=2, radius=1, shapeName=base_L_skirt_b_2_capsule}]
 base_L_skirt_c_2 [{shapeType=Capsule, length=2, radius=1, shapeName=base_L_skirt_c_2_capsule}]
 calf_l [{shapeType=Capsule, length=39.7913, radius=8.2847, shapeName=calf_l_capsule, center=<-23.8013 -2.3204 0>, rotation=<rot -90 180 180>}]
 dyn_L_skirt_a_2 [
  {shapeType=Capsule, length=7.3, radius=6.8876, shapeName=dyn_C_skirt_a_2_capsule, center=<7.38 2.0024 6.2681>, rotation=<rot -90 0 0>}
  {shapeType=Capsule, length=7.3, radius=7.1722, shapeName=dyn_C_skirt_a_2_capsule, center=<7.5 -9.1045 3.0373>, rotation=<rot -90 0 0>}
 ]
 dyn_L_skirt_a_3 [{
   shapeType=Capsule
   length=8.7437
   radius=3.8705
   shapeName=dyn_R_skirt_a_3_capsule
   center=<6.8938 0.7818 2.3717>
   rotation=<rot 0 -4.5 -82.9>
  }]
 dyn_L_skirt_b_2 [{shapeType=Capsule, length=8.8533, radius=5.6467, shapeName=dyn_C_skirt_a_2_capsule, center=<7.5 -1.2 4.8767>, rotation=<rot 0 0 -90>}]
 dyn_L_skirt_b_3 [{shapeType=Capsule, length=10.55, radius=4, shapeName=dyn_C_skirt_a_2_capsule, center=<2.1 -1.2 3.5576>, rotation=<rot 0 0 -90>}]
 dyn_L_skirt_c_2 [
  {shapeType=Capsule, length=7, radius=4, shapeName=dyn_C_skirt_a_2_capsule, center=<7.5 4.75 1>, rotation=<rot -90 0 0>}
  {shapeType=Capsule, length=7.3939, radius=4.7279, shapeName=dyn_C_skirt_a_2_capsule, center=<7.5 -3.2 1.3093>, rotation=<rot -90 0 0>}
 ]
 dyn_L_skirt_c_3 [{shapeType=Capsule, length=18, radius=4, shapeName=dyn_C_skirt_c_3_capsule, center=<5.55 0 1.6>, rotation=<rot 0 0 -88>}]
 thigh_l [
  {
   shapeType=Capsule
   length=18.6599
   radius=8.361
   shapeName=thigh_l_capsule
   center=<-26.057 0.4079 2.4328>
   rotation=<rot 88.7672 177.9805 177.7048>
  }
  {
   shapeType=Capsule
   length=19.5414
   radius=7.4795
   shapeName=thigh_l_capsule
   center=<-26.0155 -1.0069 0.8197>
   rotation=<rot 88.7672 177.9805 177.7048>
  }
 ]
}
constraints [
 {bone1Name=dyn_L_skirt_a_2, bone2Name=base_L_skirt_a_2, swing1LimitDegrees=5, swing2LimitDegrees=22, twistLimitDegrees=0}
 {bone1Name=dyn_L_skirt_a_3, bone2Name=dyn_L_skirt_a_2, swing1LimitDegrees=15, swing2LimitDegrees=35, twistLimitDegrees=0}
 {bone1Name=dyn_L_skirt_b_2, bone2Name=base_L_skirt_b_2, swing1LimitDegrees=5, swing2LimitDegrees=22, twistLimitDegrees=0}
 {bone1Name=dyn_L_skirt_b_3, bone2Name=dyn_L_skirt_b_2, swing1LimitDegrees=15, swing2LimitDegrees=35, twistLimitDegrees=0}
 {bone1Name=dyn_L_skirt_c_3, bone2Name=dyn_L_skirt_c_2, swing1LimitDegrees=15, swing2LimitDegrees=35, twistLimitDegrees=0}
 {bone1Name=dyn_L_skirt_c_2, bone2Name=base_L_skirt_c_2, swing1LimitDegrees=5, swing2LimitDegrees=22, twistLimitDegrees=0}
 {
  bone1Name=calf_l
  bone2Name=thigh_l
  swing1LimitDegrees=45
  swing1Motion=Free
  swing2LimitDegrees=45
  swing2Motion=Free
  twistLimitDegrees=45
  twistMotion=Free
 }
]
