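# Post-process style Anim Blueprint for the F_MED_SplishSplash cosmetic body, adding secondary motion to skirt, sleeves, duck grenades, slipper puffs and axolotl plushie.
# Copies the attached parent's pose into a state-driven Control Rig, then chains cached poses through bone tweaks, AnimDynamics and switchable skirt RigidBody sims.
# Post-sim fixes precede plushie constraints, Look Ats and its RigidBody.
# The update event computes gravity override and player movement.
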AnimBlueprint /BRCosmetics/Characters/Player/Female/Medium/Bodies/F_MED_SplishSplash/Meshes/F_MED_SplishSplash_AnimBP
target_skeleton=&/F_MED_SplishSplash_Skeleton
graphs [
 AnimGraph {nodes [
   RigidBody!(ComponentPose, Alpha, OverrideWorldGravity:Vector="0,0,0") -> (Pose) {
    bOverrideWorldGravity=true
    overridePhysicsAsset=&/F_MED_SplishSplash_Physics_Ropes
    simulationSpace=BaseBoneSpace
    alphaBoolBlend {blendOption=Linear}
    baseBoneRef {boneName=spine_02}
    componentAppliedLinearAccClamp=<3000 3000 3000>
    simSpaceSettings {
     maxAngularAcceleration=1
     maxAngularVelocity=3
     maxLinearAcceleration=800
     maxLinearVelocity=400
     worldAlpha=1
     externalLinearDragV=<0.2 0.2 0.2>
    }
   }
   Output Pose(Result=Use cached pose 'plushieCR'()) {layerGroup=None}
   n1 = Control Rig!(
    Source=Copy Pose From Mesh() {bUseAttachedParent=true}
    Alpha
    pub isKneeSliding:bool=Get bIsKneeSliding()
    pub isBaseballSliding:bool=Get bIsBaseballSliding()
    pub strafeLeft:double=Get strafeLeft float() -> (strafeLeft float:double)
    pub strafeRight:double=Get strafeRight float() -> (strafeRight float:double)
    pub moveForward:double=Get moveForward() -> (moveForward:double)
    pub moveBackward:double=Get moveBackward() -> (moveBackward:double)
    pub isDBNO:bool=Get bIsDBNO()
    pub isDBNO Carried:bool=Get bIsDBNOCarried()
    pub isCrouching:bool=Get bIsCrouching()
    pub isZipline:bool=Get isInZipline() -> (isInZipline:bool)
    pub isGrindRail:bool=Get isOnGrindRail() -> (isOnGrindRail:bool)
    pub isPlayingEasternBloc:bool=Get isPlayingEasterBloc() -> (isPlayingEasterBloc:bool)
    pub isPlayingEmote:bool=Get bIsPlayingEmote()
    pub isRBANOn:bool=Get bRigidBodyEnabled()
   ) -> (Pose) {
    alphaBoolBlend {blendOption=Linear}
    controlRigAssetReference {blueprintRigClass=&/F_MED_SplishSplash_CtrlRig.F_MED_SplishSplash_CtrlRig_C}
   }
   Save cached pose 'controlRigCP'(Pose=@n1) {cacheName=controlRigCP}
   n2 = RigidBody(ComponentPose=Local To Component(LocalPose=Use cached pose 'animDynamicsCP'()), OverrideWorldGravity="0,0,0") {
    bOverrideWorldGravity=true
    overridePhysicsAsset=&/F_MED_SplishSplash_Physics_SkirtBig_2
    simulationSpace=BaseBoneSpace
    alphaBoolBlend {blendOption=Linear}
    alphaScaleBiasClamp {bInterpResult=true, interpSpeedDecreasing=2, interpSpeedIncreasing=2}
    baseBoneRef {boneName=thigh_l}
    componentAppliedLinearAccClamp=<3000 3000 3000>
    simSpaceSettings {
     maxAngularAcceleration=1
     maxAngularVelocity=3
     maxLinearAcceleration=800
     maxLinearVelocity=400
     worldAlpha=1
     externalLinearDragV=<0.2 0.2 0.2>
    }
   }
   n3 = RigidBody(ComponentPose=@n2, OverrideWorldGravity="0,0,0") {
    bOverrideWorldGravity=true
    overridePhysicsAsset=&/F_MED_SplishSplash_Physics_SkirtBig
    simulationSpace=BaseBoneSpace
    alphaBoolBlend {blendOption=Linear}
    alphaScaleBiasClamp {bInterpResult=true, interpSpeedDecreasing=2, interpSpeedIncreasing=2}
    baseBoneRef {boneName=root}
    componentAppliedLinearAccClamp=<3000 3000 3000>
    simSpaceSettings {
     maxAngularAcceleration=1
     maxAngularVelocity=3
     maxLinearAcceleration=800
     maxLinearVelocity=400
     worldAlpha=1
     externalLinearDragV=<0.2 0.2 0.2>
    }
   }
   n4 = RigidBody(ComponentPose=Local To Component(LocalPose=Use cached pose 'animDynamicsCP'()), OverrideWorldGravity="0,0,0") {
    bOverrideWorldGravity=true
    overridePhysicsAsset=&/F_MED_SplishSplash_Physics
    simulationSpace=BaseBoneSpace
    alphaBoolBlend {blendOption=Linear}
    alphaScaleBiasClamp {bInterpResult=true, interpSpeedDecreasing=2, interpSpeedIncreasing=2}
    baseBoneRef {boneName=root}
    componentAppliedLinearAccClamp=<3000 3000 3000>
    simSpaceSettings {
     maxAngularAcceleration=1
     maxAngularVelocity=3
     maxLinearAcceleration=800
     maxLinearVelocity=400
     worldAlpha=1
     externalLinearDragV=<0.2 0.2 0.2>
    }
   }
   Save cached pose 'dynamicsCP'(Pose=Blend Poses by bool(BlendPose_0=Component To Local(ComponentPose=@n4), BlendPose_1=Component To Local(ComponentPose=@n3)) {blendType=Linear}) {cacheName=dynamicsCP}
   n5 = "Transform (Modify) Bone - Bone: base_L_skirt_c_1"!(ComponentPose=Local To Component(LocalPose=Use cached pose 'controlRigCP'()), Alpha=1, Translation="0,0,2", Rotation="30,0,0") -> (Pose) {
    rotationMode=BMM_Additive
    rotationSpace=BCS_BoneSpace
    translationMode=BMM_Additive
    translationSpace=BCS_BoneSpace
    alphaBoolBlend {blendOption=Linear}
    alphaScaleBiasClamp {bInterpResult=true}
    boneToModify {boneName=base_L_skirt_c_1}
    rotation {pitch=30}
    translation {z=2}
   }
   n6 = "Transform (Modify) Bone - Bone: base_C_skirt_b_1"!(ComponentPose=@n5, Alpha=1, Translation="0,0,2", Rotation="30,0,0") -> (Pose) {
    rotationMode=BMM_Additive
    rotationSpace=BCS_BoneSpace
    translationMode=BMM_Additive
    translationSpace=BCS_BoneSpace
    alphaBoolBlend {blendOption=Linear}
    alphaScaleBiasClamp {bInterpResult=true}
    boneToModify {boneName=base_C_skirt_b_1}
    rotation {pitch=30}
    translation {z=2}
   }
   n7 = "Transform (Modify) Bone - Bone: base_R_skirt_c_1"!(ComponentPose=@n6, Alpha=1, Translation="0,0,2", Rotation="30,0,0") -> (Pose) {
    rotationMode=BMM_Additive
    rotationSpace=BCS_BoneSpace
    translationMode=BMM_Additive
    translationSpace=BCS_BoneSpace
    alphaBoolBlend {blendOption=Linear}
    alphaScaleBiasClamp {bInterpResult=true}
    boneToModify {boneName=base_R_skirt_c_1}
    rotation {pitch=30}
    translation {z=2}
   }
   n8 = "Transform (Modify) Bone - Bone: base_L_skirt_a_1"!(ComponentPose=@n7, Alpha=1, Translation="0,0,0", Rotation="10,0,0") -> (Pose) {
    rotationMode=BMM_Additive
    rotationSpace=BCS_BoneSpace
    translationMode=BMM_Additive
    translationSpace=BCS_BoneSpace
    alphaBoolBlend {blendOption=Linear}
    alphaScaleBiasClamp {bInterpResult=true}
    boneToModify {boneName=base_L_skirt_a_1}
    rotation {pitch=10}
   }
   n9 = "Transform (Modify) Bone - Bone: base_C_skirt_a_1"!(ComponentPose=@n8, Alpha=1, Translation="0,0,0", Rotation="10,0,0") -> (Pose) {
    rotationMode=BMM_Additive
    rotationSpace=BCS_BoneSpace
    translationMode=BMM_Additive
    translationSpace=BCS_BoneSpace
    alphaBoolBlend {blendOption=Linear}
    alphaScaleBiasClamp {bInterpResult=true}
    boneToModify {boneName=base_C_skirt_a_1}
    rotation {pitch=10}
   }
   n10 = "Transform (Modify) Bone - Bone: base_R_skirt_a_1"!(ComponentPose=@n9, Alpha=1, Translation="0,0,0", Rotation="10,0,0") -> (Pose) {
    rotationMode=BMM_Additive
    rotationSpace=BCS_BoneSpace
    translationMode=BMM_Additive
    translationSpace=BCS_BoneSpace
    alphaBoolBlend {blendOption=Linear}
    alphaScaleBiasClamp {bInterpResult=true}
    boneToModify {boneName=base_R_skirt_a_1}
    rotation {pitch=10}
   }
   Save cached pose 'preSimTweaks'(Pose=Component To Local(ComponentPose=@n10)) {cacheName=preSimTweaks}
   n11 = "Constraint - attach_cape (P)"(ComponentPose=Local To Component(LocalPose=Use cached pose 'preSimTweaks'()), Alpha=1, ConstraintWeights_0=1) {
    alphaBoolBlend {blendOption=Linear}
    boneToModify {boneName=attach_cape}
    constraintSetup [{targetBone {boneName=spine_05}}]
    constraintWeights [1]
   }
   n12 = "Transform (Modify) Bone - Bone: attach_cape"!(ComponentPose=@n11, bAlphaBoolEnabled:bool=True, Translation="0,-1,0", Rotation="0,0,16") -> (Pose) {
    alphaInputType=Bool
    rotationMode=BMM_Additive
    rotationSpace=BCS_BoneSpace
    translationMode=BMM_Additive
    translationSpace=BCS_BoneSpace
    alphaBoolBlend {blendInTime=0.5, blendOption=CircularOut, blendOutTime=0.5}
    boneToModify {boneName=attach_cape}
    rotation {roll=16}
    translation {y=-1}
   }
   n13 = "Transform (Modify) Bone - Bone: attach_cape"(ComponentPose=@n12, bAlphaBoolEnabled=True, Translation="0,-2,0", Rotation="0,0,20") {
    alphaInputType=Bool
    rotationMode=BMM_Additive
    rotationSpace=BCS_BoneSpace
    translationMode=BMM_Additive
    translationSpace=BCS_BoneSpace
    alphaBoolBlend {blendInTime=0.5, blendOption=CircularOut, blendOutTime=0.5}
    boneToModify {boneName=attach_cape}
    rotation {roll=20}
    translation {y=-2}
   }
   n14 = "Transform (Modify) Bone - Bone: attach_cape"(ComponentPose=@n13, bAlphaBoolEnabled=True, Translation="0,0,0", Rotation="0,0,15") {
    alphaInputType=Bool
    rotationMode=BMM_Additive
    rotationSpace=BCS_BoneSpace
    translationMode=BMM_Additive
    translationSpace=BCS_BoneSpace
    alphaBoolBlend {blendInTime=0.5, blendOption=Cubic, blendOutTime=0.5}
    boneToModify {boneName=attach_cape}
    rotation {roll=15}
   }
   Save cached pose 'generalAdjuistementsCP'(Pose=Component To Local(ComponentPose=@n14)) {cacheName=generalAdjuistementsCP}
   n15 = Anim Dynamics - Bone: dyn_sleeve_l!(ComponentPose=Local To Component(LocalPose=Use cached pose 'generalAdjuistementsCP'()), Alpha=0.7, GravityOverride:Vector="0,0,0") -> (Pose) {
    angularDampingOverride=1
    angularSpringConstant=800
    bAngularSpring=true
    bOverrideLinearDamping=true
    bUseGravityOverride=true
    gravityScale=0
    linearDampingOverride=1
    simulationSpace=RootRelative
    alphaBoolBlend {blendOption=Linear}
    boundBone {boneName=dyn_sleeve_l}
    componentAppliedLinearAccClamp=<3000 3000 3000>
    componentLinearAccScale=<0.2 0.2 0.1>
    componentLinearVelScale=<0.2 0.2 0.1>
    physicsBodyDefinitions [{
      boundBone {boneName=dyn_sleeve_l}
      boxExtents=<25 25 25>
      constraintSetup {angularLimitsMax=<0 4 3>, angularLimitsMin=<0 -4 -3>, angularTarget=<1 0 0>}
      localJointOffset=<8 0 0>
     }]
    relativeSpaceBone {boneName=lowerarm_l}
   }
   n16 = Anim Dynamics - Bone: dyn_sleeve_r!(ComponentPose=@n15, Alpha=0.7, GravityOverride:Vector="0,0,0") -> (Pose) {
    angularDampingOverride=1
    angularSpringConstant=800
    bAngularSpring=true
    bOverrideLinearDamping=true
    bUseGravityOverride=true
    gravityScale=0
    linearDampingOverride=1
    simulationSpace=RootRelative
    alphaBoolBlend {blendOption=Linear}
    boundBone {boneName=dyn_sleeve_r}
    componentAppliedLinearAccClamp=<3000 3000 3000>
    componentLinearAccScale=<0.2 0.2 0.1>
    componentLinearVelScale=<0.2 0.2 0.1>
    physicsBodyDefinitions [{
      boundBone {boneName=dyn_sleeve_r}
      boxExtents=<25 25 25>
      constraintSetup {angularLimitsMax=<0 4 3>, angularLimitsMin=<0 -4 -3>, angularTarget=<1 0 0>}
      localJointOffset=<8 0 0>
     }]
    relativeSpaceBone {boneName=lowerarm_l}
   }
   n17 = Anim Dynamics - Bone: dyn_duckNade_A!(ComponentPose=@n16, Alpha=1, GravityOverride:Vector="0,0,-980") -> (Pose) {
    angularDampingOverride=0.9999
    angularSpringConstant=8000
    bAngularSpring=true
    bOverrideAngularDamping=true
    bUseGravityOverride=true
    bUseSphericalLimits=true
    gravityScale=0.35
    simulationSpace=RootRelative
    alphaBoolBlend {blendOption=Linear}
    boundBone {boneName=dyn_duckNade_A}
    componentAppliedLinearAccClamp=<5000 5000 5000>
    componentLinearAccScale=<1 1 1>
    componentLinearVelScale=<1 1 1>
    gravityOverride {z=-980}
    physicsBodyDefinitions [{
      boundBone {boneName=dyn_duckNade_A}
      boxExtents=<20 20 20>
      constraintSetup {angularLimitsMax=<0 10 15>, angularLimitsMin=<0 -50 0>, angularTarget=<1 0 0>}
      localJointOffset=<3 0 0>
     }]
    relativeSpaceBone {boneName=pelvis}
    simSpaceSettings {maxAngularAcceleration=30, maxAngularVelocity=5, simSpaceAngularAlpha=1}
    sphericalLimits [{limitRadius=18.25, drivingBone {boneName=spine_01}, sphereLocalOffset=<-8.42 0.7 -0.5>}]
   }
   n18 = Anim Dynamics - Bone: dyn_duckNade_B!(ComponentPose=@n17, Alpha=1, GravityOverride:Vector="0,0,-980") -> (Pose) {
    angularDampingOverride=0.9999
    angularSpringConstant=8000
    bAngularSpring=true
    bOverrideAngularDamping=true
    bUseGravityOverride=true
    bUseSphericalLimits=true
    gravityScale=0.35
    simulationSpace=RootRelative
    alphaBoolBlend {blendOption=Linear}
    boundBone {boneName=dyn_duckNade_B}
    componentAppliedLinearAccClamp=<5000 5000 5000>
    componentLinearAccScale=<1 1 1>
    componentLinearVelScale=<1 1 1>
    gravityOverride {z=-980}
    physicsBodyDefinitions [{
      boundBone {boneName=dyn_duckNade_B}
      boxExtents=<20 20 20>
      constraintSetup {angularLimitsMax=<0 10 0>, angularLimitsMin=<0 -50 -15>, angularTarget=<1 0 0>}
      localJointOffset=<3 0 0>
     }]
    relativeSpaceBone {boneName=pelvis}
    simSpaceSettings {maxAngularAcceleration=30, maxAngularVelocity=5, simSpaceAngularAlpha=1}
    sphericalLimits [{limitRadius=18.25, drivingBone {boneName=spine_01}, sphereLocalOffset=<-8.42 2 -0.5>}]
   }
   n19 = Anim Dynamics - Bone: dyn_duckNade_ring_A!(ComponentPose=@n18, Alpha=1) -> (Pose) {
    angularDampingOverride=1
    angularSpringConstant=100
    bAngularSpring=true
    bOverrideAngularDamping=true
    bUseGravityOverride=true
    gravityScale=0.35
    simulationSpace=BoneRelative
    alphaBoolBlend {blendOption=Linear}
    boundBone {boneName=dyn_duckNade_ring_A}
    componentAppliedLinearAccClamp=<3000 3000 3000>
    componentLinearAccScale=<0.2 0.2 0.2>
    componentLinearVelScale=<0.2 0.2 0.2>
    gravityOverride {z=-980}
    physicsBodyDefinitions [{
      boundBone {boneName=dyn_duckNade_ring_A}
      boxExtents=<10 10 10>
      constraintSetup {angularLimitsMax=<0 0 30>, angularLimitsMin=<0 0 -30>, angularTarget=<1 0 0>}
      localJointOffset=<2 0 0>
     }]
    simSpaceSettings {maxAngularAcceleration=36, maxAngularVelocity=6, simSpaceAngularAlpha=1}
   }
   n20 = Anim Dynamics - Bone: dyn_duckNade_ring_B!(ComponentPose=@n19, Alpha=1) -> (Pose) {
    angularDampingOverride=1
    angularSpringConstant=100
    bAngularSpring=true
    bOverrideAngularDamping=true
    bUseGravityOverride=true
    gravityScale=0.35
    simulationSpace=BoneRelative
    alphaBoolBlend {blendOption=Linear}
    boundBone {boneName=dyn_duckNade_ring_B}
    componentAppliedLinearAccClamp=<3000 3000 3000>
    componentLinearAccScale=<0.2 0.2 0.2>
    componentLinearVelScale=<0.2 0.2 0.2>
    gravityOverride {z=-980}
    physicsBodyDefinitions [{
      boundBone {boneName=dyn_duckNade_ring_B}
      boxExtents=<10 10 10>
      constraintSetup {angularLimitsMax=<0 0 30>, angularLimitsMin=<0 0 -30>, angularTarget=<1 0 0>}
      localJointOffset=<2 0 0>
     }]
    simSpaceSettings {maxAngularAcceleration=36, maxAngularVelocity=6, simSpaceAngularAlpha=1}
   }
   n21 = Anim Dynamics - Bone: dyn_slipperPuff_l!(ComponentPose=@n20, Alpha=1) -> (Pose) {
    angularDampingOverride=1
    angularSpringConstant=1000
    bAngularSpring=true
    bOverrideAngularDamping=true
    bUseGravityOverride=true
    gravityScale=0.35
    simulationSpace=RootRelative
    alphaBoolBlend {blendOption=Linear}
    boundBone {boneName=dyn_slipperPuff_l}
    componentAppliedLinearAccClamp=<5000 5000 5000>
    componentLinearAccScale=<1 1 1>
    componentLinearVelScale=<1 1 1>
    gravityOverride {z=-980}
    physicsBodyDefinitions [{
      boundBone {boneName=dyn_slipperPuff_l}
      boxExtents=<20 30 30>
      constraintSetup {angularLimitsMax=<0 15 15>, angularLimitsMin=<0 -15 -15>, angularTarget=<1 0 0>}
      localJointOffset=<4 0 0>
     }]
   }
   n22 = Anim Dynamics - Bone: dyn_slipperPuff_r!(ComponentPose=@n21, Alpha=1) -> (Pose) {
    angularDampingOverride=1
    angularSpringConstant=1000
    bAngularSpring=true
    bOverrideAngularDamping=true
    bUseGravityOverride=true
    gravityScale=0.35
    simulationSpace=RootRelative
    alphaBoolBlend {blendOption=Linear}
    boundBone {boneName=dyn_slipperPuff_r}
    componentAppliedLinearAccClamp=<5000 5000 5000>
    componentLinearAccScale=<1 1 1>
    componentLinearVelScale=<1 1 1>
    gravityOverride {z=-980}
    physicsBodyDefinitions [{
      boundBone {boneName=dyn_slipperPuff_r}
      boxExtents=<20 30 30>
      constraintSetup {angularLimitsMax=<0 15 15>, angularLimitsMin=<0 -15 -15>, angularTarget=<1 0 0>}
      localJointOffset=<4 0 0>
     }]
   }
   Save cached pose 'animDynamicsCP'(Pose=Component To Local(ComponentPose=@n22)) {cacheName=animDynamicsCP}
   n23 = "Transform (Modify) Bone - Bone: dyn_L_skirt_a_3"!(ComponentPose=Local To Component(LocalPose=Use cached pose 'postSimFixes'()), Alpha=1, Rotation="0,0,0") -> (Pose) {rotationMode=BMM_Additive, rotationSpace=BCS_BoneSpace, alphaBoolBlend {blendOption=Linear}, boneToModify {boneName=dyn_L_skirt_a_3}}
   n24 = "Transform (Modify) Bone - Bone: dyn_L_skirt_b_3"!(ComponentPose=@n23, Alpha=1, Rotation="0,0,0") -> (Pose) {rotationMode=BMM_Additive, rotationSpace=BCS_BoneSpace, alphaBoolBlend {blendOption=Linear}, boneToModify {boneName=dyn_L_skirt_b_3}}
   n25 = "Transform (Modify) Bone - Bone: dyn_L_skirt_c_3"!(ComponentPose=@n24, Alpha=1, Rotation="0,0,0") -> (Pose) {rotationMode=BMM_Additive, rotationSpace=BCS_BoneSpace, alphaBoolBlend {blendOption=Linear}, boneToModify {boneName=dyn_L_skirt_c_3}}
   n26 = "Transform (Modify) Bone - Bone: dyn_R_skirt_a_3"!(ComponentPose=@n25, Alpha=1, Rotation="0,0,0") -> (Pose) {rotationMode=BMM_Additive, rotationSpace=BCS_BoneSpace, alphaBoolBlend {blendOption=Linear}, boneToModify {boneName=dyn_R_skirt_a_3}}
   n27 = "Transform (Modify) Bone - Bone: dyn_R_skirt_b_3"!(ComponentPose=@n26, Alpha=1, Rotation="0,0,0") -> (Pose) {rotationMode=BMM_Additive, rotationSpace=BCS_BoneSpace, alphaBoolBlend {blendOption=Linear}, boneToModify {boneName=dyn_R_skirt_b_3}}
   n28 = "Transform (Modify) Bone - Bone: dyn_R_skirt_c_3"!(ComponentPose=@n27, Alpha=1, Rotation="0,0,0") -> (Pose) {rotationMode=BMM_Additive, rotationSpace=BCS_BoneSpace, alphaBoolBlend {blendOption=Linear}, boneToModify {boneName=dyn_R_skirt_c_3}}
   n29 = "Transform (Modify) Bone - Bone: dyn_C_skirt_a_3"!(ComponentPose=@n28, Alpha=1, Rotation="0,0,0") -> (Pose) {rotationMode=BMM_Additive, rotationSpace=BCS_BoneSpace, alphaBoolBlend {blendOption=Linear}, boneToModify {boneName=dyn_C_skirt_a_3}}
   n30 = "Transform (Modify) Bone - Bone: dyn_C_skirt_b_3"!(ComponentPose=@n29, Alpha=1, Rotation="0,0,0") -> (Pose) {rotationMode=BMM_Additive, rotationSpace=BCS_BoneSpace, alphaBoolBlend {blendOption=Linear}, boneToModify {boneName=dyn_C_skirt_b_3}}
   Save cached pose 'fakeWindFXCR'(Pose=Blend Poses by bool(BlendPose_0=Component To Local(ComponentPose=@n30), BlendPose_1=Use cached pose 'postSimFixes'()) {blendType=Linear}) {cacheName=fakeWindFXCR}
   n31 = "Constraint - kin_plushie (PPPPPPPP)"!(
    ComponentPose=Local To Component(LocalPose=Use cached pose 'fakeWindFXCR'())
    Alpha=1
    ConstraintWeights_0=0.1978
    ConstraintWeights_1:float=0.1927
    ConstraintWeights_2:float=0.1849
    ConstraintWeights_3:float=0.1141
    ConstraintWeights_4:float=0.1009
    ConstraintWeights_5:float=0.0895
    ConstraintWeights_6:float=0.0837
    ConstraintWeights_7:float=0.0364
   ) -> (Pose) {
    alphaBoolBlend {blendOption=Linear}
    boneToModify {boneName=kin_plushie}
    constraintSetup [
     {targetBone {boneName=base_R_skirt_b_1}}
     {targetBone {boneName=front_thigh_skirt_r}}
     {targetBone {boneName=base_R_skirt_a_1}}
     {targetBone {boneName=deform_groin_r}}
     {targetBone {boneName=pelvis}}
     {targetBone {boneName=dyn_R_skirt_b_2}}
     {targetBone {boneName=dyn_R_skirt_a_2}}
     {targetBone {boneName=base_C_skirt_a_1}}
    ]
    constraintWeights [1,1,1,1,1,1,1,1]
   }
   n32 = "Constraint - base_plushieHead (PPPPPPPP)"!(
    ComponentPose=@n31
    Alpha=1
    ConstraintWeights_0=0.3786
    ConstraintWeights_1:float=0.1642
    ConstraintWeights_2:float=0.1018
    ConstraintWeights_3:float=0.1015
    ConstraintWeights_4:float=0.1011
    ConstraintWeights_5:float=0.0711
    ConstraintWeights_6:float=0.0481
    ConstraintWeights_7:float=0.0336
   ) -> (Pose) {
    alphaBoolBlend {blendOption=Linear}
    boneToModify {boneName=base_plushieHead}
    constraintSetup [
     {targetBone {boneName=kin_plushie}}
     {targetBone {boneName=deform_groin_r}}
     {targetBone {boneName=pelvis}}
     {targetBone {boneName=base_R_skirt_a_1}}
     {targetBone {boneName=front_thigh_skirt_r}}
     {targetBone {boneName=base_R_skirt_b_1}}
     {targetBone {boneName=dyn_R_skirt_a_2}}
     {targetBone {boneName=dyn_R_skirt_b_2}}
    ]
    constraintWeights [1,1,1,1,1,1,1,1]
   }
   n33 = "Constraint - coll_plushieArms_1 (PPPPPPPP)"!(
    ComponentPose=@n32
    Alpha=1
    ConstraintWeights_0=0.2777
    ConstraintWeights_1:float=0.2256
    ConstraintWeights_2:float=0.1718
    ConstraintWeights_3:float=0.0899
    ConstraintWeights_4:float=0.0685
    ConstraintWeights_5:float=0.0618
    ConstraintWeights_6:float=0.0576
    ConstraintWeights_7:float=0.0471
   ) -> (Pose) {
    alphaBoolBlend {blendOption=Linear}
    boneToModify {boneName=coll_plushieArms_1}
    constraintSetup [
     {targetBone {boneName=kin_plushie}}
     {targetBone {boneName=base_R_skirt_b_1}}
     {targetBone {boneName=dyn_R_skirt_b_2}}
     {targetBone {boneName=base_R_skirt_a_1}}
     {targetBone {boneName=dyn_R_skirt_a_2}}
     {targetBone {boneName=base_R_skirt_c_1}}
     {targetBone {boneName=pelvis}}
     {targetBone {boneName=dyn_R_skirt_c_2}}
    ]
    constraintWeights [1,1,1,1,1,1,1,1]
   }
   n34 = "Constraint - coll_plushieArms_2 (PPPPPPPP)"!(
    ComponentPose=@n33
    Alpha=1
    ConstraintWeights_0=0.2084
    ConstraintWeights_1:float=0.1802
    ConstraintWeights_2:float=0.1735
    ConstraintWeights_3:float=0.1599
    ConstraintWeights_4:float=0.1114
    ConstraintWeights_5:float=0.0699
    ConstraintWeights_6:float=0.0491
    ConstraintWeights_7:float=0.0475
   ) -> (Pose) {
    alphaBoolBlend {blendOption=Linear}
    boneToModify {boneName=coll_plushieArms_2}
    constraintSetup [
     {targetBone {boneName=base_R_skirt_a_1}}
     {targetBone {boneName=deform_groin_r}}
     {targetBone {boneName=kin_plushie}}
     {targetBone {boneName=front_thigh_skirt_r}}
     {targetBone {boneName=pelvis}}
     {targetBone {boneName=dyn_R_skirt_a_2}}
     {targetBone {boneName=base_C_skirt_a_1}}
     {targetBone {boneName=dyn_C_skirt_a_2}}
    ]
    constraintWeights [1,1,1,1,1,1,1,1]
   }
   n35 = "Constraint - coll_plushieArms_3 (PPPPPPPP)"!(
    ComponentPose=@n34
    Alpha=1
    ConstraintWeights_0=0.539
    ConstraintWeights_1:float=0.0921
    ConstraintWeights_2:float=0.0849
    ConstraintWeights_3:float=0.0705
    ConstraintWeights_4:float=0.065
    ConstraintWeights_5:float=0.0639
    ConstraintWeights_6:float=0.0554
    ConstraintWeights_7:float=0.0292
   ) -> (Pose) {
    alphaBoolBlend {blendOption=Linear}
    boneToModify {boneName=coll_plushieArms_3}
    constraintSetup [
     {targetBone {boneName=kin_plushie}}
     {targetBone {boneName=base_R_skirt_b_1}}
     {targetBone {boneName=base_R_skirt_a_1}}
     {targetBone {boneName=dyn_R_skirt_b_2}}
     {targetBone {boneName=dyn_R_skirt_a_2}}
     {targetBone {boneName=deform_groin_r}}
     {targetBone {boneName=pelvis}}
     {targetBone {boneName=front_thigh_skirt_r}}
    ]
    constraintWeights [1,1,1,1,1,1,1,1]
   }
   n36 = "Constraint - coll_plushieArms_4 (PPPPPPPP)"!(
    ComponentPose=@n35
    Alpha=1
    ConstraintWeights_0=0.4979
    ConstraintWeights_1:float=0.1363
    ConstraintWeights_2:float=0.0903
    ConstraintWeights_3:float=0.0725
    ConstraintWeights_4:float=0.0629
    ConstraintWeights_5:float=0.0557
    ConstraintWeights_6:float=0.0456
    ConstraintWeights_7:float=0.0388
   ) -> (Pose) {
    alphaBoolBlend {blendOption=Linear}
    boneToModify {boneName=coll_plushieArms_4}
    constraintSetup [
     {targetBone {boneName=kin_plushie}}
     {targetBone {boneName=deform_groin_r}}
     {targetBone {boneName=base_R_skirt_a_1}}
     {targetBone {boneName=pelvis}}
     {targetBone {boneName=dyn_R_skirt_a_2}}
     {targetBone {boneName=base_R_skirt_b_1}}
     {targetBone {boneName=front_thigh_skirt_r}}
     {targetBone {boneName=dyn_R_skirt_b_2}}
    ]
    constraintWeights [1,1,1,1,1,1,1,1]
   }
   n37 = "Constraint - coll_plushieArms_1_target (PPPPPPPP)"!(
    ComponentPose=@n36
    Alpha=1
    ConstraintWeights_0=0.4092
    ConstraintWeights_1:float=0.2589
    ConstraintWeights_2:float=0.1571
    ConstraintWeights_3:float=0.1067
    ConstraintWeights_4:float=0.028
    ConstraintWeights_5:float=0.0186
    ConstraintWeights_6:float=0.0114
    ConstraintWeights_7:float=0.0102
   ) -> (Pose) {
    alphaBoolBlend {blendOption=Linear}
    boneToModify {boneName=coll_plushieArms_1_target}
    constraintSetup [
     {targetBone {boneName=dyn_R_skirt_b_2}}
     {targetBone {boneName=kin_plushie}}
     {targetBone {boneName=dyn_R_skirt_a_2}}
     {targetBone {boneName=dyn_R_skirt_c_2}}
     {targetBone {boneName=base_R_skirt_b_1}}
     {targetBone {boneName=dyn_R_skirt_b_3}}
     {targetBone {boneName=dyn_C_skirt_a_2}}
     {targetBone {boneName=base_R_skirt_a_1}}
    ]
    constraintWeights [1,1,1,1,1,1,1,1]
   }
   n38 = "Constraint - coll_plushieArms_2_target (PPPPPPPP)"!(
    ComponentPose=@n37
    Alpha=1
    ConstraintWeights_0=0.3157
    ConstraintWeights_1:float=0.3058
    ConstraintWeights_2:float=0.2023
    ConstraintWeights_3:float=0.096
    ConstraintWeights_4:float=0.0244
    ConstraintWeights_5:float=0.023
    ConstraintWeights_6:float=0.0167
    ConstraintWeights_7:float=0.016
   ) -> (Pose) {
    alphaBoolBlend {blendOption=Linear}
    boneToModify {boneName=coll_plushieArms_2_target}
    constraintSetup [
     {targetBone {boneName=dyn_R_skirt_a_2}}
     {targetBone {boneName=kin_plushie}}
     {targetBone {boneName=dyn_C_skirt_a_2}}
     {targetBone {boneName=dyn_R_skirt_b_2}}
     {targetBone {boneName=deform_groin_r}}
     {targetBone {boneName=dyn_L_skirt_a_2}}
     {targetBone {boneName=base_R_skirt_a_1}}
     {targetBone {boneName=dyn_R_skirt_a_3}}
    ]
    constraintWeights [1,1,1,1,1,1,1,1]
   }
   n39 = "Constraint - coll_plushieArms_3_target (PPPPPPPP)"!(
    ComponentPose=@n38
    Alpha=1
    ConstraintWeights_0=0.4785
    ConstraintWeights_1:float=0.255
    ConstraintWeights_2:float=0.1645
    ConstraintWeights_3:float=0.0401
    ConstraintWeights_4:float=0.0224
    ConstraintWeights_5:float=0.0171
    ConstraintWeights_6:float=0.0113
    ConstraintWeights_7:float=0.011
   ) -> (Pose) {
    alphaBoolBlend {blendOption=Linear}
    boneToModify {boneName=coll_plushieArms_3_target}
    constraintSetup [
     {targetBone {boneName=kin_plushie}}
     {targetBone {boneName=dyn_R_skirt_b_2}}
     {targetBone {boneName=dyn_R_skirt_a_2}}
     {targetBone {boneName=dyn_R_skirt_c_2}}
     {targetBone {boneName=dyn_C_skirt_a_2}}
     {targetBone {boneName=base_R_skirt_b_1}}
     {targetBone {boneName=dyn_R_skirt_b_3}}
     {targetBone {boneName=base_R_skirt_a_1}}
    ]
    constraintWeights [1,1,1,1,1,1,1,1]
   }
   n40 = "Constraint - coll_plushieArms_4_target (PPPPPPPP)"!(
    ComponentPose=@n39
    Alpha=1
    ConstraintWeights_0=0.5181
    ConstraintWeights_1:float=0.2246
    ConstraintWeights_2:float=0.1104
    ConstraintWeights_3:float=0.0912
    ConstraintWeights_4:float=0.0243
    ConstraintWeights_5:float=0.0143
    ConstraintWeights_6:float=0.01
    ConstraintWeights_7:float=0.007
   ) -> (Pose) {
    alphaBoolBlend {blendOption=Linear}
    boneToModify {boneName=coll_plushieArms_4_target}
    constraintSetup [
     {targetBone {boneName=kin_plushie}}
     {targetBone {boneName=dyn_R_skirt_a_2}}
     {targetBone {boneName=dyn_R_skirt_b_2}}
     {targetBone {boneName=dyn_C_skirt_a_2}}
     {targetBone {boneName=deform_groin_r}}
     {targetBone {boneName=base_R_skirt_a_1}}
     {targetBone {boneName=dyn_R_skirt_a_3}}
     {targetBone {boneName=base_R_skirt_b_1}}
    ]
    constraintWeights [1,1,1,1,1,1,1,1]
   }
   n41 = Look At - Bone: coll_plushieArms_1(ComponentPose=@n40, Alpha=1) {
    alphaBoolBlend {blendOption=Linear}
    boneToModify {boneName=coll_plushieArms_1}
    lookAtLocation {x=0}
    lookAtTarget {boneReference {boneName=coll_plushieArms_1_target}}
    lookAt_Axis {axis=<1 0>}
    lookUp_Axis {axis {x=0, z=-1}}
   }
   n42 = Look At - Bone: coll_plushieArms_2(ComponentPose=@n41, Alpha=1) {
    alphaBoolBlend {blendOption=Linear}
    boneToModify {boneName=coll_plushieArms_2}
    lookAtLocation {x=0}
    lookAtTarget {boneReference {boneName=coll_plushieArms_2_target}}
    lookAt_Axis {axis=<1 0>}
    lookUp_Axis {axis {x=0, z=-1}}
   }
   n43 = Look At - Bone: coll_plushieArms_3(ComponentPose=@n42, Alpha=1) {
    alphaBoolBlend {blendOption=Linear}
    boneToModify {boneName=coll_plushieArms_3}
    lookAtLocation {x=0}
    lookAtTarget {boneReference {boneName=coll_plushieArms_3_target}}
    lookAt_Axis {axis=<1 0>}
    lookUp_Axis {axis {x=0, z=-1}}
   }
   n44 = Look At - Bone: coll_plushieArms_4(ComponentPose=@n43, Alpha=1) {
    alphaBoolBlend {blendOption=Linear}
    boneToModify {boneName=coll_plushieArms_4}
    lookAtLocation {x=0}
    lookAtTarget {boneReference {boneName=coll_plushieArms_4_target}}
    lookAt_Axis {axis=<1 0>}
    lookUp_Axis {axis {x=0, z=-1}}
   }
   n45 = "Transform (Modify) Bone - Bone: kin_plushie"(ComponentPose=@n44, Alpha=1, Translation="0,0,0", Rotation="-6,-16,50", Scale="1,1,1") {
    rotationMode=BMM_Additive
    rotationSpace=BCS_BoneSpace
    scaleMode=BMM_Additive
    scaleSpace=BCS_BoneSpace
    translationMode=BMM_Additive
    translationSpace=BCS_BoneSpace
    alphaBoolBlend {blendOption=Linear}
    boneToModify {boneName=kin_plushie}
    rotation=<rot -6 -16 50>
   }
   n46 = "Transform (Modify) Bone - Bone: kin_plushie"!(ComponentPose=@n45, bAlphaBoolEnabled:bool=True, Translation="0,0,1", Rotation="10,0,0", Scale="1,1,1") -> (Pose) {
    alphaInputType=Bool
    rotationMode=BMM_Additive
    rotationSpace=BCS_BoneSpace
    scaleMode=BMM_Additive
    scaleSpace=BCS_BoneSpace
    translationMode=BMM_Additive
    translationSpace=BCS_BoneSpace
    alphaBoolBlend {blendInTime=1, blendOption=Cubic, blendOutTime=1}
    boneToModify {boneName=kin_plushie}
    rotation {pitch=10}
    translation {z=1}
   }
   n47 = "Transform (Modify) Bone - Bone: kin_plushie"!(ComponentPose=@n46, Alpha:float=1, Translation="0,0,5", Rotation="20,0,0", Scale="1,1,1") -> (Pose) {
    rotationMode=BMM_Additive
    rotationSpace=BCS_BoneSpace
    scaleMode=BMM_Additive
    scaleSpace=BCS_BoneSpace
    translationMode=BMM_Additive
    translationSpace=BCS_BoneSpace
    alphaBoolBlend {blendOption=Cubic}
    boneToModify {boneName=kin_plushie}
    rotation {pitch=20}
    translation {z=5}
   }
   n48 = "Transform (Modify) Bone - Bone: kin_plushieHead"(ComponentPose=@n47, Alpha=1, Translation="0,0,0", Rotation="-12,0,0", Scale="1,1,1") {
    rotationMode=BMM_Additive
    rotationSpace=BCS_BoneSpace
    scaleMode=BMM_Additive
    scaleSpace=BCS_BoneSpace
    translationMode=BMM_Additive
    translationSpace=BCS_BoneSpace
    alphaBoolBlend {blendOption=Linear}
    boneToModify {boneName=kin_plushieHead}
    rotation {pitch=-12}
   }
   n49 = "Transform (Modify) Bone - Bone: kin_plushieHead"!(ComponentPose=@n48, bAlphaBoolEnabled:bool=True, Translation="0,0,0", Rotation="0,0,-13", Scale="1,1,1") -> (Pose) {
    alphaInputType=Bool
    rotationMode=BMM_Additive
    rotationSpace=BCS_BoneSpace
    scaleMode=BMM_Additive
    scaleSpace=BCS_BoneSpace
    translationMode=BMM_Additive
    translationSpace=BCS_BoneSpace
    alphaBoolBlend {blendInTime=2, blendOption=Linear, blendOutTime=2}
    boneToModify {boneName=kin_plushieHead}
    rotation {roll=-13}
   }
   n50 = "Transform (Modify) Bone - Bone: kin_plushieHead"!(ComponentPose=@n49, Alpha:float=1, Translation="0,0,0", Rotation="0,0,-20", Scale="1,1,1") -> (Pose) {
    rotationMode=BMM_Additive
    rotationSpace=BCS_BoneSpace
    scaleMode=BMM_Additive
    scaleSpace=BCS_BoneSpace
    translationMode=BMM_Additive
    translationSpace=BCS_BoneSpace
    alphaBoolBlend {blendOption=Linear}
    alphaScaleBiasClamp {bInterpResult=true, interpSpeedDecreasing=7, interpSpeedIncreasing=1}
    boneToModify {boneName=kin_plushieHead}
    rotation {roll=-20}
   }
   n51 = "Transform (Modify) Bone - Bone: kin_plushieHead"!(ComponentPose=@n50, bAlphaBoolEnabled:bool=True, Translation="0,0,0", Rotation="0,0,90", Scale="1,1,1") -> (Pose) {
    alphaInputType=Bool
    rotationMode=BMM_Additive
    rotationSpace=BCS_BoneSpace
    scaleMode=BMM_Additive
    scaleSpace=BCS_BoneSpace
    translationMode=BMM_Additive
    translationSpace=BCS_BoneSpace
    alphaBoolBlend {blendInTime=1, blendOption=Cubic, blendOutTime=1}
    boneToModify {boneName=kin_plushieHead}
    rotation {roll=90}
   }
   n52 = "Transform (Modify) Bone - Bone: base_plushieHead"!(ComponentPose=@n51, bAlphaBoolEnabled:bool=True, Translation="0,0,0", Rotation="0,0,-30", Scale="1,1,1") -> (Pose) {
    alphaInputType=Bool
    rotationMode=BMM_Additive
    rotationSpace=BCS_BoneSpace
    scaleMode=BMM_Additive
    scaleSpace=BCS_BoneSpace
    translationMode=BMM_Additive
    translationSpace=BCS_BoneSpace
    alphaBoolBlend {blendInTime=1, blendOption=Cubic, blendOutTime=1}
    boneToModify {boneName=base_plushieHead}
    rotation {roll=-30}
   }
   n53 = RigidBody(ComponentPose=@n52, OverrideWorldGravity="0,0,0") {
    bOverrideWorldGravity=true
    overridePhysicsAsset=&/F_MED_SplishSplash_Physics_Axolotl
    simulationSpace=BaseBoneSpace
    alphaBoolBlend {blendOption=Linear}
    baseBoneRef {boneName=base_R_skirt_a_1}
    componentAppliedLinearAccClamp=<3000 3000 3000>
    simSpaceSettings {
     maxAngularAcceleration=1
     maxAngularVelocity=3
     maxLinearAcceleration=1200
     maxLinearVelocity=400
     worldAlpha=1
     externalLinearDragV=<0.5 0.5 0.5>
    }
   }
   n54 = "Transform (Modify) Bone - Bone: dyn_plushieHead"!(ComponentPose=@n53, bAlphaBoolEnabled:bool=True, Translation="0,0,0", Rotation="0,0,0", Scale="0.88,0.88,0.88") -> (Pose) {
    alphaInputType=Bool
    rotationMode=BMM_Additive
    rotationSpace=BCS_BoneSpace
    scaleMode=BMM_Additive
    scaleSpace=BCS_BoneSpace
    translationMode=BMM_Additive
    translationSpace=BCS_BoneSpace
    alphaBoolBlend {blendInTime=0.35, blendOption=Cubic, blendOutTime=0.35}
    boneToModify {boneName=dyn_plushieHead}
    scale=<0.88 0.88 0.88>
   }
   Save cached pose 'plushieCR'(Pose=Component To Local(ComponentPose=@n54)) {cacheName=plushieCR}
   n55 = "Transform (Modify) Bone - Bone: dyn_L_skirt_c_3"!(
    ComponentPose=Local To Component(LocalPose=Use cached pose 'dynamicsCP'())
    bAlphaBoolEnabled:bool=True
    Translation:Vector="0,0,0"
    Rotation="30,0,0"
   ) -> (Pose) {
    alphaInputType=Bool
    rotationMode=BMM_Additive
    rotationSpace=BCS_BoneSpace
    alphaBoolBlend {blendInTime=2, blendOption=Cubic, blendOutTime=1}
    boneToModify {boneName=dyn_L_skirt_c_3}
    rotation {pitch=30}
   }
   n56 = "Transform (Modify) Bone - Bone: dyn_R_skirt_c_3"!(ComponentPose=@n55, bAlphaBoolEnabled:bool=True, Translation:Vector="0,0,0", Rotation="30,0,0") -> (Pose) {
    alphaInputType=Bool
    rotationMode=BMM_Additive
    rotationSpace=BCS_BoneSpace
    alphaBoolBlend {blendInTime=2, blendOption=Cubic, blendOutTime=1}
    boneToModify {boneName=dyn_R_skirt_c_3}
    rotation {pitch=30}
   }
   n57 = "Transform (Modify) Bone - Bone: dyn_C_skirt_b_3"!(ComponentPose=@n56, bAlphaBoolEnabled:bool=True, Translation:Vector="0,0,0", Rotation="30,0,0") -> (Pose) {
    alphaInputType=Bool
    rotationMode=BMM_Additive
    rotationSpace=BCS_BoneSpace
    alphaBoolBlend {blendInTime=2, blendOption=Cubic, blendOutTime=1}
    boneToModify {boneName=dyn_C_skirt_b_3}
    rotation {pitch=30}
   }
   n58 = "Transform (Modify) Bone - Bone: dyn_C_skirt_a_2"(ComponentPose=@n57, Alpha=1, Translation="0,0,0", Rotation="-10,0,0", Scale="1,1,1") {
    rotationMode=BMM_Additive
    rotationSpace=BCS_BoneSpace
    scaleMode=BMM_Additive
    scaleSpace=BCS_BoneSpace
    translationMode=BMM_Additive
    translationSpace=BCS_BoneSpace
    alphaBoolBlend {blendOption=Linear}
    boneToModify {boneName=dyn_C_skirt_a_2}
    rotation {pitch=-10}
   }
   n59 = "Transform (Modify) Bone - Bone: dyn_L_skirt_c_2"!(ComponentPose=@n58, bAlphaBoolEnabled:bool=True, Translation="0,10,0", Rotation="0,0,0", Scale="1,1,1") -> (Pose) {
    alphaInputType=Bool
    rotationMode=BMM_Additive
    rotationSpace=BCS_BoneSpace
    scaleMode=BMM_Additive
    scaleSpace=BCS_BoneSpace
    translationMode=BMM_Additive
    translationSpace=BCS_BoneSpace
    alphaBoolBlend {blendInTime=0.5, blendOption=Sinusoidal, blendOutTime=0.5}
    boneToModify {boneName=dyn_L_skirt_c_2}
    translation {y=10}
   }
   n60 = "Transform (Modify) Bone - Bone: dyn_R_skirt_c_2"!(ComponentPose=@n59, bAlphaBoolEnabled:bool=True, Translation="-15,-10,0", Rotation="0,0,0", Scale="1,1,1") -> (Pose) {
    alphaInputType=Bool
    rotationMode=BMM_Additive
    rotationSpace=BCS_BoneSpace
    scaleMode=BMM_Additive
    scaleSpace=BCS_BoneSpace
    translationMode=BMM_Additive
    translationSpace=BCS_BoneSpace
    alphaBoolBlend {blendInTime=0.5, blendOption=Sinusoidal, blendOutTime=0.5}
    boneToModify {boneName=dyn_R_skirt_c_2}
    translation=<-15 -10>
   }
   n61 = "Transform (Modify) Bone - Bone: dyn_L_skirt_b_3"!(ComponentPose=@n60, bAlphaBoolEnabled:bool=True, Translation:Vector="0,0,0", Rotation="0,0,0", Scale:Vector="0.5,1,1") -> (Pose) {
    alphaInputType=Bool
    rotationMode=BMM_Additive
    rotationSpace=BCS_BoneSpace
    scaleMode=BMM_Additive
    scaleSpace=BCS_BoneSpace
    translationMode=BMM_Additive
    translationSpace=BCS_BoneSpace
    alphaBoolBlend {blendInTime=0.2, blendOption=Cubic, blendOutTime=0.2}
    boneToModify {boneName=dyn_L_skirt_b_3}
    scale {x=0.5}
   }
   n62 = "Transform (Modify) Bone - Bone: base_R_skirt_b_1"(ComponentPose=@n61, Alpha=1, Translation="0,0,0", Rotation="35,0,0", Scale="1,1,1") {
    rotationMode=BMM_Additive
    rotationSpace=BCS_BoneSpace
    scaleMode=BMM_Additive
    scaleSpace=BCS_BoneSpace
    translationMode=BMM_Additive
    translationSpace=BCS_BoneSpace
    alphaBoolBlend {blendInTime=0.2, blendOption=Cubic, blendOutTime=0.2}
    boneToModify {boneName=base_R_skirt_b_1}
    rotation {pitch=35}
   }
   n63 = "Constraint - dyn_duckNade_A (T)"(ComponentPose=@n62, Alpha=1, ConstraintWeights_0=1) {
    alphaBoolBlend {blendOption=Linear}
    boneToModify {boneName=dyn_duckNade_A}
    constraintSetup [{offsetOption=None, transformType=Translation, targetBone {boneName=base_duckNade_A}}]
    constraintWeights [1]
   }
   n64 = "Constraint - dyn_duckNade_B (T)"(ComponentPose=@n63, Alpha=1, ConstraintWeights_0=1) {
    alphaBoolBlend {blendOption=Linear}
    boneToModify {boneName=dyn_duckNade_B}
    constraintSetup [{offsetOption=None, transformType=Translation, targetBone {boneName=base_duckNade_B}}]
    constraintWeights [1]
   }
   n65 = "Constraint - dyn_duckNade_ring_A (T)"(ComponentPose=@n64, Alpha=1, ConstraintWeights_0=1) {
    alphaBoolBlend {blendOption=Linear}
    boneToModify {boneName=dyn_duckNade_ring_A}
    constraintSetup [{offsetOption=None, transformType=Translation, targetBone {boneName=base_duckNade_ring_A}}]
    constraintWeights [1]
   }
   n66 = "Constraint - dyn_duckNade_ring_B (T)"(ComponentPose=@n65, Alpha=1, ConstraintWeights_0=1) {
    alphaBoolBlend {blendOption=Linear}
    boneToModify {boneName=dyn_duckNade_ring_B}
    constraintSetup [{offsetOption=None, transformType=Translation, targetBone {boneName=base_duckNade_ring_B}}]
    constraintWeights [1]
   }
   Save cached pose 'postSimFixes'(Pose=Component To Local(ComponentPose=@n66)) {cacheName=postSimFixes}
  ]}
 EventGraph {nodes [
   Event BlueprintUpdateAnimation()
   > n1 = Sequence!(execute) -> (then_0, then_1, then_2:Exec, then_3:Exec, then_4:Exec)
   > n2 = "Gravity Override State Function (Parameter Asset) (Nativized)"(
    CharacterPartAnimInstance=Self-Reference()
    GameTimeInSeconds=GetGameTimeInSeconds()
    Parameters=&/F_MED_SplishSplash_Parameters
   )
   > Set GravityOverride(GravityOverride=@n2)
   n3 = PlayerMovement(execute=@n1.then_1, AnimInstance=Self-Reference(), XDotProductCoefficient=1, YDotProductCoefficient=1, ZDotProductCoefficient=1)
   n4 = BreakVector(InVec=MakeVector(X=@n3, Y=@n3.YMovement))
   > Set strafeLeft float(strafeLeft float:double="Absolute (Float)"(A="Clamp (Float)"(Value=@n4, Min=-1, Max=0))) -> (Output_Get:double)
   > Set strafeRight float(strafeRight float:double="Clamp (Float)"(Value=@n4, Min=0, Max=1)) -> (Output_Get:double)
   > Set moveForward(moveForward:double="Clamp (Float)"(Value=@n4.Y, Min=0, Max=1)) -> (Output_Get:double)
   > Set moveBackward(moveBackward:double="Absolute (Float)"(A="Clamp (Float)"(Value=@n4.Y, Min=-1, Max=0))) -> (Output_Get:double)
   > Set crouching alpha(crouching alpha:double=GetCurveValue(CurveName=crouchingAlpha)) -> (Output_Get:double)
   > Set legRight outside float(legRight outside float:double=Remap Float(Input=GetCurveValue(CurveName=r_thighLocal_ry_float), OldRange=(X=0,Y=-65), NewRange=(X=0,Y=1), Clamp=true) -> (Output=~)) -> (Output_Get:double)
   > Set legRight foward float(legRight foward float:double=Remap Float(Input=GetCurveValue(CurveName=r_thighLocal_rz_float), OldRange=(X=0,Y=-100), NewRange=(X=0,Y=1), Clamp=true) -> (Output=~)) -> (Output_Get:double)
   n5 = Get bIsSliding()
   n6 = Get moveForward()
   n7 = Get moveBackward()
   > Set moveBackForward crouched(moveBackForward crouched:double=float * float!(A=float + float(A=@n6, B=@n7), B:bool=Get bIsCrouching(), C:bool=NOT Boolean(A=@n5)) -> (ReturnValue) {is=title K2Node_PromotableOperator}) -> (Output_Get:double)
   n8 = GetOwnerSkeletalMeshComponent()
   n9 = BreakVector(InVec=BreakTransform(InTransform=GetBoneTransform!(self:Skinned Mesh Component Object Reference=@n8, InBoneName:Name=pelvis, TransformSpace:ERelativeTransformSpace Enum=RTS_ParentBoneSpace) -> (ReturnValue:Transform)))
   > Set isEmotingEasterBlocCrouched(isEmotingEasterBlocCrouched:double=float * float!(A:bool=Get isPlayingEasterBloc(), B:double=Remap Float(Input=@n9.Z, OldRange=(X=70,Y=50), NewRange=(X=0,Y=1), Clamp=true) -> (Output=~)) -> (ReturnValue)) -> (Output_Get:double)
   n10 = Sequence!(execute=@n1.then_4) -> (then_0, then_1, then_2)
   > n11 = Branch(Condition=OR Boolean(A=Get bIsSkydiveDiveMode(), B=Get bIsSkydiveFloating()))
   > Set enableFakeWind(enableFakeWind:bool=true) -> (Output_Get:bool)
   Set enableFakeWind(execute=@n11.else, enableFakeWind=false)
   n12 = Branch(execute=@n10.then_1, Condition=Get bIsSkydiveDiveMode())
   > Set freqVar1(freqVar1:double=40) -> (Output_Get:double)
   > Set freqVar2(freqVar2:double=38) -> (Output_Get:double)
   > Set freqVar3(freqVar3:double=35) -> (Output_Get:double)
   Set freqVar1(execute=@n12.else, freqVar1=35)
   > Set freqVar2(freqVar2=32)
   > Set freqVar3(freqVar3=30)
   Set skirtRotVar1(
    execute=@n10.then_2
    skirtRotVar1:Rotator=Rotator Rumble!(frequency:double=Get freqVar1() -> (freqVar1:double), amplitude:double=2, time offset:Vector="0,1,2", coefficient:Vector="0,1,0") -> (rotatorOutput:Rotator=~) {is=title K2Node_MacroInstance}
   ) -> (Output_Get:Rotator)
   > Set skirtRotVar2(skirtRotVar2:Rotator=Rotator Rumble(frequency=Get freqVar2() -> (freqVar2:double), amplitude=1.85, coefficient="0,1,0") -> (rotatorOutput=~)) -> (Output_Get:Rotator)
   > Set skirtRotVar3(skirtRotVar3:Rotator=Rotator Rumble(frequency=Get freqVar3() -> (freqVar3:double), amplitude=1.7, time offset="0,2,0", coefficient="0,1,0") -> (rotatorOutput=~)) -> (Output_Get:Rotator)
   n13 = GetAnimInstance(self=GetOwnerSkeletalMeshComponent())
   Is Valid(exec=@n1.then_3, InputObject=@n13)
   > Set isPlayingEasterBloc(isPlayingEasterBloc:bool=Montage_IsPlaying(self=@n13, Montage=/BRCosmetics/Animation/Game/MainPlayer/Emotes/Eastern_Bloc/Emote_EasternBloc_CMF_M)) -> (Output_Get:bool)
   n14 = Sequence(execute=@n1.then_2)
   n15 = Get bIsFalling()
   > n16 = Branch(Condition=AND Boolean(A=Get bIsTacticalSprinting(), B=NOT Boolean(A=@n15)))
   > Set moveSkirtOutside(moveSkirtOutside:double=1) -> (Output_Get:double)
   n17 = Branch(execute=@n16.else, Condition=Get bIsSprinting())
   > Set moveSkirtOutside(moveSkirtOutside=0.6)
   Set moveSkirtOutside(execute=@n17.else, moveSkirtOutside=0)
   n18 = Get bIsTacticalSprinting()
   n19 = Get bIsFalling()
   n20 = Get bIsJumping()
   n21 = Branch(
    execute=@n14.then_1
    Condition=OR Boolean!(A=AND Boolean(A=@n20, B=@n18), B=AND Boolean(A=@n18, B=@n19), C:bool=Get bIsTacticalSprinting()) -> (ReturnValue)
   )
   > Set alphaSkirtRigidBody float(alphaSkirtRigidBody float:double=0.5) -> (Output_Get:double)
   n22 = Branch(execute=@n21.else, Condition=Get bIsBaseballSliding())
   > Set alphaSkirtRigidBody float(alphaSkirtRigidBody float=0)
   n23 = Branch(
    execute=@n22.else
    Condition=OR Boolean!(A=Get bIsJumping(), B=Get bIsFalling(), C=Get bIsKneeSliding(), D:bool=Get isPlayingEasterBloc()) -> (ReturnValue)
   )
   > Set alphaSkirtRigidBody float(alphaSkirtRigidBody float=0.15)
   n24 = Branch(execute=@n23.else, Condition=Get isInZipline())
   n25 = Rotate Vector!(A:Vector="1,0,0", B:Rotator=Get PawnDeltaRotation() -> (PawnDeltaRotation:Rotator)) -> (ReturnValue:Vector)
   > Set alphaSkirtRigidBody float(alphaSkirtRigidBody float=Remap Float(Input=Dot Product!(A:Vector="1,0,0", B:Vector=@n25) -> (ReturnValue:double), OldRange=(X=1,Y=0.98), NewRange=(X=0.4,Y=0), Clamp=true) -> (Output=~))
   Set alphaSkirtRigidBody float(execute=@n24.else, alphaSkirtRigidBody float=1)
   n26 = Branch(execute=@n14.then_2, Condition=Get bIsPlayingEmote())
   > Set legRight outside emoting float(legRight outside emoting float:double=Get legRight outside float() -> (legRight outside float:double)) -> (Output_Get:double)
   Set legRight outside emoting float(execute=@n26.else, legRight outside emoting float=0)
   Event OnParachuteOpened()
   > Switch on EGliderType(Selection=Get PlayerGliderType() -> (PlayerGliderType=HangGlider))
   > Set isUsingHangGliding(isUsingHangGliding:bool=true) -> (Output_Get:bool)
   Event OnParachuteClosed()
   > Set isUsingHangGliding(isUsingHangGliding=false)
   Event OnEnterGrindRail() {is=title K2Node_Event}
   > Set isOnGrindRail(isOnGrindRail:bool=true) -> (Output_Get:bool)
   n27 = Event BlueprintProcessAnimInputEvent()
   > n28 = Branch(Condition=IsNamed(self=@n27.AnimInputEvent, InEventName=aie.Player.Zipline.Entered))
   > Set isInZipline(isInZipline:bool=true) -> (Output_Get:bool)
   Branch(execute=@n28.else, Condition=IsNamed(self=@n27.AnimInputEvent, InEventName=aie.Player.Zipline.Exited))
   > Set isInZipline(isInZipline=false)
   Event OnExitGrindRail() {is=title K2Node_Event}
   > Set isOnGrindRail(isOnGrindRail=false)
   Event OnBeginKneeSliding() {is=title K2Node_Event}
   > Set kneeSlidingBegin(kneeSlidingBegin:bool=true) -> (Output_Get:bool)
   > Delay(Duration=0.6)
   > Set kneeSlidingBegin(kneeSlidingBegin=false)
  ]}
 floatSinFunc {nodes [
   n1 = Inputs!() -> (time:double=~, time offset:double=~, frequency:double=~, amplitude:double=~, coefficient:double=~)
   n2 = float + float(A=@n1, B=@n1.time offset)
   n3 = MultiplyByPi!(Value:double="Sin (Radians)"(A=float * float(A=@n2, B=@n1.frequency) {is=title K2Node_CommutativeAssociativeBinaryOperator})) -> (ReturnValue:double)
   Outputs!(floatOutput:double=float * float(A=float * float(A=@n3, B=@n1.amplitude), B=@n1.coefficient))
  ]}
 rotatorRumble {nodes [
   n1 = Inputs!() -> (frequency, amplitude, time offset:Vector=~, coefficient:Vector=~)
   n2 = BreakVector(InVec=@n1.coefficient)
   n3 = BreakVector(InVec=@n1.time offset)
   n4 = GetGameTimeInSeconds()
   Outputs!(rotatorOutput:Rotator=MakeRotator(
     Roll=Float Sin Func!(time:double=@n4, time offset:double=@n3, frequency:double=@n1, amplitude:double=@n1.amplitude, coefficient:double=@n2) -> (floatOutput:double=~) {is=title K2Node_MacroInstance}
     Pitch=Float Sin Func(time=@n4, time offset=@n3.Y, frequency=@n1, amplitude=@n1.amplitude, coefficient=@n2.Y) -> (floatOutput=~)
     Yaw=Float Sin Func(time=@n4, time offset=@n3.Z, frequency=@n1, amplitude=@n1.amplitude, coefficient=@n2.Z) -> (floatOutput=~)
    ))
  ]}
]
variables [
 alphaSkirtRigidBody float {type=number}
 crouching alpha {type=number}
 enableFakeWind {type=boolean}
 freqVar1 {type=number}
 freqVar2 {type=number}
 freqVar3 {type=number}
 GravityOverride {type=Vector}
 isEmotingEasterBlocCrouched {type=number}
 isInZipline {type=boolean}
 isOnGrindRail {type=boolean}
 isPlayingEasterBloc {type=boolean}
 isUsingHangGliding {type=boolean}
 kneeSlidingBegin {type=boolean}
 legRight foward float {type=number}
 legRight outside emoting float {type=number}
 legRight outside float {type=number}
 moveBackForward crouched {type=number}
 moveBackward {type=number}
 moveForward {type=number}
 moveSkirtOutside {type=number}
 skirtRotVar1 {type=Rotator}
 skirtRotVar2 {type=Rotator}
 skirtRotVar3 {type=Rotator}
 strafeLeft float {type=number}
 strafeRight float {type=number}
]
