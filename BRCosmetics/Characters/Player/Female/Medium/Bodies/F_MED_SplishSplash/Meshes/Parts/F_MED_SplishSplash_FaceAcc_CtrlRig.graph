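# Control Rig for the SplishSplash female face accessory that procedurally corrects the head towel bones as the head moves.
# A reusable DeltaRotation function turns head-versus-spine_05 swing or twist into weights for bone rotations and offsets.
# Without back bling, spherical pose readers add poses scaled by a smoothed targeting value.
# With back bling, disabled rigid bodies switch dynamic bones to weighted parent constraints plus offsets, while enabled ones get one corrective.
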ControlRigBlueprint /BRCosmetics/Characters/Player/Female/Medium/Bodies/F_MED_SplishSplash/Meshes/Parts/F_MED_SplishSplash_FaceAcc_CtrlRig
graphs [
 CollapseNode_ContainedGraph {nodes [
   n1 = RigVMFunctionEntryNode()
   RigVMFunction_Sequence = &RigVMFunction_Sequence(ExecuteContext=@n1.ExecutePin)
   n2 = &RigVMFunction_AlphaInterp(
    InterpSpeedDecreasing=15
    InterpSpeedIncreasing=15
    Scale=1
    bClampResult=False
    Value=&RigVMFunction_MathBoolToFloat(Value=RigVMVariableNode() {variable=pub isTargeting}.Value).Result
   )
   RigVMVariableNode(Value=@n2.Result, ExecuteContext=@RigVMFunction_Sequence.A) {variable=isTargeting Lerp}
   > DeltaRotation v4 = RigVMFunctionReferenceNode(
    Rotation Order=ZYX
    absoluteValue=true
    child=(Type=Bone,Name="head")
    debugRotation=false
    onlyTwist=true
    outputMinMax=(X=-1,Y=1)
    parent=(Type=Bone,Name="spine_05")
    premult=1
    rotationMinMax=(X=-30,Y=30)
    twistAxis=(X=1,Y=0,Z=0)
   )
   RigVMFunction_MathQuaternionToEuler = &RigVMFunction_MathQuaternionToEuler(RotationOrder=ZYX, Value=@"DeltaRotation v4".outputQuat)
   Multiply = &RigVMFunction_MathDoubleMul(B=-1, A=@RigVMFunction_MathQuaternionToEuler.Result.X)
   Clamp = &RigVMFunction_MathDoubleClamp(Maximum=40, Minimum=-40, Value=@Multiply.Result)
   RigVMFunction_MathQuaternionFromEuler = &RigVMFunction_MathQuaternionFromEuler(Euler=(X=0,Y=0,Z=0), Euler.Z=@Clamp.Result)
   > n3 = &RigUnit_ModifyTransforms(
    ItemToModify=((Item=(Type=Bone,Name="base_headTowel_3"),Transform=(Rotation=(X=0,Y=0,Z=-0.2588,W=0.9659),Translation=(X=0,Y=0,Z=0),Scale3D=(X=1,Y=1,Z=1))))
    WeightMinimum=0
    Weight=@"DeltaRotation v4".outputWeight
    ItemToModify.0.Transform.Rotation=@RigVMFunction_MathQuaternionFromEuler.Result
   )
   RigVMFunction_ControlFlowBranch = &RigVMFunction_ControlFlowBranch(ExecuteContext=@n3.ExecutePin, Condition=RigVMVariableNode() {variable=pub isWearingBackBling}.Value)
   DeltaRotation v4_1 = RigVMFunctionReferenceNode(
    Rotation Order=ZXY
    absoluteValue=false
    child=(Type=Bone,Name="head")
    debugRotation=false
    onlyTwist=false
    outputMinMax=(X=0,Y=1)
    parent=(Type=Bone,Name="spine_05")
    premult=1
    rotationMinMax=(X=-15,Y=-35)
    triggerAxis=Z
    twistAxis=(X=1,Y=0,Z=0)
    rotationMinMax.X=RigVMDispatchNode(False=-5, True=-15, Condition=RigVMVariableNode() {variable=pub isRigidBodiesEnabled}.Value) {template="DISPATCH_RigVMDispatch_If(in Condition,in True,in False,out Result)"}.Result
    ExecuteContext=@RigVMFunction_ControlFlowBranch.False
   )
   > &RigUnit_ModifyTransforms(
    ItemToModify="((Item=(Type=Bone,Name=\"kin_headTowel_1\"),Transform=(Rotation=(X=0,Y=-0.1736,Z=0,W=0.9848),Translation=(X=0,Y=0,Z=0),Scale3D=(X=1,Y=1,Z=1))),(Item=(Type=Bone,Name=\"kin_headTowel_2\"),Transform=(Rotation=(X=0,Y=-0.342,Z=0,W=0.9397),Translation=(X=0,Y=0,Z=0), ...(truncated)"
    WeightMinimum=0
    Weight=@"DeltaRotation v4_1".outputWeight
   )
   > &RigUnit_ModifyTransforms(
    ItemToModify="((Item=(Type=Bone,Name=\"dyn_headTowel_3\"),Transform=(Rotation=(X=0,Y=0.1736,Z=0,W=0.9848),Translation=(X=0,Y=0,Z=0),Scale3D=(X=1,Y=1,Z=1))),(Item=(Type=Bone,Name=\"dyn_headTowel_4\"),Transform=(Rotation=(X=0,Y=0.1736,Z=0,W=0.9848),Translation=(X=0,Y=0,Z=0),S ...(truncated)"
    WeightMinimum=0
    Weight=RigVMDispatchNode(True=0, Condition=RigVMVariableNode() {variable=pub isRigidBodiesEnabled}.Value, False=@"DeltaRotation v4_1".outputWeight) {template="DISPATCH_RigVMDispatch_If(in Condition,in True,in False,out Result)"}.Result
   )
   > n4 = &RigUnit_SphericalPoseReader(
    ActiveRegionScaleFactors=(PositiveWidth=1,NegativeWidth=1,PositiveHeight=1,NegativeHeight=1)
    ActiveRegionSize=0.05
    Debug=(bDrawDebug=false,bDraw2D=False,bDrawLocalAxes=False,DebugScale=25,DebugSegments=20,DebugThickness=0.1)
    DriverAxis=(X=1,Y=0,Z=0)
    DriverItem=(Type=Bone,Name="head")
    FalloffRegionScaleFactors=(PositiveWidth=1,NegativeWidth=1,PositiveHeight=1,NegativeHeight=1)
    FalloffSize=0.2
    FlipHeightScaling=False
    FlipWidthScaling=False
    OptionalParentItem=(Type=Bone,Name="spine_05")
    RotationOffset=(X=-12,Y=-47,Z=0)
   )
   > &RigUnit_ModifyTransforms(
    ItemToModify="((Item=(Type=Bone,Name=\"kin_headTowel_1\"),Transform=(Rotation=(X=0,Y=-0.2588,Z=0,W=0.9659),Translation=(X=0,Y=0,Z=0),Scale3D=(X=1,Y=1,Z=1))),(Item=(Type=Bone,Name=\"kin_headTowel_2\"),Transform=(Rotation=(X=0,Y=-0.1736,Z=0,W=0.9848),Translation=(X=0,Y=0,Z=0) ...(truncated)"
    WeightMinimum=0
    Weight=&RigVMFunction_MathDoubleMul(A=@n4.OutputParam, B=RigVMVariableNode() {variable=isTargeting Lerp}.Value).Result
   )
   > n5 = &RigUnit_SphericalPoseReader(
    ActiveRegionScaleFactors=(PositiveWidth=1,NegativeWidth=1,PositiveHeight=1,NegativeHeight=1)
    ActiveRegionSize=0.08
    Debug=(bDrawDebug=false,bDraw2D=False,bDrawLocalAxes=False,DebugScale=25,DebugSegments=20,DebugThickness=0.1)
    DriverAxis=(X=1,Y=0,Z=0)
    DriverItem=(Type=Bone,Name="head")
    FalloffRegionScaleFactors=(PositiveWidth=1,NegativeWidth=1,PositiveHeight=1,NegativeHeight=1)
    FalloffSize=0.2
    FlipHeightScaling=False
    FlipWidthScaling=False
    OptionalParentItem=(Type=Bone,Name="spine_05")
    RotationOffset=(X=0,Y=-67.8,Z=-88)
   )
   > &RigUnit_ModifyTransforms(
    ItemToModify="((Item=(Type=Bone,Name=\"kin_headTowel_1\"),Transform=(Rotation=(X=0,Y=-0.4226,Z=0,W=0.9063),Translation=(X=0,Y=0,Z=0),Scale3D=(X=1,Y=1,Z=1))),(Item=(Type=Bone,Name=\"kin_headTowel_2\"),Transform=(Rotation=(X=0,Y=0,Z=0,W=1),Translation=(X=0,Y=0,Z=0),Scale3D=(X ...(truncated)"
    WeightMinimum=0
    Weight=&RigVMFunction_MathDoubleMul(A=@n5.OutputParam, B=RigVMVariableNode() {variable=isTargeting Lerp}.Value).Result
   )
   n6 = &RigVMFunction_ControlFlowBranch(ExecuteContext=@RigVMFunction_ControlFlowBranch.True, Condition=RigVMVariableNode() {variable=pub isRigidBodiesEnabled}.Value)
   &RigUnit_ParentConstraint(
    AdvancedSettings=(RotationOrderForFilter=XZY)
    Child=(Type=Bone,Name="dyn_headTowel_3")
    Parents=((Item=(Type=Bone,Name="neck_01"),Weight=0.0494),(Item=(Type=Bone,Name="neck_02"),Weight=0.0967),(Item=(Type=Bone,Name="head"),Weight=0.8539))
    Weight=1
    ExecutePin=@n6.False
   )
   > &RigUnit_ParentConstraint(
    AdvancedSettings=(RotationOrderForFilter=XZY)
    Child=(Type=Bone,Name="dyn_headTowel_4")
    Parents=((Item=(Type=Bone,Name="neck_02"),Weight=0.1719),(Item=(Type=Bone,Name="spine_05"),Weight=0.399),(Item=(Type=Bone,Name="neck_01"),Weight=0.4291))
    Weight=1
   )
   > &RigUnit_ParentConstraint(
    AdvancedSettings=(RotationOrderForFilter=XZY)
    Child=(Type=Bone,Name="dyn_headTowel_5")
    Parents=((Item=(Type=Bone,Name="clavicle_l"),Weight=0.0508),(Item=(Type=Bone,Name="clavicle_r"),Weight=0.051),(Item=(Type=Bone,Name="spine_05"),Weight=0.8982))
    Weight=1
   )
   > DeltaRotation v4_3 = RigVMFunctionReferenceNode(
    Rotation Order=ZXY
    absoluteValue=false
    child=(Type=Bone,Name="head")
    debugRotation=false
    onlyTwist=false
    outputMinMax=(X=0,Y=1)
    parent=(Type=Bone,Name="spine_05")
    premult=1
    rotationMinMax=(X=-15,Y=-35)
    triggerAxis=Z
    twistAxis=(X=1,Y=0,Z=0)
   )
   > &RigUnit_OffsetTransformForItem(
    Item=(Type=Bone,Name="kin_headTowel_2")
    OffsetTransform=(Rotation=(X=0,Y=-0.1736,Z=0,W=0.9848),Translation=(X=-1,Y=0,Z=-1),Scale3D=(X=1,Y=1,Z=1))
    bPropagateToChildren=false
    Weight=@"DeltaRotation v4_3".outputWeight
   )
   > &RigUnit_OffsetTransformForItem(
    Item=(Type=Bone,Name="dyn_headTowel_3")
    OffsetTransform=(Rotation=(X=0,Y=-0.342,Z=0,W=0.9397),Translation=(X=-2,Y=0,Z=0),Scale3D=(X=1,Y=1,Z=1))
    bPropagateToChildren=false
    Weight=@"DeltaRotation v4_3".outputWeight
   )
   > DeltaRotation v4_5 = RigVMFunctionReferenceNode(
    Rotation Order=ZYX
    absoluteValue=true
    child=(Type=Bone,Name="head")
    debugRotation=false
    onlyTwist=true
    outputMinMax=(X=0,Y=1)
    parent=(Type=Bone,Name="spine_05")
    premult=1
    rotationMinMax=(X=0,Y=40)
    twistAxis=(X=1,Y=0,Z=0)
   )
   > &RigUnit_OffsetTransformForItem(
    Item=(Type=Bone,Name="kin_headTowel_1")
    OffsetTransform=(Rotation=(X=0,Y=0,Z=0.1736,W=0.9848),Translation=(X=0,Y=0,Z=0),Scale3D=(X=1,Y=1,Z=1))
    bPropagateToChildren=false
    Weight=@"DeltaRotation v4_5".outputWeight
   )
   > &RigUnit_OffsetTransformForItem(
    Item=(Type=Bone,Name="kin_headTowel_2")
    OffsetTransform=(Rotation=(X=0,Y=0,Z=0.0872,W=0.9962),Translation=(X=0,Y=0,Z=0),Scale3D=(X=1,Y=1,Z=1))
    bPropagateToChildren=false
    Weight=@"DeltaRotation v4_5".outputWeight
   )
   > &RigUnit_OffsetTransformForItem(
    Item=(Type=Bone,Name="dyn_headTowel_3")
    OffsetTransform=(Rotation=(X=0,Y=0,Z=0,W=1),Translation=(X=0,Y=2,Z=0),Scale3D=(X=1,Y=1,Z=1))
    bPropagateToChildren=false
    Weight=@"DeltaRotation v4_5".outputWeight
   )
   > DeltaRotation v4_4 = RigVMFunctionReferenceNode(
    Rotation Order=ZYX
    absoluteValue=true
    child=(Type=Bone,Name="head")
    debugRotation=false
    onlyTwist=true
    outputMinMax=(X=0,Y=1)
    parent=(Type=Bone,Name="spine_05")
    premult=1
    rotationMinMax=(X=0,Y=-40)
    twistAxis=(X=1,Y=0,Z=0)
   )
   > &RigUnit_OffsetTransformForItem(
    Item=(Type=Bone,Name="kin_headTowel_1")
    OffsetTransform=(Rotation=(X=0,Y=0,Z=-0.1736,W=0.9848),Translation=(X=0,Y=0,Z=0),Scale3D=(X=1,Y=1,Z=1))
    bPropagateToChildren=false
    Weight=@"DeltaRotation v4_4".outputWeight
   )
   > &RigUnit_OffsetTransformForItem(
    Item=(Type=Bone,Name="kin_headTowel_2")
    OffsetTransform=(Rotation=(X=0,Y=0,Z=-0.0872,W=0.9962),Translation=(X=0,Y=0,Z=0),Scale3D=(X=1,Y=1,Z=1))
    bPropagateToChildren=false
    Weight=@"DeltaRotation v4_4".outputWeight
   )
   > &RigUnit_OffsetTransformForItem(
    Item=(Type=Bone,Name="dyn_headTowel_3")
    OffsetTransform=(Rotation=(X=0,Y=0,Z=0,W=1),Translation=(X=0,Y=-2,Z=0),Scale3D=(X=1,Y=1,Z=1))
    bPropagateToChildren=false
    Weight=@"DeltaRotation v4_4".outputWeight
   )
   DeltaRotation v4_2 = RigVMFunctionReferenceNode(
    Rotation Order=ZXY
    absoluteValue=false
    child=(Type=Bone,Name="head")
    debugRotation=false
    onlyTwist=false
    outputMinMax=(X=0,Y=1)
    parent=(Type=Bone,Name="spine_05")
    premult=1
    rotationMinMax=(X=-40,Y=-50)
    triggerAxis=Z
    twistAxis=(X=1,Y=0,Z=0)
    ExecuteContext=@n6.True
   )
   > &RigUnit_ModifyTransforms(
    ItemToModify=((Item=(Type=Bone,Name="kin_headTowel_1"),Transform=(Rotation=(X=0,Y=0,Z=0.3104,W=0.9506),Translation=(X=0,Y=0,Z=0),Scale3D=(X=1,Y=1,Z=1))))
    WeightMinimum=0
    Weight=@"DeltaRotation v4_2".outputWeight
   )
   RigVMFunctionReturnNode(ExecutePin=@RigVMFunction_Sequence.B)
  ]}
 New Function_ContainedGraph {nodes [
   n1 = RigVMFunctionEntryNode()
   RigVMFunction_MathBoolToInteger = &RigVMFunction_MathBoolToInteger(Value=@n1.onlyTwist)
   RigVMFunction_MathQuaternionSwingTwist = &RigVMFunction_MathQuaternionSwingTwist(
    TwistAxis=@n1.twistAxis
    Input=&RigUnit_GetRelativeTransformForItem(bChildInitial=False, bParentInitial=False, Child=@n1.child, Parent=@n1.parent).RelativeTransform.Rotation
   )
   n2 = RigVMDispatchNode(
    Values=((),())
    Index=@RigVMFunction_MathBoolToInteger.Result
    Values.0=@RigVMFunction_MathQuaternionSwingTwist.Swing
    Values.1=@RigVMFunction_MathQuaternionSwingTwist.Twist
   ) {template="DISPATCH_RigVMDispatch_SelectInt32(in Index,in Values,out Result)"}
   RigVMFunction_MathQuaternionToEuler = &RigVMFunction_MathQuaternionToEuler(Value=@n2.Result, RotationOrder=@n1.Rotation Order)
   > DISPATCH_RigVMDispatch_Print = RigVMDispatchNode(ScreenColor=(R=1,G=1,B=1,A=1), ScreenDuration=1.5, Enabled=@n1.debugRotation, Value=@RigVMFunction_MathQuaternionToEuler.Result) {template="DISPATCH_RigVMDispatch_Print(in Prefix,in Value,in Enabled,in ScreenDuration,in ScreenColor)"}
   Cast = RigVMDispatchNode(Value=@n1.triggerAxis) {template="DISPATCH_RigVMDispatch_CastEnumToInt(in Value,out Result)"}
   DISPATCH_RigVMDispatch_SelectInt32 = RigVMDispatchNode(
    Values=(0,0,0)
    Index=@Cast.Result
    Values.0=@RigVMFunction_MathQuaternionToEuler.Result.X
    Values.1=@RigVMFunction_MathQuaternionToEuler.Result.Y
    Values.2=@RigVMFunction_MathQuaternionToEuler.Result.Z
   ) {template="DISPATCH_RigVMDispatch_SelectInt32(in Index,in Values,out Result)"}
   Remap = &RigVMFunction_MathDoubleRemap(
    bClamp=False
    Value=@DISPATCH_RigVMDispatch_SelectInt32.Result
    TargetMinimum=@n1.outputMinMax.X
    TargetMaximum=@n1.outputMinMax.Y
    SourceMinimum=@n1.rotationMinMax.X
    SourceMaximum=@n1.rotationMinMax.Y
   )
   Absolute = &RigVMFunction_MathDoubleAbs(Value=@Remap.Result)
   Multiply = &RigVMFunction_MathDoubleMul(
    A=&RigVMFunction_MathDoubleClamp(Value=RigVMDispatchNode(True=@Absolute.Result, Condition=@n1.absoluteValue, False=@Remap.Result) {template="DISPATCH_RigVMDispatch_If(in Condition,in True,in False,out Result)"}.Result).Result
    B=@n1.premult
   )
   > RigVMFunctionReturnNode(outputQuat=@n2.Result, outputWeight=@Multiply.Result)
  ]}
 RigVMFunctionLibrary {nodes [DeltaRotation v4 = RigVMCollapseNode(
    ExecuteContext=()
    Rotation Order=ZYX
    child=(Type=Bone,Name="None")
    outputMinMax=(X=0,Y=1)
    parent=(Type=Bone,Name="None")
    premult=1
    rotationMinMax=(X=0,Y=0)
    twistAxis=(X=1,Y=0,Z=0)
   )]}
 RigVMModel {nodes [
   RigUnit_BeginExecution = &RigUnit_BeginExecution()
   ponytail = RigVMCollapseNode(ExecutePin=@RigUnit_BeginExecution.ExecutePin)
   setFKCtrl = RigVMFunctionReferenceNode(ExecuteContext=(), setFK_Ctrl=false)
  ]}
 RigVMModel Rig {nodes [
   n1 = &RigUnit_PrepareForExecution()
   &RigUnit_HierarchyImportFromSkeleton(NameSpace=None, bIncludeCurves=true, bIncludeMeshSockets=False, bIncludeVirtualBones=True, ExecutePin=@n1.ExecutePin)
   Item Array = RigVMDispatchNode(Value=((Type=Bone,Name="kin_headTowel_1"),(Type=Bone,Name="kin_headTowel_2"),(Type=Bone,Name="base_headTowel_3"))) {template="DISPATCH_RigVMDispatch_Constant(io Value)"}
   addFKCtrl = RigVMFunctionReferenceNode(
    Custom_FKCtrl="((parent_2_#1=(Type=Control,Name=\"head_ctrl\"),factor_scale_21_#2=10,shape_settings_24_#3=(bVisible=True,Name=\"Circle_Thin\",Color=(R=1,G=0,B=0,A=1),Transform=(Rotation ...(truncated)"
    ExecuteContext=()
    Options_FKCtrl="(pelvis_1_#4=true,spine_3_#5=true,clavicle_5_#6=false,arm_7_#7=False,leg_9_#8=False,head_11_46DC15EF465CD2 ...(truncated)"
    Shape_Settings="(pelvis_2_#9=(bVisible=True,Name=\"Hexagon_Thin\",Color=(R=0.0117,G=0.8281,B=1,A=1),Transform=(Rotation=(X=0,Y=-0.7071,Z=0,W=0.7071),Translation=(X=2,Y=0,Z=0),Scale3D=(X=5,Y=5,Z=7))),spine_4_#10=(bVi ...(truncated)"
    Custom_FKCtrl.0.joint_names_14_#11=@"Item Array".Value
   )
  ]}
]
variables [
 isTargeting Lerp {type=float, default=0}
 pub isRigidBodiesEnabled {type=bool, default=False}
 pub isTargeting {type=bool, default=False}
 pub isWearingBackBling {type=bool, default=False}
]
