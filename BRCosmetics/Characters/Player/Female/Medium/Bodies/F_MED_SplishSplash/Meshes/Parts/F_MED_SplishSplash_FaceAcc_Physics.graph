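# Physics asset for the face accessory mesh of the F_MED_SplishSplash female body cosmetic, set up for a dangling head towel.
# Capsules on the upper spine, clavicles, neck and collar form body volumes around the towel.
# A three-bone dyn_headTowel chain hangs from base_headTowel_3 on the neck, linked by constraints with tight swing and twist angles.
# The constraints joining the body bones and towel base leave all angular motion free.
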PhysicsAsset /BRCosmetics/Characters/Player/Female/Medium/Bodies/F_MED_SplishSplash/Meshes/Parts/F_MED_SplishSplash_FaceAcc_Physics
bodies {
 base_headTowel_3 [{shapeType=Capsule, length=1, radius=0.5, shapeName=base_headTowel_3_capsule}]
 clavicle_l [{
   shapeType=Capsule
   length=9.7535
   radius=7.6159
   shapeName=clavicle_l_capsule
   center=<7.0362 0 -3.2519>
   rotation=<rot -0.0001 -90 -76.4237>
  }]
 clavicle_r [{
   shapeType=Capsule
   length=9.7535
   radius=7.6159
   shapeName=clavicle_r_capsule
   center=<-7.0362 0 3.2519>
   rotation=<rot -0.0001 -90 103.5763>
  }]
 coll_neckCollar_1_c [{shapeType=Capsule, length=12.119, radius=2.881, shapeName=coll_neckCollar_1_c_capsule, center=<0 1.4429 -0.9111>, rotation=<rot 90 0 0>}]
 dyn_headTowel_3 [{
   shapeType=Capsule
   length=2.9993
   radius=1.5298
   shapeName=dyn_headTowel_3_capsule
   center=<2.3501 0 -0.2345>
   rotation=<rot -83.0331 180 180>
  }]
 dyn_headTowel_4 [
  {
   shapeType=Capsule
   length=3.8339
   radius=1.8149
   shapeName=dyn_headTowel_3_capsule
   center=<2.5426 -0.8529 -0.1444>
   rotation=<rot -87.1572 0 -15.9165>
  }
  {shapeType=Capsule, length=3.089, radius=2.5597, shapeName=dyn_headTowel_3_capsule, center=<2.5097 0.7552 0.5166>, rotation=<rot -87.157 0 0>}
 ]
 dyn_headTowel_5 [{shapeType=Capsule, length=2.5317, radius=2.5085, shapeName=dyn_headTowel_3_capsule, center=<2.3999 0 1.2707>, rotation=<rot -74.8538 0 0>}]
 neck_01 [{shapeType=Capsule, length=6.5281, radius=4.0192, shapeName=neck_01_capsule, center=<0 0.4331 0>, rotation=<rot 90 -90 -86.6975>}]
 spine_05 [
  {shapeType=Capsule, length=13.2498, radius=9.5027, shapeName=spine_05_capsule, center=<0.8425 0.2559 0.4331>, rotation=<rot 0 -90 0>}
  {shapeType=Capsule, length=13.3571, radius=5.4095, shapeName=spine_05_capsule, center=<7.3778 -2.4224 0.4331>, rotation=<rot 0 -90 0>}
 ]
}
constraints [
 {bone1Name=dyn_headTowel_5, bone2Name=dyn_headTowel_4, swing1LimitDegrees=30, swing2LimitDegrees=30, twistLimitDegrees=20}
 {bone1Name=dyn_headTowel_3, bone2Name=base_headTowel_3, swing1LimitDegrees=30, swing2LimitDegrees=30, twistLimitDegrees=20}
 {bone1Name=dyn_headTowel_4, bone2Name=dyn_headTowel_3, swing1LimitDegrees=30, swing2LimitDegrees=30, twistLimitDegrees=20}
 {
  bone1Name=base_headTowel_3
  bone2Name=neck_01
  swing1LimitDegrees=45
  swing1Motion=Free
  swing2LimitDegrees=45
  swing2Motion=Free
  twistLimitDegrees=45
  twistMotion=Free
 }
 {
  bone1Name=neck_01
  bone2Name=spine_05
  swing1LimitDegrees=45
  swing1Motion=Free
  swing2LimitDegrees=45
  swing2Motion=Free
  twistLimitDegrees=45
  twistMotion=Free
 }
 {
  bone1Name=coll_neckCollar_1_c
  bone2Name=spine_05
  swing1LimitDegrees=45
  swing1Motion=Free
  swing2LimitDegrees=45
  swing2Motion=Free
  twistLimitDegrees=45
  twistMotion=Free
 }
 {
  bone1Name=clavicle_r
  bone2Name=spine_05
  swing1LimitDegrees=45
  swing1Motion=Free
  swing2LimitDegrees=45
  swing2Motion=Free
  twistLimitDegrees=45
  twistMotion=Free
 }
 {
  bone1Name=clavicle_l
  bone2Name=spine_05
  swing1LimitDegrees=45
  swing1Motion=Free
  swing2LimitDegrees=45
  swing2Motion=Free
  twistLimitDegrees=45
  twistMotion=Free
 }
]
