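# Animation Blueprint for the SplishSplash female cosmetic's face accessory, adding secondary motion to its head towel, hair and neck collar.
# Body and head poses are copied, layered, then passed through Control Rig, towel bone offsets and a collar constraint.
# Hair AnimDynamics feeds two head-relative RigidBody setups, chosen by a bool blend.
# The event graph caches the head mesh, flags non-cape back bling and computes state-based gravity.
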AnimBlueprint /BRCosmetics/Characters/Player/Female/Medium/Bodies/F_MED_SplishSplash/Meshes/Parts/F_MED_SplishSplash_FaceAcc_AnimBP
target_skeleton=&/F_MED_SplishSplash_FaceAcc_Skeleton
graphs [
 AnimGraph {nodes [
   Output Pose(Result=Use cached pose 'dynamicsCP'()) {layerGroup=None}
   n1 = Get Head Skeleton() -> (Head Skeleton:Skeletal Mesh Component Object Reference)
   n2 = Control Rig!(
    Source=Layered blend per bone(
     BasePose=Copy Pose From Mesh() {bUseAttachedParent=true}
     BlendPoses_0=Copy Pose From Mesh(SourceMeshComponent=@n1) {bUseAttachedParent=true}
    )
    Alpha
    pub isTargeting:bool=Get bIsTargeting()
    pub isRigidBodiesEnabled:bool=Get bRigidBodyEnabled()
    pub isWearingBackBling:bool=Get isWearingBackbling()
   ) -> (Pose) {
    alphaBoolBlend {blendOption=Linear}
    controlRigAssetReference {blueprintRigClass=&/F_MED_SplishSplash_FaceAcc_CtrlRig.F_MED_SplishSplash_FaceAcc_CtrlRig_C}
   }
   Save cached pose 'controlRigCP'(Pose=@n2) {cacheName=controlRigCP}
   n3 = RigidBody!(ComponentPose=Local To Component(LocalPose=Use cached pose 'animDynamicsCP'()), Alpha, OverrideWorldGravity:Vector="0,0,-980") -> (Pose) {
    bClampLinearTranslationLimitToRefPose=true
    bOverrideWorldGravity=true
    overridePhysicsAsset=&/F_MED_SplishSplash_FaceAcc_Pack_Physics
    simulationSpace=BaseBoneSpace
    alphaBoolBlend {blendOption=Linear}
    baseBoneRef {boneName=head}
    componentAppliedLinearAccClamp=<3500 3500 3500>
    overrideWorldGravity {z=-980}
    simSpaceSettings {
     maxAngularAcceleration=1
     maxAngularVelocity=0.3
     maxLinearAcceleration=500
     maxLinearVelocity=500
     velocityScaleZ=0.2
     worldAlpha=1
     externalLinearDragV=<1 1 1>
    }
   }
   n4 = RigidBody(ComponentPose=Local To Component(LocalPose=Use cached pose 'animDynamicsCP'()), OverrideWorldGravity="0,0,-980") {
    bClampLinearTranslationLimitToRefPose=true
    bOverrideWorldGravity=true
    overridePhysicsAsset=&/F_MED_SplishSplash_FaceAcc_Physics
    simulationSpace=BaseBoneSpace
    alphaBoolBlend {blendOption=Linear}
    baseBoneRef {boneName=head}
    componentAppliedLinearAccClamp=<3500 3500 3500>
    overrideWorldGravity {z=-980}
    simSpaceSettings {
     maxAngularAcceleration=1
     maxAngularVelocity=0.3
     maxLinearAcceleration=1000
     maxLinearVelocity=500
     velocityScaleZ=0.5
     worldAlpha=1
     externalLinearDragV=<1 1 1>
    }
   }
   Save cached pose 'dynamicsCP'(Pose=Blend Poses by bool(BlendPose_0=Component To Local(ComponentPose=@n3), BlendPose_1=Component To Local(ComponentPose=@n4)) {blendType=Linear}) {cacheName=dynamicsCP}
   n5 = "Constraint - coll_neckCollar_1_c (PPPPP)"!(
    ComponentPose=Local To Component(LocalPose=Use cached pose 'generalAdjuistementsCP'())
    Alpha=1
    ConstraintWeights_0=0.6991
    ConstraintWeights_1:float=0.2029
    ConstraintWeights_2:float=0.0513
    ConstraintWeights_3:float=0.0256
    ConstraintWeights_4:float=0.0211
   ) -> (Pose) {
    alphaBoolBlend {blendOption=Linear}
    boneToModify {boneName=coll_neckCollar_1_c}
    constraintSetup [
     {targetBone {boneName=spine_05}}
     {targetBone {boneName=neck_01}}
     {targetBone {boneName=neck_02}}
     {targetBone {boneName=clavicle_r}}
     {targetBone {boneName=clavicle_l}}
    ]
    constraintWeights [1,1,1,1,1]
   }
   n6 = Anim Dynamics - Bone: dyn_hair_1_l!(ComponentPose=@n5, Alpha=0.7, GravityOverride:Vector="0,0,-980") -> (Pose) {
    angularDampingOverride=1
    angularSpringConstant=1000
    bAngularSpring=true
    bOverrideAngularDamping=true
    bUseGravityOverride=true
    gravityScale=0.35
    simulationSpace=RootRelative
    alphaBoolBlend {blendOption=Linear}
    boundBone {boneName=dyn_hair_1_l}
    componentAppliedLinearAccClamp=<5000 5000 5000>
    componentLinearAccScale=<1 1 1>
    componentLinearVelScale=<1 1 1>
    gravityOverride {z=-980}
    physicsBodyDefinitions [{
      boundBone {boneName=dyn_hair_1_l}
      boxExtents=<15 15 15>
      constraintSetup {angularLimitsMax=<0 0 10>, angularLimitsMin=<0 -15 0>, angularTarget=<1 0 0>}
      localJointOffset=<7 0 0>
     }]
   }
   n7 = Anim Dynamics - Bone: dyn_hair_2_l!(ComponentPose=@n6, Alpha=0.7, GravityOverride:Vector="0,0,-980") -> (Pose) {
    angularDampingOverride=1
    angularSpringConstant=1000
    bAngularSpring=true
    bOverrideAngularDamping=true
    bUseGravityOverride=true
    gravityScale=0.35
    simulationSpace=RootRelative
    alphaBoolBlend {blendOption=Linear}
    boundBone {boneName=dyn_hair_2_l}
    componentAppliedLinearAccClamp=<5000 5000 5000>
    componentLinearAccScale=<1 1 1>
    componentLinearVelScale=<1 1 1>
    gravityOverride {z=-980}
    physicsBodyDefinitions [{
      boundBone {boneName=dyn_hair_2_l}
      boxExtents=<15 15 15>
      constraintSetup {angularLimitsMax=<0 0 10>, angularLimitsMin=<0 -15 -10>, angularTarget=<1 0 0>}
      localJointOffset=<7 0 0>
     }]
   }
   n8 = Anim Dynamics - Bone: dyn_hair_3_l!(ComponentPose=@n7, Alpha=0.7, GravityOverride:Vector="0,0,-980") -> (Pose) {
    angularDampingOverride=1
    angularSpringConstant=1000
    bAngularSpring=true
    bOverrideAngularDamping=true
    bUseGravityOverride=true
    gravityScale=0.35
    simulationSpace=RootRelative
    alphaBoolBlend {blendOption=Linear}
    boundBone {boneName=dyn_hair_3_l}
    componentAppliedLinearAccClamp=<5000 5000 5000>
    componentLinearAccScale=<1 1 1>
    componentLinearVelScale=<1 1 1>
    gravityOverride {z=-980}
    physicsBodyDefinitions [{
      boundBone {boneName=dyn_hair_3_l}
      boxExtents=<15 15 15>
      constraintSetup {angularLimitsMax=<0 15 15>, angularLimitsMin=<0 -15 -15>, angularTarget=<1 0 0>}
      localJointOffset=<7 0 0>
     }]
   }
   n9 = Anim Dynamics - Bone: dyn_hair_1_r!(ComponentPose=@n8, Alpha=0.7, GravityOverride:Vector="0,0,-980") -> (Pose) {
    angularDampingOverride=1
    angularSpringConstant=1000
    bAngularSpring=true
    bOverrideAngularDamping=true
    bUseGravityOverride=true
    gravityScale=0.35
    simulationSpace=RootRelative
    alphaBoolBlend {blendOption=Linear}
    boundBone {boneName=dyn_hair_1_r}
    componentAppliedLinearAccClamp=<5000 5000 5000>
    componentLinearAccScale=<1 1 1>
    componentLinearVelScale=<1 1 1>
    gravityOverride {z=-980}
    physicsBodyDefinitions [{
      boundBone {boneName=dyn_hair_1_r}
      boxExtents=<15 15 15>
      constraintSetup {angularLimitsMax=<0 15 15>, angularLimitsMin=<0 -15 -15>, angularTarget=<1 0 0>}
      localJointOffset=<7 0 0>
     }]
   }
   n10 = Anim Dynamics - Bone: dyn_hair_2_r!(ComponentPose=@n9, Alpha=0.7, GravityOverride:Vector="0,0,-980") -> (Pose) {
    angularDampingOverride=1
    angularSpringConstant=1000
    bAngularSpring=true
    bOverrideAngularDamping=true
    bUseGravityOverride=true
    gravityScale=0.35
    simulationSpace=RootRelative
    alphaBoolBlend {blendOption=Linear}
    boundBone {boneName=dyn_hair_2_r}
    componentAppliedLinearAccClamp=<5000 5000 5000>
    componentLinearAccScale=<1 1 1>
    componentLinearVelScale=<1 1 1>
    gravityOverride {z=-980}
    physicsBodyDefinitions [{
      boundBone {boneName=dyn_hair_2_r}
      boxExtents=<15 15 15>
      constraintSetup {angularLimitsMax=<0 15 15>, angularLimitsMin=<0 -15 -15>, angularTarget=<1 0 0>}
      localJointOffset=<7 0 0>
     }]
   }
   n11 = Anim Dynamics - Bone: dyn_hair_3_r!(ComponentPose=@n10, Alpha=0.7, GravityOverride:Vector="0,0,-980") -> (Pose) {
    angularDampingOverride=1
    angularSpringConstant=1000
    bAngularSpring=true
    bOverrideAngularDamping=true
    bUseGravityOverride=true
    gravityScale=0.35
    simulationSpace=RootRelative
    alphaBoolBlend {blendOption=Linear}
    boundBone {boneName=dyn_hair_3_r}
    componentAppliedLinearAccClamp=<5000 5000 5000>
    componentLinearAccScale=<1 1 1>
    componentLinearVelScale=<1 1 1>
    gravityOverride {z=-980}
    physicsBodyDefinitions [{
      boundBone {boneName=dyn_hair_3_r}
      boxExtents=<15 15 15>
      constraintSetup {angularLimitsMax=<0 15 15>, angularLimitsMin=<0 -15 -15>, angularTarget=<1 0 0>}
      localJointOffset=<7 0 0>
     }]
   }
   Save cached pose 'animDynamicsCP'(Pose=Component To Local(ComponentPose=@n11)) {cacheName=animDynamicsCP}
   n12 = "Transform (Modify) Bone - Bone: kin_headTowel_2"!(
    ComponentPose=Local To Component(LocalPose=Use cached pose 'controlRigCP'())
    bAlphaBoolEnabled:bool=True
    Translation="0,0,0"
    Rotation="15,0,0"
   ) -> (Pose) {
    alphaInputType=Bool
    rotationMode=BMM_Additive
    rotationSpace=BCS_BoneSpace
    translationMode=BMM_Additive
    translationSpace=BCS_BoneSpace
    alphaBoolBlend {blendInTime=0.5, blendOption=CircularOut, blendOutTime=0.5}
    boneToModify {boneName=kin_headTowel_2}
    rotation {pitch=15}
   }
   n13 = "Transform (Modify) Bone - Bone: base_headTowel_3"!(ComponentPose=@n12, bAlphaBoolEnabled:bool=True, Translation="0,0,0", Rotation="5,0,0") -> (Pose) {
    alphaInputType=Bool
    rotationMode=BMM_Additive
    rotationSpace=BCS_BoneSpace
    translationMode=BMM_Additive
    translationSpace=BCS_BoneSpace
    alphaBoolBlend {blendInTime=0.5, blendOption=CircularOut, blendOutTime=0.5}
    boneToModify {boneName=base_headTowel_3}
    rotation {pitch=5}
   }
   Save cached pose 'generalAdjuistementsCP'(Pose=Component To Local(ComponentPose=@n13)) {cacheName=generalAdjuistementsCP}
  ]}
 EventGraph {nodes [
   Event BlueprintUpdateAnimation()
   > n1 = "Gravity Override State Function (Parameter Asset) (Nativized)"(
    CharacterPartAnimInstance=Self-Reference()
    GameTimeInSeconds=GetGameTimeInSeconds()
    Parameters=/BRCosmetics/Characters/Player/Female/Medium/Bodies/F_MED_SplishSplash/Meshes/F_MED_SplishSplash_Parameters
   )
   > Set GravityOverride(GravityOverride=@n1)
   n2 = Event OnCharacterCustomizationLoaded()
   > Set Head Skeleton(Head Skeleton:Skeletal Mesh Component Object Reference=Get Array Item with EFort Custom Part Type(Array=@n2.CustomMeshParts, PartType=Head)) -> (Output_Get:Skeletal Mesh Component Object Reference)
   n3 = CheckForBackbling()
   > Set isWearingBackbling(isWearingBackbling=AND Boolean(A=@n3, B=NOT Boolean(A=@n3.bIsBackblingACape)))
  ]}
]
variables [
 Gravity Override Params {
  type=FortGravityOverrideParameterSet
  default=(Skydive_GoingUp=(GravityOverride=(WindFrequency=(X=1.500000,Y=7.000000,Z=8.000000),WindAmplitude=(X=200.000000,Y=-500.000000,Z=400.000000),GravityOverride=(X=0.000000,Y=-300.000000,Z=1500.000000),JointName="root")),Skydive_FloatingDown=(GravityOverride=(WindFrequency=(X=1.500000,Y=7.000000,Z=8.000000),WindAmplitude=(X=200.000000,Y=-500.000000,Z=400.000000),WindOffset=(X=0.000000,Y=-600.000000,Z=400.000000),GravityOverride=(X=0.000000,Y=-700.000000,Z=600.000000),JointName="root")),Skydive_DivingDown=(GravityOverride=(WindFrequency=(X=1.500000,Y=12.000000,Z=10.000000),WindAmplitude=(X=50.000000,Y=300.000000,Z=100.000000),GravityOverride=(X=-1500.000000,Y=0.000000,Z=0.000000),JointName="pelvis")),Skydive_Parachuting=(GravityOverride=(WindFrequency=(X=1.500000,Y=7.000000,Z=8.000000),WindAmplitude=(X=200.000000,Y=-500.000000,Z=400.000000),GravityOverride=(X=0.000000,Y=-800.000000,Z=-980.000000),JointName="root")),OnGround_Standing=(GravityOverride=(GravityOverride=(X=0.000000,Y=0.000000,Z=-980.000000),JointName="root")),OnGround_Moving=(GravityOverride=(GravityOverride=(X=0.000000,Y=0.000000,Z=-980.000000),JointName="root")),OnGround_Moving_HighRotationRate=(GravityOverride=(GravityOverride=(X=0.000000,Y=0.000000,Z=-980.000000),JointName="root")),OnGround_Crouch_Moving=(GravityOverride=(GravityOverride=(X=0.000000,Y=0.000000,Z=-980.000000),JointName="root")),Emote_or_Melee=(GravityOverride=(GravityOverride=(X=0.000000,Y=0.000000,Z=-980.000000),JointName="root")),Swimming=(GravityOverride=(GravityOverride=(X=-800.000000,Y=0.000000,Z=0.000000),JointName="pelvis")),Front_End=(GravityOverride=(GravityOverride=(X=0.000000,Y=0.000000,Z=-980.000000),JointName="root")),DBNO=(GravityOverride=(GravityOverride=(X=-980.000000,Y=400.000000,Z=0.000000),JointName="pelvis")),DBNO_Carried=(GravityOverride=(GravityOverride=(X=-980.000000,Y=500.000000,Z=-500.000000),JointName="pelvis")),OnGround_Falling=(GravityOverride=(GravityOverride=(X=0.000000,Y=0.000000,Z=-980.000000),JointName="root")),Ziplining=(GravityOverride=(GravityOverride=(X=0.000000,Y=0.000000,Z=-980.000000),JointName="root")),Skydive_Surfing=(GravityOverride=(WindFrequency=(X=1.500000,Y=7.000000,Z=8.000000),WindAmplitude=(X=200.000000,Y=-500.000000,Z=400.000000),GravityOverride=(X=500.000000,Y=-300.000000,Z=-980.000000),JointName="root")),Riding_Standing=(GravityOverride=(GravityOverride=(X=0.000000,Y=-350.000000,Z=-980.000000),JointName="root")),Riding_Moving=(GravityOverride=(GravityOverride=(X=0.000000,Y=-450.000000,Z=-450.000000),JointName="root")),Riding_Moving_Targeting=(GravityOverride=(GravityOverride=(X=0.000000,Y=-200.000000,Z=-980.000000),JointName="root")),Tactical_Sprint=(GravityOverride=(GravityOverride=(X=0.000000,Y=-450.000000,Z=-500.000000),JointName="root")),Rail_Grinding=(GravityOverride=(GravityOverride=(X=0.000000,Y=0.000000,Z=-980.000000),JointName="root")),)
 }
 GravityOverride {type=Vector}
 Head Skeleton {type=SkeletalMeshComponent}
 isWearingBackbling {type=boolean}
]
